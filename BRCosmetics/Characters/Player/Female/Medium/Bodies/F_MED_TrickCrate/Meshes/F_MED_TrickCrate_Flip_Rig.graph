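# Control Rig for the F_MED_TrickCrate female player body, with helper functions for thigh pose reading and skirt aiming.
# Thigh Readers runs spherical pose readers on a driver bone against its parent, returning front, back, in, out and crouch weights, with optional debug draw and prints.
# Skirt Aim And Sec Vectors projects thigh axes onto pelvis-relative rest planes, choosing raw or projected vectors by thigh twist bone position.
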ControlRigBlueprint /BRCosmetics/Characters/Player/Female/Medium/Bodies/F_MED_TrickCrate/Meshes/F_MED_TrickCrate_Flip_Rig
graphs [
 New Function_ContainedGraph {nodes [
   n1 = RigVMFunctionEntryNode()
   > RigVMVariableNode(Value=/Script/ControlRig.RigUnit_HierarchyGetParent(bDefaultParent=True, Child=@n1.Driver Item) {is=struct RigVMUnitNode}.Parent) {variable=Driver Item Parent}
   To_String = RigVMDispatchNode(Value=@n1.Driver Item.Name) {template="DISPATCH_RigDispatch_ToString(in Value,out Result)"}
   > RigVMVariableNode(Value=@To_String.Result) {variable=Driver Item Name}
   > RigVMFunction_Sequence = RigVMAggregateNode()
   RigVMFunctionReturnNode(
    ExecuteContext=@RigVMFunction_Sequence.F
    Fr Post=RigVMVariableNode() {variable=Result Fr Post}.Value
    Crouch Pre=RigVMVariableNode() {variable=Result Crouch Pre}.Value
    Crouch Post=RigVMVariableNode() {variable=Result Crouch Post}.Value
    Crouch Ot=RigVMVariableNode() {variable=Result Crouch Ot}.Value
    Crouch Pre Ot=RigVMVariableNode() {variable=Result Crouch Pre Ot}.Value
    Crouch Post Ot=RigVMVariableNode() {variable=Result Crouch Post Ot}.Value
    Ba=RigVMVariableNode() {variable=Result Ba}.Value
    Ba Pre=RigVMVariableNode() {variable=Result Ba Pre}.Value
    Ba Post=RigVMVariableNode() {variable=Result Ba Post}.Value
    In=RigVMVariableNode() {variable=Result In}.Value
    In Pre=RigVMVariableNode() {variable=Result In Pre}.Value
    In Post=RigVMVariableNode() {variable=Result In Post}.Value
    Ot=RigVMVariableNode() {variable=Result Ot}.Value
    Ot Pre=RigVMVariableNode() {variable=Result Ot Pre}.Value
    Ot Post=RigVMVariableNode() {variable=Result Ot Post}.Value
    Fr=RigVMVariableNode() {variable=Result Fr}.Value
    Crouch=RigVMVariableNode() {variable=Result Crouch}.Value
    Fr Pre=RigVMVariableNode() {variable=Result Fr Pre}.Value
   )
   n2 = &RigUnit_SphericalPoseReader(
    ActiveRegionScaleFactors=(PositiveWidth=1,NegativeWidth=1,PositiveHeight=1,NegativeHeight=1)
    ActiveRegionSize=0.001
    Debug=(bDrawDebug=false,bDraw2D=False,bDrawLocalAxes=False,DebugScale=25,DebugSegments=20,DebugThickness=0.05)
    DriverAxis=(X=1,Y=0,Z=0)
    FalloffRegionScaleFactors=(PositiveWidth=1,NegativeWidth=1,PositiveHeight=1,NegativeHeight=1)
    FalloffSize=0.5
    FlipHeightScaling=False
    FlipWidthScaling=False
    RotationOffset=(X=-90,Y=0,Z=0)
    Debug.bDrawDebug=RigVMVariableNode() {variable=Fr Viz}.Value
    DriverItem=RigVMVariableNode() {variable=Driver Item}.Value
    OptionalParentItem=RigVMVariableNode() {variable=Driver Item Parent}.Value
    ExecutePin=@RigVMFunction_Sequence.A
   )
   > n3 = RigVMVariableNode(Value=@n2.OutputParam) {variable=Result Fr}
   RigVMFunction_ControlFlowBranch = &RigVMFunction_ControlFlowBranch(ExecuteContext=@n3.ExecuteContext, Condition=RigVMVariableNode() {variable=Fr Viz}.Value)
   n4 = &RigUnit_SphericalPoseReader(
    ActiveRegionScaleFactors=(PositiveWidth=1,NegativeWidth=1,PositiveHeight=1,NegativeHeight=1)
    ActiveRegionSize=0.001
    Debug=(bDrawDebug=false,bDraw2D=False,bDrawLocalAxes=False,DebugScale=25,DebugSegments=20,DebugThickness=0.05)
    DriverAxis=(X=1,Y=0,Z=0)
    FalloffRegionScaleFactors=(PositiveWidth=1,NegativeWidth=1,PositiveHeight=2,NegativeHeight=2)
    FalloffSize=0.25
    FlipHeightScaling=False
    FlipWidthScaling=False
    RotationOffset=(X=-90,Y=0,Z=45)
    ExecutePin=@RigVMFunction_ControlFlowBranch.Completed
    DriverItem=RigVMVariableNode() {variable=Driver Item}.Value
    OptionalParentItem=RigVMVariableNode() {variable=Driver Item Parent}.Value
    Debug.bDrawDebug=RigVMVariableNode() {variable=Fr Pre Viz}.Value
   )
   > n5 = RigVMVariableNode(Value=@n4.OutputParam) {variable=Result Fr Pre}
   n6 = &RigVMFunction_ControlFlowBranch(Condition=RigVMVariableNode() {variable=Fr Pre Viz}.Value, ExecuteContext=@n5.ExecuteContext)
   n7 = &RigUnit_SphericalPoseReader(
    ActiveRegionScaleFactors=(PositiveWidth=1,NegativeWidth=1,PositiveHeight=1,NegativeHeight=1)
    ActiveRegionSize=0.001
    Debug=(bDrawDebug=false,bDraw2D=False,bDrawLocalAxes=False,DebugScale=25,DebugSegments=20,DebugThickness=0.05)
    DriverAxis=(X=1,Y=0,Z=0)
    FalloffRegionScaleFactors=(PositiveWidth=0.5,NegativeWidth=0.5,PositiveHeight=1,NegativeHeight=1)
    FalloffSize=0.5
    FlipHeightScaling=False
    FlipWidthScaling=False
    RotationOffset=(X=-90,Y=0,Z=-45)
    ExecutePin=@n6.Completed
    Debug.bDrawDebug=RigVMVariableNode() {variable=Fr Post Viz}.Value
    DriverItem=RigVMVariableNode() {variable=Driver Item}.Value
    OptionalParentItem=RigVMVariableNode() {variable=Driver Item Parent}.Value
   )
   > n8 = RigVMVariableNode(Value=@n7.OutputParam) {variable=Result Fr Post}
   n9 = &RigVMFunction_ControlFlowBranch(Condition=RigVMVariableNode() {variable=Fr Post Viz}.Value, ExecuteContext=@n8.ExecuteContext)
   n10 = &RigUnit_SphericalPoseReader(
    ActiveRegionScaleFactors=(PositiveWidth=1,NegativeWidth=1,PositiveHeight=1,NegativeHeight=1)
    ActiveRegionSize=0.001
    Debug=(bDrawDebug=true,bDraw2D=False,bDrawLocalAxes=False,DebugScale=25,DebugSegments=20,DebugThickness=0.05)
    DriverAxis=(X=1,Y=0,Z=0)
    FalloffRegionScaleFactors=(PositiveWidth=1,NegativeWidth=1,PositiveHeight=0.125,NegativeHeight=0.125)
    FalloffSize=0.5
    FlipHeightScaling=False
    FlipWidthScaling=False
    RotationOffset=(X=-75,Y=0,Z=0)
    ExecutePin=@n9.Completed
    DriverItem=RigVMVariableNode() {variable=Driver Item}.Value
    Debug.bDrawDebug=RigVMVariableNode() {variable=Crouch Viz}.Value
    OptionalParentItem=RigVMVariableNode() {variable=Driver Item Parent}.Value
   )
   > n11 = RigVMVariableNode(Value=@n10.OutputParam) {variable=Result Crouch}
   n12 = &RigVMFunction_ControlFlowBranch(Condition=RigVMVariableNode() {variable=Crouch Viz}.Value, ExecuteContext=@n11.ExecuteContext)
   n13 = &RigUnit_SphericalPoseReader(
    ActiveRegionScaleFactors=(PositiveWidth=1,NegativeWidth=1,PositiveHeight=1,NegativeHeight=1)
    ActiveRegionSize=0.001
    Debug=(bDrawDebug=true,bDraw2D=False,bDrawLocalAxes=False,DebugScale=25,DebugSegments=20,DebugThickness=0.05)
    DriverAxis=(X=1,Y=0,Z=0)
    FalloffRegionScaleFactors=(PositiveWidth=0.4825,NegativeWidth=0.5,PositiveHeight=0.075,NegativeHeight=0.075)
    FalloffSize=0.5
    FlipHeightScaling=False
    FlipWidthScaling=False
    RotationOffset=(X=-101.3223,Y=-189.762,Z=42.3664)
    ExecutePin=@n12.Completed
    DriverItem=RigVMVariableNode() {variable=Driver Item}.Value
    OptionalParentItem=RigVMVariableNode() {variable=Driver Item Parent}.Value
    Debug.bDrawDebug=RigVMVariableNode() {variable=Crouch Pre Viz}.Value
   )
   > n14 = RigVMVariableNode(Value=@n13.OutputParam) {variable=Result Crouch Pre}
   n15 = &RigVMFunction_ControlFlowBranch(ExecuteContext=@n14.ExecuteContext, Condition=RigVMVariableNode() {variable=Crouch Pre Viz}.Value)
   n16 = &RigUnit_SphericalPoseReader(
    ActiveRegionScaleFactors=(PositiveWidth=1,NegativeWidth=1,PositiveHeight=1,NegativeHeight=1)
    ActiveRegionSize=0.001
    Debug=(bDrawDebug=true,bDraw2D=False,bDrawLocalAxes=False,DebugScale=25,DebugSegments=20,DebugThickness=0.05)
    DriverAxis=(X=1,Y=0,Z=0)
    FalloffRegionScaleFactors=(PositiveWidth=0.5,NegativeWidth=0.5,PositiveHeight=0.075,NegativeHeight=0.075)
    FalloffSize=0.5
    FlipHeightScaling=False
    FlipWidthScaling=False
    RotationOffset=(X=-101.5496,Y=-169.7841,Z=-42.8133)
    ExecutePin=@n15.Completed
    DriverItem=RigVMVariableNode() {variable=Driver Item}.Value
    OptionalParentItem=RigVMVariableNode() {variable=Driver Item Parent}.Value
    Debug.bDrawDebug=RigVMVariableNode() {variable=Crouch Post Viz}.Value
   )
   > n17 = RigVMVariableNode(Value=@n16.OutputParam) {variable=Result Crouch Post}
   n18 = &RigVMFunction_ControlFlowBranch(ExecuteContext=@n17.ExecuteContext, Condition=RigVMVariableNode() {variable=Crouch Post Viz}.Value)
   n19 = &RigUnit_SphericalPoseReader(
    ActiveRegionScaleFactors=(PositiveWidth=1,NegativeWidth=1,PositiveHeight=1,NegativeHeight=1)
    ActiveRegionSize=0.001
    Debug=(bDrawDebug=true,bDraw2D=False,bDrawLocalAxes=False,DebugScale=25,DebugSegments=20,DebugThickness=0.05)
    DriverAxis=(X=1,Y=0,Z=0)
    FalloffRegionScaleFactors=(PositiveWidth=1,NegativeWidth=1,PositiveHeight=0.666,NegativeHeight=0.333)
    FalloffSize=0.5
    FlipHeightScaling=False
    FlipWidthScaling=False
    RotationOffset=(X=-60,Y=0,Z=0)
    ExecutePin=@n18.Completed
    DriverItem=RigVMVariableNode() {variable=Driver Item}.Value
    OptionalParentItem=RigVMVariableNode() {variable=Driver Item Parent}.Value
    Debug.bDrawDebug=RigVMVariableNode() {variable=Crouch Ot Viz}.Value
   )
   > n20 = RigVMVariableNode(Value=@n19.OutputParam) {variable=Result Crouch Ot}
   n21 = &RigVMFunction_ControlFlowBranch(Condition=RigVMVariableNode() {variable=Crouch Ot Viz}.Value, ExecuteContext=@n20.ExecuteContext)
   n22 = &RigUnit_SphericalPoseReader(
    ActiveRegionScaleFactors=(PositiveWidth=1,NegativeWidth=1,PositiveHeight=1,NegativeHeight=1)
    ActiveRegionSize=0.001
    Debug=(bDrawDebug=true,bDraw2D=False,bDrawLocalAxes=False,DebugScale=25,DebugSegments=20,DebugThickness=0.05)
    DriverAxis=(X=1,Y=0,Z=0)
    FalloffRegionScaleFactors=(PositiveWidth=0.5,NegativeWidth=0.5,PositiveHeight=0.2,NegativeHeight=0.4)
    FalloffSize=0.5
    FlipHeightScaling=False
    FlipWidthScaling=False
    RotationOffset=(X=-110.6339,Y=-201.1783,Z=39.9802)
    ExecutePin=@n21.Completed
    DriverItem=RigVMVariableNode() {variable=Driver Item}.Value
    OptionalParentItem=RigVMVariableNode() {variable=Driver Item Parent}.Value
    Debug.bDrawDebug=RigVMVariableNode() {variable=Crouch Ot Pre Viz}.Value
   )
   > n23 = RigVMVariableNode(Value=@n22.OutputParam) {variable=Result Crouch Pre Ot}
   n24 = &RigVMFunction_ControlFlowBranch(Condition=RigVMVariableNode() {variable=Crouch Ot Pre Viz}.Value, ExecuteContext=@n23.ExecuteContext)
   n25 = &RigUnit_SphericalPoseReader(
    ActiveRegionScaleFactors=(PositiveWidth=1,NegativeWidth=1,PositiveHeight=1,NegativeHeight=1)
    ActiveRegionSize=0.001
    Debug=(bDrawDebug=true,bDraw2D=False,bDrawLocalAxes=False,DebugScale=25,DebugSegments=20,DebugThickness=0.05)
    DriverAxis=(X=1,Y=0,Z=0)
    FalloffRegionScaleFactors=(PositiveWidth=0.5,NegativeWidth=0.5,PositiveHeight=0.2,NegativeHeight=0.4)
    FalloffSize=0.5
    FlipHeightScaling=False
    FlipWidthScaling=False
    RotationOffset=(X=-112.4531,Y=-160.3786,Z=-41.3069)
    ExecutePin=@n24.Completed
    DriverItem=RigVMVariableNode() {variable=Driver Item}.Value
    OptionalParentItem=RigVMVariableNode() {variable=Driver Item Parent}.Value
    Debug.bDrawDebug=RigVMVariableNode() {variable=Crouch Ot Post Viz}.Value
   )
   > n26 = RigVMVariableNode(Value=@n25.OutputParam) {variable=Result Crouch Post Ot}
   n27 = &RigVMFunction_ControlFlowBranch(ExecuteContext=@n26.ExecuteContext, Condition=RigVMVariableNode() {variable=Crouch Ot Post Viz}.Value)
   RigVMDispatchNode(
    Enabled=True
    ScreenColor=(R=1,G=1,B=1,A=1)
    ScreenDuration=0.05
    Prefix=/Script/RigVM.RigVMFunction_StringConcat(B=" - CROUCH OT POST:   ", A=RigVMVariableNode() {variable=Driver Item Name}.Value) {is=struct RigVMUnitNode, template="Concat::Execute(in A,in B,out Result)"}.Result
    Value=RigVMVariableNode() {variable=Result Crouch Post Ot}.Value
    ExecuteContext=@n27.True
   ) {template="DISPATCH_RigVMDispatch_Print(in Prefix,in Value,in Enabled,in ScreenDuration,in ScreenColor)"}
   RigVMDispatchNode(
    Enabled=True
    ScreenColor=(R=1,G=1,B=1,A=1)
    ScreenDuration=0.05
    Prefix=&RigVMFunction_StringConcat(B=" - CROUCH OT PRE:   ", A=RigVMVariableNode() {variable=Driver Item Name}.Value) {template="Concat::Execute(in A,in B,out Result)"}.Result
    Value=RigVMVariableNode() {variable=Result Crouch Pre Ot}.Value
    ExecuteContext=@n24.True
   ) {template="DISPATCH_RigVMDispatch_Print(in Prefix,in Value,in Enabled,in ScreenDuration,in ScreenColor)"}
   RigVMDispatchNode(
    Enabled=True
    ScreenColor=(R=1,G=1,B=1,A=1)
    ScreenDuration=0.05
    Prefix=&RigVMFunction_StringConcat(B=" - Crouch Ot:   ", A=RigVMVariableNode() {variable=Driver Item Name}.Value) {template="Concat::Execute(in A,in B,out Result)"}.Result
    Value=RigVMVariableNode() {variable=Result Crouch Ot}.Value
    ExecuteContext=@n21.True
   ) {template="DISPATCH_RigVMDispatch_Print(in Prefix,in Value,in Enabled,in ScreenDuration,in ScreenColor)"}
   RigVMDispatchNode(
    Enabled=True
    ScreenColor=(R=1,G=1,B=1,A=1)
    ScreenDuration=0.05
    Prefix=&RigVMFunction_StringConcat(B=" - CROUCH POST:   ", A=RigVMVariableNode() {variable=Driver Item Name}.Value) {template="Concat::Execute(in A,in B,out Result)"}.Result
    Value=RigVMVariableNode() {variable=Result Crouch Post}.Value
    ExecuteContext=@n18.True
   ) {template="DISPATCH_RigVMDispatch_Print(in Prefix,in Value,in Enabled,in ScreenDuration,in ScreenColor)"}
   RigVMDispatchNode(
    Enabled=True
    ScreenColor=(R=1,G=1,B=1,A=1)
    ScreenDuration=0.05
    Prefix=&RigVMFunction_StringConcat(B=" - CROUCH PRE:   ", A=RigVMVariableNode() {variable=Driver Item Name}.Value) {template="Concat::Execute(in A,in B,out Result)"}.Result
    Value=RigVMVariableNode() {variable=Result Crouch Pre}.Value
    ExecuteContext=@n15.True
   ) {template="DISPATCH_RigVMDispatch_Print(in Prefix,in Value,in Enabled,in ScreenDuration,in ScreenColor)"}
   RigVMDispatchNode(
    Enabled=True
    ScreenColor=(R=1,G=1,B=1,A=1)
    ScreenDuration=0.05
    Prefix=&RigVMFunction_StringConcat(B=" - Crouch:   ", A=RigVMVariableNode() {variable=Driver Item Name}.Value) {template="Concat::Execute(in A,in B,out Result)"}.Result
    Value=RigVMVariableNode() {variable=Result Crouch}.Value
    ExecuteContext=@n12.True
   ) {template="DISPATCH_RigVMDispatch_Print(in Prefix,in Value,in Enabled,in ScreenDuration,in ScreenColor)"}
   RigVMDispatchNode(
    Enabled=True
    ScreenColor=(R=1,G=1,B=1,A=1)
    ScreenDuration=0.05
    Prefix=&RigVMFunction_StringConcat(B=" - FR POST:   ", A=RigVMVariableNode() {variable=Driver Item Name}.Value) {template="Concat::Execute(in A,in B,out Result)"}.Result
    Value=RigVMVariableNode() {variable=Result Fr Post}.Value
    ExecuteContext=@n9.True
   ) {template="DISPATCH_RigVMDispatch_Print(in Prefix,in Value,in Enabled,in ScreenDuration,in ScreenColor)"}
   RigVMDispatchNode(
    Enabled=True
    ScreenColor=(R=1,G=1,B=1,A=1)
    ScreenDuration=0.05
    Prefix=&RigVMFunction_StringConcat(B=" - FR PRE:   ", A=RigVMVariableNode() {variable=Driver Item Name}.Value) {template="Concat::Execute(in A,in B,out Result)"}.Result
    Value=RigVMVariableNode() {variable=Result Fr Pre}.Value
    ExecuteContext=@n6.True
   ) {template="DISPATCH_RigVMDispatch_Print(in Prefix,in Value,in Enabled,in ScreenDuration,in ScreenColor)"}
   Concat = &RigVMFunction_StringConcat(B=" - FR:   ", A=RigVMVariableNode() {variable=Driver Item Name}.Value) {template="Concat::Execute(in A,in B,out Result)"}
   RigVMDispatchNode(
    Enabled=True
    ScreenColor=(R=1,G=1,B=1,A=1)
    ScreenDuration=0.05
    Prefix=@Concat.Result
    Value=RigVMVariableNode() {variable=Result Fr}.Value
    ExecuteContext=@RigVMFunction_ControlFlowBranch.True
   ) {template="DISPATCH_RigVMDispatch_Print(in Prefix,in Value,in Enabled,in ScreenDuration,in ScreenColor)"}
   n28 = &RigUnit_SphericalPoseReader(
    ActiveRegionScaleFactors=(PositiveWidth=1,NegativeWidth=1,PositiveHeight=1,NegativeHeight=1)
    ActiveRegionSize=0.001
    Debug=(bDrawDebug=false,bDraw2D=False,bDrawLocalAxes=False,DebugScale=25,DebugSegments=20,DebugThickness=0.05)
    DriverAxis=(X=1,Y=0,Z=0)
    FalloffRegionScaleFactors=(PositiveWidth=1,NegativeWidth=1,PositiveHeight=0.5,NegativeHeight=0.5)
    FalloffSize=0.5
    FlipHeightScaling=False
    FlipWidthScaling=False
    RotationOffset=(X=-135,Y=0,Z=0)
    DriverItem=RigVMVariableNode() {variable=Driver Item}.Value
    OptionalParentItem=RigVMVariableNode() {variable=Driver Item Parent}.Value
    Debug.bDrawDebug=RigVMVariableNode() {variable=In Viz}.Value
    ExecutePin=@RigVMFunction_Sequence.C
   )
   > n29 = RigVMVariableNode(Value=@n28.OutputParam) {variable=Result In}
   n30 = &RigVMFunction_ControlFlowBranch(ExecuteContext=@n29.ExecuteContext, Condition=RigVMVariableNode() {variable=In Viz}.Value)
   n31 = &RigUnit_SphericalPoseReader(
    ActiveRegionScaleFactors=(PositiveWidth=1,NegativeWidth=1,PositiveHeight=1,NegativeHeight=1)
    ActiveRegionSize=0.001
    Debug=(bDrawDebug=false,bDraw2D=False,bDrawLocalAxes=False,DebugScale=25,DebugSegments=20,DebugThickness=0.05)
    DriverAxis=(X=1,Y=0,Z=0)
    FalloffRegionScaleFactors=(PositiveWidth=2,NegativeWidth=2,PositiveHeight=0.5,NegativeHeight=0.5)
    FalloffSize=0.25
    FlipHeightScaling=False
    FlipWidthScaling=False
    RotationOffset=(X=-112.5,Y=0,Z=0)
    ExecutePin=@n30.Completed
    DriverItem=RigVMVariableNode() {variable=Driver Item}.Value
    OptionalParentItem=RigVMVariableNode() {variable=Driver Item Parent}.Value
    Debug.bDrawDebug=RigVMVariableNode() {variable=In Pre Viz}.Value
   )
   > n32 = RigVMVariableNode(Value=@n31.OutputParam) {variable=Result In Pre}
   n33 = &RigVMFunction_ControlFlowBranch(ExecuteContext=@n32.ExecuteContext, Condition=RigVMVariableNode() {variable=In Pre Viz}.Value)
   n34 = &RigUnit_SphericalPoseReader(
    ActiveRegionScaleFactors=(PositiveWidth=1,NegativeWidth=1,PositiveHeight=1,NegativeHeight=1)
    ActiveRegionSize=0.001
    Debug=(bDrawDebug=false,bDraw2D=False,bDrawLocalAxes=False,DebugScale=25,DebugSegments=20,DebugThickness=0.05)
    DriverAxis=(X=1,Y=0,Z=0)
    FalloffRegionScaleFactors=(PositiveWidth=2,NegativeWidth=2,PositiveHeight=1,NegativeHeight=1)
    FalloffSize=0.25
    FlipHeightScaling=False
    FlipWidthScaling=False
    RotationOffset=(X=135,Y=0,Z=90)
    ExecutePin=@n33.Completed
    DriverItem=RigVMVariableNode() {variable=Driver Item}.Value
    OptionalParentItem=RigVMVariableNode() {variable=Driver Item Parent}.Value
    Debug.bDrawDebug=RigVMVariableNode() {variable=In Post Viz}.Value
   )
   > n35 = RigVMVariableNode(Value=@n34.OutputParam) {variable=Result In Post}
   n36 = &RigVMFunction_ControlFlowBranch(ExecuteContext=@n35.ExecuteContext, Condition=RigVMVariableNode() {variable=In Post Viz}.Value)
   RigVMDispatchNode(
    Enabled=True
    ScreenColor=(R=1,G=1,B=1,A=1)
    ScreenDuration=0.05
    Prefix=&RigVMFunction_StringConcat(B=" - IN POST:   ", A=RigVMVariableNode() {variable=Driver Item Name}.Value) {template="Concat::Execute(in A,in B,out Result)"}.Result
    ExecuteContext=@n36.True
    Value=RigVMVariableNode() {variable=Result In Post}.Value
   ) {template="DISPATCH_RigVMDispatch_Print(in Prefix,in Value,in Enabled,in ScreenDuration,in ScreenColor)"}
   RigVMDispatchNode(
    Enabled=True
    ScreenColor=(R=1,G=1,B=1,A=1)
    ScreenDuration=0.05
    Prefix=&RigVMFunction_StringConcat(B=" - IN PRE:   ", A=RigVMVariableNode() {variable=Driver Item Name}.Value) {template="Concat::Execute(in A,in B,out Result)"}.Result
    ExecuteContext=@n33.True
    Value=RigVMVariableNode() {variable=Result In Pre}.Value
   ) {template="DISPATCH_RigVMDispatch_Print(in Prefix,in Value,in Enabled,in ScreenDuration,in ScreenColor)"}
   RigVMDispatchNode(
    Enabled=True
    ScreenColor=(R=1,G=1,B=1,A=1)
    ScreenDuration=0.05
    Prefix=&RigVMFunction_StringConcat(B=" - IN:   ", A=RigVMVariableNode() {variable=Driver Item Name}.Value) {template="Concat::Execute(in A,in B,out Result)"}.Result
    ExecuteContext=@n30.True
    Value=RigVMVariableNode() {variable=Result In}.Value
   ) {template="DISPATCH_RigVMDispatch_Print(in Prefix,in Value,in Enabled,in ScreenDuration,in ScreenColor)"}
   n37 = &RigUnit_SphericalPoseReader(
    ActiveRegionScaleFactors=(PositiveWidth=1,NegativeWidth=1,PositiveHeight=1,NegativeHeight=1)
    ActiveRegionSize=0.001
    Debug=(bDrawDebug=false,bDraw2D=False,bDrawLocalAxes=False,DebugScale=25,DebugSegments=20,DebugThickness=0.05)
    DriverAxis=(X=1,Y=0,Z=0)
    FalloffRegionScaleFactors=(PositiveWidth=1,NegativeWidth=1,PositiveHeight=1,NegativeHeight=1)
    FalloffSize=0.5
    FlipHeightScaling=False
    FlipWidthScaling=False
    RotationOffset=(X=0,Y=0,Z=0)
    DriverItem=RigVMVariableNode() {variable=Driver Item}.Value
    OptionalParentItem=RigVMVariableNode() {variable=Driver Item Parent}.Value
    Debug.bDrawDebug=RigVMVariableNode() {variable=Ot Viz}.Value
    ExecutePin=@RigVMFunction_Sequence.D
   )
   > n38 = RigVMVariableNode(Value=@n37.OutputParam) {variable=Result Ot}
   n39 = &RigVMFunction_ControlFlowBranch(ExecuteContext=@n38.ExecuteContext, Condition=RigVMVariableNode() {variable=Ot Viz}.Value)
   n40 = &RigUnit_SphericalPoseReader(
    ActiveRegionScaleFactors=(PositiveWidth=1,NegativeWidth=1,PositiveHeight=1,NegativeHeight=1)
    ActiveRegionSize=0.001
    Debug=(bDrawDebug=false,bDraw2D=False,bDrawLocalAxes=False,DebugScale=25,DebugSegments=20,DebugThickness=0.05)
    DriverAxis=(X=1,Y=0,Z=0)
    FalloffRegionScaleFactors=(PositiveWidth=1,NegativeWidth=1,PositiveHeight=2,NegativeHeight=2)
    FalloffSize=0.25
    FlipHeightScaling=False
    FlipWidthScaling=False
    RotationOffset=(X=0,Y=-45,Z=0)
    ExecutePin=@n39.Completed
    Debug.bDrawDebug=RigVMVariableNode() {variable=Ot Pre Viz}.Value
    DriverItem=RigVMVariableNode() {variable=Driver Item}.Value
    OptionalParentItem=RigVMVariableNode() {variable=Driver Item Parent}.Value
   )
   > n41 = RigVMVariableNode(Value=@n40.OutputParam) {variable=Result Ot Pre}
   n42 = &RigVMFunction_ControlFlowBranch(Condition=RigVMVariableNode() {variable=Ot Pre Viz}.Value, ExecuteContext=@n41.ExecuteContext)
   n43 = &RigUnit_SphericalPoseReader(
    ActiveRegionScaleFactors=(PositiveWidth=1,NegativeWidth=1,PositiveHeight=1,NegativeHeight=1)
    ActiveRegionSize=0.001
    Debug=(bDrawDebug=false,bDraw2D=False,bDrawLocalAxes=False,DebugScale=25,DebugSegments=20,DebugThickness=0.05)
    DriverAxis=(X=1,Y=0,Z=0)
    FalloffRegionScaleFactors=(PositiveWidth=1,NegativeWidth=1,PositiveHeight=2,NegativeHeight=2)
    FalloffSize=0.25
    FlipHeightScaling=False
    FlipWidthScaling=False
    RotationOffset=(X=0,Y=45,Z=0)
    ExecutePin=@n42.Completed
    Debug.bDrawDebug=RigVMVariableNode() {variable=Ot Post Viz}.Value
    DriverItem=RigVMVariableNode() {variable=Driver Item}.Value
    OptionalParentItem=RigVMVariableNode() {variable=Driver Item Parent}.Value
   )
   > n44 = RigVMVariableNode(Value=@n43.OutputParam) {variable=Result Ot Post}
   n45 = &RigVMFunction_ControlFlowBranch(Condition=RigVMVariableNode() {variable=Ot Post Viz}.Value, ExecuteContext=@n44.ExecuteContext)
   RigVMDispatchNode(
    Enabled=True
    ScreenColor=(R=1,G=1,B=1,A=1)
    ScreenDuration=0.05
    Prefix=&RigVMFunction_StringConcat(B=" - OT POST:   ", A=RigVMVariableNode() {variable=Driver Item Name}.Value) {template="Concat::Execute(in A,in B,out Result)"}.Result
    ExecuteContext=@n45.True
    Value=RigVMVariableNode() {variable=Result Ot Post}.Value
   ) {template="DISPATCH_RigVMDispatch_Print(in Prefix,in Value,in Enabled,in ScreenDuration,in ScreenColor)"}
   RigVMDispatchNode(
    Enabled=True
    ScreenColor=(R=1,G=1,B=1,A=1)
    ScreenDuration=0.05
    Prefix=&RigVMFunction_StringConcat(B=" - OT PRE:   ", A=RigVMVariableNode() {variable=Driver Item Name}.Value) {template="Concat::Execute(in A,in B,out Result)"}.Result
    ExecuteContext=@n42.True
    Value=RigVMVariableNode() {variable=Result Ot Pre}.Value
   ) {template="DISPATCH_RigVMDispatch_Print(in Prefix,in Value,in Enabled,in ScreenDuration,in ScreenColor)"}
   RigVMDispatchNode(
    Enabled=True
    ScreenColor=(R=1,G=1,B=1,A=1)
    ScreenDuration=0.05
    Prefix=&RigVMFunction_StringConcat(B=" - OT:   ", A=RigVMVariableNode() {variable=Driver Item Name}.Value) {template="Concat::Execute(in A,in B,out Result)"}.Result
    ExecuteContext=@n39.True
    Value=RigVMVariableNode() {variable=Result Ot}.Value
   ) {template="DISPATCH_RigVMDispatch_Print(in Prefix,in Value,in Enabled,in ScreenDuration,in ScreenColor)"}
   n46 = &RigUnit_SphericalPoseReader(
    ActiveRegionScaleFactors=(PositiveWidth=1,NegativeWidth=1,PositiveHeight=1,NegativeHeight=1)
    ActiveRegionSize=0.001
    Debug=(bDrawDebug=false,bDraw2D=False,bDrawLocalAxes=False,DebugScale=25,DebugSegments=20,DebugThickness=0.05)
    DriverAxis=(X=1,Y=0,Z=0)
    FalloffRegionScaleFactors=(PositiveWidth=1,NegativeWidth=1,PositiveHeight=1,NegativeHeight=1)
    FalloffSize=0.5
    FlipHeightScaling=False
    FlipWidthScaling=False
    RotationOffset=(X=90,Y=0,Z=0)
    DriverItem=RigVMVariableNode() {variable=Driver Item}.Value
    OptionalParentItem=RigVMVariableNode() {variable=Driver Item Parent}.Value
    Debug.bDrawDebug=RigVMVariableNode() {variable=Ba Viz}.Value
    ExecutePin=@RigVMFunction_Sequence.B
   )
   > n47 = RigVMVariableNode(Value=@n46.OutputParam) {variable=Result Ba}
   n48 = &RigVMFunction_ControlFlowBranch(ExecuteContext=@n47.ExecuteContext, Condition=RigVMVariableNode() {variable=Ba Viz}.Value)
   n49 = &RigUnit_SphericalPoseReader(
    ActiveRegionScaleFactors=(PositiveWidth=1,NegativeWidth=1,PositiveHeight=1,NegativeHeight=1)
    ActiveRegionSize=0.001
    Debug=(bDrawDebug=false,bDraw2D=False,bDrawLocalAxes=False,DebugScale=25,DebugSegments=20,DebugThickness=0.05)
    DriverAxis=(X=1,Y=0,Z=0)
    FalloffRegionScaleFactors=(PositiveWidth=1,NegativeWidth=1,PositiveHeight=2,NegativeHeight=2)
    FalloffSize=0.25
    FlipHeightScaling=False
    FlipWidthScaling=False
    RotationOffset=(X=90,Y=0,Z=-45)
    ExecutePin=@n48.Completed
    DriverItem=RigVMVariableNode() {variable=Driver Item}.Value
    OptionalParentItem=RigVMVariableNode() {variable=Driver Item Parent}.Value
    Debug.bDrawDebug=RigVMVariableNode() {variable=Ba Pre Viz}.Value
   )
   > n50 = RigVMVariableNode(Value=@n49.OutputParam) {variable=Result Ba Pre}
   n51 = &RigVMFunction_ControlFlowBranch(ExecuteContext=@n50.ExecuteContext, Condition=RigVMVariableNode() {variable=Ba Pre Viz}.Value)
   n52 = &RigUnit_SphericalPoseReader(
    ActiveRegionScaleFactors=(PositiveWidth=1,NegativeWidth=1,PositiveHeight=1,NegativeHeight=1)
    ActiveRegionSize=0.001
    Debug=(bDrawDebug=false,bDraw2D=False,bDrawLocalAxes=False,DebugScale=25,DebugSegments=20,DebugThickness=0.05)
    DriverAxis=(X=1,Y=0,Z=0)
    FalloffRegionScaleFactors=(PositiveWidth=1,NegativeWidth=1,PositiveHeight=2,NegativeHeight=2)
    FalloffSize=0.25
    FlipHeightScaling=False
    FlipWidthScaling=False
    RotationOffset=(X=90,Y=0,Z=45)
    ExecutePin=@n51.Completed
    DriverItem=RigVMVariableNode() {variable=Driver Item}.Value
    OptionalParentItem=RigVMVariableNode() {variable=Driver Item Parent}.Value
    Debug.bDrawDebug=RigVMVariableNode() {variable=Ba Post Viz}.Value
   )
   > n53 = RigVMVariableNode(Value=@n52.OutputParam) {variable=Result Ba Post}
   n54 = &RigVMFunction_ControlFlowBranch(ExecuteContext=@n53.ExecuteContext, Condition=RigVMVariableNode() {variable=Ba Post Viz}.Value)
   RigVMDispatchNode(
    Enabled=True
    ScreenColor=(R=1,G=1,B=1,A=1)
    ScreenDuration=0.05
    Prefix=&RigVMFunction_StringConcat(B=" - BA POST:   ", A=RigVMVariableNode() {variable=Driver Item Name}.Value) {template="Concat::Execute(in A,in B,out Result)"}.Result
    ExecuteContext=@n54.True
    Value=RigVMVariableNode() {variable=Result Ba Post}.Value
   ) {template="DISPATCH_RigVMDispatch_Print(in Prefix,in Value,in Enabled,in ScreenDuration,in ScreenColor)"}
   RigVMDispatchNode(
    Enabled=True
    ScreenColor=(R=1,G=1,B=1,A=1)
    ScreenDuration=0.05
    Prefix=&RigVMFunction_StringConcat(B=" - BA PRE:   ", A=RigVMVariableNode() {variable=Driver Item Name}.Value) {template="Concat::Execute(in A,in B,out Result)"}.Result
    ExecuteContext=@n51.True
    Value=RigVMVariableNode() {variable=Result Ba Pre}.Value
   ) {template="DISPATCH_RigVMDispatch_Print(in Prefix,in Value,in Enabled,in ScreenDuration,in ScreenColor)"}
   RigVMDispatchNode(
    Enabled=True
    ScreenColor=(R=1,G=1,B=1,A=1)
    ScreenDuration=0.05
    Prefix=&RigVMFunction_StringConcat(B=" - BA:   ", A=RigVMVariableNode() {variable=Driver Item Name}.Value) {template="Concat::Execute(in A,in B,out Result)"}.Result
    ExecuteContext=@n48.True
    Value=RigVMVariableNode() {variable=Result Ba}.Value
   ) {template="DISPATCH_RigVMDispatch_Print(in Prefix,in Value,in Enabled,in ScreenDuration,in ScreenColor)"}
  ]}
 New Function_ContainedGraph {nodes [
   n1 = RigVMFunctionEntryNode()
   RigVMFunction_MathVectorDot = /Script/RigVM.RigVMFunction_MathVectorDot(A=@n1.In Vector, B=@n1.Plane Normal) {is=struct RigVMUnitNode}
   RigVMFunction_MathVectorLengthSquared = /Script/RigVM.RigVMFunction_MathVectorLengthSquared(Value=@n1.Plane Normal) {is=struct RigVMUnitNode, template="LengthSquared::Execute(in Value,out Result)"}
   Divide = &RigVMFunction_MathDoubleDiv(A=@RigVMFunction_MathVectorDot.Result, B=@RigVMFunction_MathVectorLengthSquared.Result)
   Scale = /Script/RigVM.RigVMFunction_MathVectorScale(Factor=@Divide.Result, Value=@n1.Plane Normal) {is=struct RigVMUnitNode, template="Scale::Execute(in Value,in Factor,out Result)"}
   Subtract = /Script/RigVM.RigVMFunction_MathVectorSub(A=@n1.In Vector, B=@Scale.Result) {is=struct RigVMUnitNode, template="Subtract::Execute(in A,in B,out Result)"}
   > RigVMFunctionReturnNode(Projected Vector=@Subtract.Result)
  ]}
 New Function_ContainedGraph {nodes [
   RigVMFunctionEntryNode()
   > RigVMFunction_Sequence_1 = RigVMAggregateNode()
   RigVMFunctionReturnNode(
    ExecuteContext=@RigVMFunction_Sequence_1.C
    L Fr Sec Vector=RigVMVariableNode() {variable=L Front Sec Result}.Value
    L Ba Sec Vector=RigVMVariableNode() {variable=L Back Sec Result}.Value
    L IO Sec Vector=RigVMVariableNode() {variable=L InOt Sec Result}.Value
    L Fr Vector=RigVMVariableNode() {variable=L Front Result}.Value
    L Ba Vector=RigVMVariableNode() {variable=L Back Result}.Value
    L InOt Vector=RigVMVariableNode() {variable=L InOt Result}.Value
    R Fr Vector=RigVMVariableNode() {variable=R Front Result}.Value
    R Ba Vector=RigVMVariableNode() {variable=R Back Result}.Value
    R InOt Vector=RigVMVariableNode() {variable=R InOt Result}.Value
   )
   RigVMFunction_Sequence = RigVMAggregateNode(ExecuteContext=())
   n1 = /Script/RigVM.RigVMFunction_MathTransformRotateVector(Vector=(X=-1,Y=0,Z=0), Transform=&RigUnit_GetTransform(Item=(Type=Bone,Name="thigh_l"), bInitial=False).Transform) {is=struct RigVMUnitNode, template="Rotate Vector::Execute(in Transform,in Vector,out Result)"}
   n2 = /Script/RigVM.RigVMFunction_MathTransformMakeAbsolute(
    Local=&RigVMFunction_MathTransformMakeRelative(
     Global=&RigUnit_GetTransform(Item=(Type=Bone,Name="thigh_l"), bInitial=true).Transform
     Parent=&RigUnit_GetTransform(Item=(Type=Bone,Name="pelvis"), bInitial=true).Transform
    ).Local
    Parent=&RigUnit_GetTransform(Item=(Type=Bone,Name="pelvis"), bInitial=false).Transform
   ) {is=struct RigVMUnitNode, template="Make Absolute::Execute(in Local,in Parent,out Global)"}
   Project Vector Onto Plane = RigVMFunctionReferenceNode(
    In Vector=@n1.Result
    Plane Normal=&RigVMFunction_MathTransformRotateVector(Vector=(X=0,Y=1,Z=0), Transform=@n2.Global) {template="Rotate Vector::Execute(in Transform,in Vector,out Result)"}.Result
    ExecuteContext=@RigVMFunction_Sequence.A
   )
   n3 = &RigVMFunction_MathTransformMakeRelative(Parent=@n2.Global, Global=&RigUnit_GetTransform(Item=(Type=Bone,Name="thigh_twist_01_l"), bInitial=False).Transform)
   n4 = /Script/RigVM.RigVMFunction_MathDoubleGreaterEqual(B=0, A=@n3.Local.Translation.Y) {is=struct RigVMUnitNode, template="GreaterEqual::Execute(in A,in B,out Result)"}
   GreaterEqual = &RigVMFunction_MathDoubleGreaterEqual(B=0, A=@n3.Local.Translation.X) {template="GreaterEqual::Execute(in A,in B,out Result)"}
   Scale = &RigVMFunction_MathVectorScale(Factor=-1, Value=@"Project Vector Onto Plane".Projected Vector) {template="Scale::Execute(in Value,in Factor,out Result)"}
   n5 = RigVMDispatchNode(Condition=@GreaterEqual.Result, False=@"Project Vector Onto Plane".Projected Vector, True=@Scale.Result) {template="DISPATCH_RigVMDispatch_If(in Condition,in True,in False,out Result)"}
   > RigVMVariableNode(Value=RigVMDispatchNode(Condition=@n4.Result, False=@n5.Result, True=@n1.Result) {template="DISPATCH_RigVMDispatch_If(in Condition,in True,in False,out Result)"}.Result) {variable=L Front Result}
   > RigVMVariableNode(Value=RigVMDispatchNode(Condition=@n4.Result, True=@n5.Result, False=@n1.Result) {template="DISPATCH_RigVMDispatch_If(in Condition,in True,in False,out Result)"}.Result) {variable=L Back Result}
   n6 = &RigVMFunction_MathTransformRotateVector(Vector=(X=0,Y=1,Z=0), Transform=@n2.Global) {template="Rotate Vector::Execute(in Transform,in Vector,out Result)"}
   n7 = &RigVMFunction_MathTransformRotateVector(Vector=(X=0,Y=1,Z=0), Transform=&RigUnit_GetTransform(Item=(Type=Bone,Name="thigh_twist_01_l"), bInitial=False).Transform) {template="Rotate Vector::Execute(in Transform,in Vector,out Result)"}
   > RigVMVariableNode(Value=RigVMDispatchNode(Condition=@n4.Result, False=@n6.Result, True=@n7.Result) {template="DISPATCH_RigVMDispatch_If(in Condition,in True,in False,out Result)"}.Result) {variable=L Front Sec Result}
   > RigVMVariableNode(Value=RigVMDispatchNode(Condition=@n4.Result, True=@n6.Result, False=@n7.Result) {template="DISPATCH_RigVMDispatch_If(in Condition,in True,in False,out Result)"}.Result) {variable=Skirt L Ba Sec Vector}
   n8 = &RigVMFunction_MathTransformRotateVector(Vector=(X=-1,Y=0,Z=0), Transform=&RigUnit_GetTransform(Item=(Type=Bone,Name="thigh_l"), bInitial=False).Transform) {template="Rotate Vector::Execute(in Transform,in Vector,out Result)"}
   n9 = &RigVMFunction_MathTransformMakeAbsolute(
    Local=&RigVMFunction_MathTransformMakeRelative(
     Global=&RigUnit_GetTransform(Item=(Type=Bone,Name="thigh_l"), bInitial=true).Transform
     Parent=&RigUnit_GetTransform(Item=(Type=Bone,Name="pelvis"), bInitial=true).Transform
    ).Local
    Parent=&RigUnit_GetTransform(Item=(Type=Bone,Name="pelvis"), bInitial=false).Transform
   ) {template="Make Absolute::Execute(in Local,in Parent,out Global)"}
   Project Vector Onto Plane_1 = RigVMFunctionReferenceNode(
    In Vector=@n8.Result
    Plane Normal=&RigVMFunction_MathTransformRotateVector(Vector=(X=0,Y=0,Z=1), Transform=@n9.Global) {template="Rotate Vector::Execute(in Transform,in Vector,out Result)"}.Result
    ExecuteContext=@RigVMFunction_Sequence.B
   )
   n10 = &RigVMFunction_MathTransformMakeRelative(Parent=@n9.Global, Global=&RigUnit_GetTransform(Item=(Type=Bone,Name="thigh_twist_01_l"), bInitial=False).Transform)
   > RigVMVariableNode(Value=RigVMDispatchNode(
     Condition=&RigVMFunction_MathDoubleGreaterEqual(B=0, A=@n10.Local.Translation.Z) {template="GreaterEqual::Execute(in A,in B,out Result)"}.Result
     True=@"Project Vector Onto Plane_1".Projected Vector
     False=@n8.Result
    ) {template="DISPATCH_RigVMDispatch_If(in Condition,in True,in False,out Result)"}.Result) {variable=L InOt Result}
   > RigVMVariableNode(Value=&RigVMFunction_MathTransformRotateVector(Vector=(X=0,Y=1,Z=0), Transform=@n9.Global) {template="Rotate Vector::Execute(in Transform,in Vector,out Result)"}.Result) {variable=Skirt L IO Sec Vector}
   n11 = &RigVMFunction_MathTransformRotateVector(Vector=(X=1,Y=0,Z=0), Transform=&RigUnit_GetTransform(Item=(Type=Bone,Name="thigh_r"), bInitial=False).Transform) {template="Rotate Vector::Execute(in Transform,in Vector,out Result)"}
   n12 = &RigVMFunction_MathTransformMakeAbsolute(
    Local=&RigVMFunction_MathTransformMakeRelative(
     Global=&RigUnit_GetTransform(Item=(Type=Bone,Name="thigh_r"), bInitial=true).Transform
     Parent=&RigUnit_GetTransform(Item=(Type=Bone,Name="pelvis"), bInitial=true).Transform
    ).Local
    Parent=&RigUnit_GetTransform(Item=(Type=Bone,Name="pelvis"), bInitial=false).Transform
   ) {template="Make Absolute::Execute(in Local,in Parent,out Global)"}
   Project Vector Onto Plane_2 = RigVMFunctionReferenceNode(
    In Vector=@n11.Result
    Plane Normal=&RigVMFunction_MathTransformRotateVector(Vector=(X=0,Y=-1,Z=0), Transform=@n12.Global) {template="Rotate Vector::Execute(in Transform,in Vector,out Result)"}.Result
    ExecuteContext=@RigVMFunction_Sequence.C
   )
   n13 = &RigVMFunction_MathDoubleGreaterEqual(
    B=0
    A=&RigVMFunction_MathTransformMakeRelative(Parent=@n12.Global, Global=&RigUnit_GetTransform(Item=(Type=Bone,Name="thigh_twist_01_r"), bInitial=False).Transform).Local.Translation.Y
   ) {template="GreaterEqual::Execute(in A,in B,out Result)"}
   > RigVMVariableNode(Value=RigVMDispatchNode(Condition=@n13.Result, True=@"Project Vector Onto Plane_2".Projected Vector, False=@n11.Result) {template="DISPATCH_RigVMDispatch_If(in Condition,in True,in False,out Result)"}.Result) {variable=R Front Result}
   > RigVMVariableNode(Value=RigVMDispatchNode(Condition=@n13.Result, False=@"Project Vector Onto Plane_2".Projected Vector, True=@n11.Result) {template="DISPATCH_RigVMDispatch_If(in Condition,in True,in False,out Result)"}.Result) {variable=R Back Result}
   n14 = &RigVMFunction_MathTransformRotateVector(Vector=(X=1,Y=0,Z=0), Transform=&RigUnit_GetTransform(Item=(Type=Bone,Name="thigh_r"), bInitial=False).Transform) {template="Rotate Vector::Execute(in Transform,in Vector,out Result)"}
   n15 = &RigVMFunction_MathTransformMakeAbsolute(
    Local=&RigVMFunction_MathTransformMakeRelative(
     Global=&RigUnit_GetTransform(Item=(Type=Bone,Name="thigh_r"), bInitial=true).Transform
     Parent=&RigUnit_GetTransform(Item=(Type=Bone,Name="pelvis"), bInitial=true).Transform
    ).Local
    Parent=&RigUnit_GetTransform(Item=(Type=Bone,Name="pelvis"), bInitial=false).Transform
   ) {template="Make Absolute::Execute(in Local,in Parent,out Global)"}
   Project Vector Onto Plane_3 = RigVMFunctionReferenceNode(
    In Vector=@n14.Result
    Plane Normal=&RigVMFunction_MathTransformRotateVector(Vector=(X=0,Y=0,Z=1), Transform=@n15.Global) {template="Rotate Vector::Execute(in Transform,in Vector,out Result)"}.Result
    ExecuteContext=@RigVMFunction_Sequence.D
   )
   n16 = &RigVMFunction_MathTransformMakeRelative(Parent=@n15.Global, Global=&RigUnit_GetTransform(Item=(Type=Bone,Name="thigh_twist_01_r"), bInitial=False).Transform)
   > RigVMVariableNode(Value=RigVMDispatchNode(
     Condition=&RigVMFunction_MathDoubleGreaterEqual(B=0, A=@n16.Local.Translation.Z) {template="GreaterEqual::Execute(in A,in B,out Result)"}.Result
     False=@"Project Vector Onto Plane_3".Projected Vector
     True=@n14.Result
    ) {template="DISPATCH_RigVMDispatch_If(in Condition,in True,in False,out Result)"}.Result) {variable=R InOt Result}
   Project Vector Onto Plane_4 = RigVMFunctionReferenceNode(ExecuteContext=(), In Vector=(X=0,Y=0,Z=0), Plane Normal=(X=0,Y=0,Z=0))
   RigVMFunction_MathTransformFromSRT = /Script/RigVM.RigVMFunction_MathTransformFromSRT(Location=(X=0,Y=0,Z=0), Rotation=(X=0,Y=0,Z=0), RotationOrder=XYZ, Scale=(X=1,Y=1,Z=1)) {is=struct RigVMUnitNode}
   RigVMFunction_MathTransformTransformVector = /Script/RigVM.RigVMFunction_MathTransformTransformVector(Location=(X=0,Y=0,Z=0), Transform=(Rotation=(X=0,Y=0,Z=0,W=1),Translation=(X=0,Y=0,Z=0),Scale3D=(X=1,Y=1,Z=1))) {is=struct RigVMUnitNode}
   Rotate Vector_11 = RigVMUnitNode(Transform=(), Vector=(X=0,Y=0,Z=0)) {template="Rotate Vector::Execute(in Transform,in Vector,out Result)"}
  ]}
 RigVMFunctionLibrary {nodes [
   Project Vector Onto Plane = RigVMCollapseNode(ExecuteContext=(), In Vector=(X=0,Y=0,Z=0), Plane Normal=(X=0,Y=0,Z=0))
   Skirt Aim And Sec Vectors = RigVMCollapseNode(ExecuteContext=())
   Thigh Readers = RigVMCollapseNode(Driver Item=(Type=None,Name="None"), ExecuteContext=())
  ]}
 RigVMFunction_Sequence_1_ContainedGraph {nodes [
   n1 = RigVMFunctionEntryNode()
   RigVMFunction_Sequence = &RigVMFunction_Sequence(ExecuteContext=@n1.ExecuteContext)
   n2 = &RigVMFunction_Sequence(ExecuteContext=@RigVMFunction_Sequence.B)
   n3 = &RigVMFunction_Sequence(ExecuteContext=@n2.B)
   n4 = &RigVMFunction_Sequence(ExecuteContext=@n3.B)
   n5 = &RigVMFunction_Sequence(ExecuteContext=@n4.B)
   n6 = &RigVMFunction_Sequence(ExecuteContext=@n5.B)
   n7 = &RigVMFunction_Sequence(ExecuteContext=@n6.B)
   n8 = &RigVMFunction_Sequence(ExecuteContext=@n7.B)
   n9 = &RigVMFunction_Sequence(ExecuteContext=@n8.B)
   n10 = &RigVMFunction_Sequence(ExecuteContext=@n9.B)
   RigVMFunctionReturnNode(A=@RigVMFunction_Sequence.A, B=@n2.A, C=@n3.A, D=@n4.A, E=@n5.A, F=@n6.A, G=@n7.A, H=@n8.A, I=@n9.A, J=@n10.A, K=@n10.B)
  ]}
 RigVMFunction_Sequence_1_ContainedGraph {nodes [
   n1 = RigVMFunctionEntryNode()
   RigVMFunction_Sequence = &RigVMFunction_Sequence(ExecuteContext=@n1.ExecuteContext)
   n2 = &RigVMFunction_Sequence(ExecuteContext=@RigVMFunction_Sequence.B)
   RigVMFunctionReturnNode(A=@RigVMFunction_Sequence.A, B=@n2.A, C=@n2.B)
  ]}
 RigVMFunction_Sequence_1_ContainedGraph {nodes [
   n1 = RigVMFunctionEntryNode()
   RigVMFunction_Sequence = &RigVMFunction_Sequence(ExecuteContext=@n1.ExecuteContext)
   n2 = &RigVMFunction_Sequence(ExecuteContext=@RigVMFunction_Sequence.B)
   n3 = &RigVMFunction_Sequence(ExecuteContext=@n2.B)
   n4 = &RigVMFunction_Sequence(ExecuteContext=@n3.B)
   n5 = &RigVMFunction_Sequence(ExecuteContext=@n4.B)
   RigVMFunctionReturnNode(A=@RigVMFunction_Sequence.A, B=@n2.A, C=@n3.A, D=@n4.A, E=@n5.A, F=@n5.B)
  ]}
 RigVMFunction_Sequence_1_ContainedGraph {nodes [
   n1 = RigVMFunctionEntryNode()
   RigVMFunction_Sequence = &RigVMFunction_Sequence(ExecuteContext=@n1.ExecuteContext)
   n2 = &RigVMFunction_Sequence(ExecuteContext=@RigVMFunction_Sequence.B)
   n3 = &RigVMFunction_Sequence(ExecuteContext=@n2.B)
   n4 = &RigVMFunction_Sequence(ExecuteContext=@n3.B)
   RigVMFunctionReturnNode(A=@RigVMFunction_Sequence.A, B=@n2.A, C=@n3.A, D=@n4.A, E=@n4.B)
  ]}
 RigVMFunction_Sequence_2_ContainedGraph {nodes [
   n1 = RigVMFunctionEntryNode()
   n2 = &RigVMFunction_Sequence(ExecuteContext=@n1.ExecuteContext)
   n3 = &RigVMFunction_Sequence(ExecuteContext=@n2.B)
   RigVMFunctionReturnNode(A=@n2.A, B=@n3.A, C=@n3.B)
  ]}
 RigVMModel {nodes [
   RigUnit_BeginExecution = &RigUnit_BeginExecution()
   RigVMFunction_Sequence = RigVMAggregateNode(ExecuteContext=@RigUnit_BeginExecution.ExecutePin)
   n1 = &RigUnit_SphericalPoseReader(
    ActiveRegionScaleFactors=(PositiveWidth=1,NegativeWidth=1,PositiveHeight=1,NegativeHeight=1)
    ActiveRegionSize=0.001
    Debug=(bDrawDebug=false,bDraw2D=False,bDrawLocalAxes=False,DebugScale=10,DebugSegments=20,DebugThickness=0.01)
    DriverAxis=(X=-1,Y=0,Z=0)
    DriverItem=(Type=Bone,Name="thigh_l")
    FalloffRegionScaleFactors=(PositiveWidth=1,NegativeWidth=1,PositiveHeight=1,NegativeHeight=1)
    FalloffSize=0.5
    FlipHeightScaling=False
    FlipWidthScaling=False
    OptionalParentItem=(Type=Bone,Name="None")
    RotationOffset=(X=0,Y=-180,Z=0)
    ExecutePin=@RigVMFunction_Sequence.D
   )
   > &RigUnit_ModifyTransforms(
    ItemToModify=((Item=(Type=Bone,Name="thigh_l_twist_up"),Transform=(Rotation=(X=0,Y=-0.7071,Z=0,W=0.7071),Translation=(X=0,Y=0,Z=0),Scale3D=(X=1,Y=1,Z=1))))
    WeightMinimum=0
    Weight=@n1.OutputParam
   )
   > n2 = &RigUnit_SphericalPoseReader(
    ActiveRegionScaleFactors=(PositiveWidth=1,NegativeWidth=1,PositiveHeight=1,NegativeHeight=1)
    ActiveRegionSize=0.001
    Debug=(bDrawDebug=false,bDraw2D=False,bDrawLocalAxes=False,DebugScale=10,DebugSegments=20,DebugThickness=0.01)
    DriverAxis=(X=-1,Y=0,Z=0)
    DriverItem=(Type=Bone,Name="thigh_l")
    FalloffRegionScaleFactors=(PositiveWidth=0.5,NegativeWidth=0.5,PositiveHeight=1,NegativeHeight=1)
    FalloffSize=0.5
    FlipHeightScaling=False
    FlipWidthScaling=False
    OptionalParentItem=(Type=Bone,Name="None")
    RotationOffset=(X=0,Y=-135,Z=0)
   )
   > &RigUnit_ModifyTransforms(
    ItemToModify=((Item=(Type=Bone,Name="thigh_l_twist_up"),Transform=(Rotation=(X=0,Y=-0.7071,Z=0,W=0.7071),Translation=(X=0,Y=0,Z=0),Scale3D=(X=1,Y=1,Z=1))))
    WeightMinimum=0
    Weight=@n2.OutputParam
   )
   > n3 = &RigUnit_SphericalPoseReader(
    ActiveRegionScaleFactors=(PositiveWidth=1,NegativeWidth=1,PositiveHeight=1,NegativeHeight=1)
    ActiveRegionSize=0.001
    Debug=(bDrawDebug=false,bDraw2D=False,bDrawLocalAxes=False,DebugScale=10,DebugSegments=20,DebugThickness=0.01)
    DriverAxis=(X=-1,Y=0,Z=0)
    DriverItem=(Type=Bone,Name="thigh_l")
    FalloffRegionScaleFactors=(PositiveWidth=1,NegativeWidth=1,PositiveHeight=1,NegativeHeight=1)
    FalloffSize=0.5
    FlipHeightScaling=False
    FlipWidthScaling=False
    OptionalParentItem=(Type=Bone,Name="None")
    RotationOffset=(X=0,Y=0,Z=0)
   )
   > &RigUnit_ModifyTransforms(
    ItemToModify=((Item=(Type=Bone,Name="thigh_l_twist_up"),Transform=(Rotation=(X=0,Y=0.7071,Z=0,W=0.7071),Translation=(X=0,Y=0,Z=0),Scale3D=(X=1,Y=1,Z=1))))
    WeightMinimum=0
    Weight=@n3.OutputParam
   )
   > n4 = &RigUnit_SphericalPoseReader(
    ActiveRegionScaleFactors=(PositiveWidth=1,NegativeWidth=1,PositiveHeight=1,NegativeHeight=1)
    ActiveRegionSize=0.001
    Debug=(bDrawDebug=false,bDraw2D=False,bDrawLocalAxes=False,DebugScale=10,DebugSegments=20,DebugThickness=0.01)
    DriverAxis=(X=-1,Y=0,Z=0)
    DriverItem=(Type=Bone,Name="thigh_l")
    FalloffRegionScaleFactors=(PositiveWidth=0.5,NegativeWidth=0.5,PositiveHeight=1,NegativeHeight=1)
    FalloffSize=0.5
    FlipHeightScaling=False
    FlipWidthScaling=False
    OptionalParentItem=(Type=Bone,Name="None")
    RotationOffset=(X=0,Y=-45,Z=0)
   )
   > &RigUnit_ModifyTransforms(
    ItemToModify=((Item=(Type=Bone,Name="thigh_l_twist_up"),Transform=(Rotation=(X=0,Y=0.7071,Z=0,W=0.7071),Translation=(X=0,Y=0,Z=0),Scale3D=(X=1,Y=1,Z=1))))
    WeightMinimum=0
    Weight=@n4.OutputParam
   )
   > RigVMVariableNode(Value=&RigVMFunction_MathTransformRotateVector(Vector=(X=0,Y=0,Z=-1), Transform=&RigUnit_GetTransform(Item=(Type=Bone,Name="thigh_l_twist_up"), bInitial=False).Transform) {template="Rotate Vector::Execute(in Transform,in Vector,out Result)"}.Result) {variable=Skirt L Sec Vector}
   Thigh Readers = RigVMFunctionReferenceNode(
    Ba Viz=false
    Crouch Ot Post Viz=false
    Crouch Ot Pre Viz=false
    Crouch Ot Viz=false
    Crouch Viz=false
    Driver Item=(Type=Bone,Name="thigh_l")
    Fr Post Viz=false
    Fr Viz=false
    In Pre Viz=false
    In Viz=false
    Ot Viz=false
    ExecuteContext=@RigVMFunction_Sequence.C
   )
   > RigVMVariableNode(Value=@"Thigh Readers".Fr) {variable=L Thigh Fr}
   > RigVMVariableNode(Value=@"Thigh Readers".Fr Pre) {variable=L Thigh Fr Pre}
   > RigVMVariableNode(Value=@"Thigh Readers".Fr Post) {variable=L Thigh Fr Post}
   > RigVMVariableNode(Value=@"Thigh Readers".Ba) {variable=L Thigh Ba}
   > RigVMVariableNode(Value=@"Thigh Readers".In) {variable=L Thigh In}
   > RigVMVariableNode(Value=@"Thigh Readers".Ot) {variable=L Thigh Ot}
   > RigVMVariableNode(Value=@"Thigh Readers".Crouch Ot) {variable=L Thigh Crouch Ot}
   > RigVMVariableNode(Value=@"Thigh Readers".Crouch Pre Ot) {variable=L Thigh Crouch Pre Ot}
   > RigVMVariableNode(Value=@"Thigh Readers".Crouch Post Ot) {variable=L Thigh Crouch Ot Post}
   > Thigh Readers_1 = RigVMFunctionReferenceNode(
    Ba Viz=false
    Crouch Ot Post Viz=false
    Crouch Ot Pre Viz=false
    Crouch Ot Viz=false
    Crouch Pre Viz=false
    Crouch Viz=false
    Driver Item=(Type=Bone,Name="thigh_r")
    Fr Post Viz=false
    Fr Pre Viz=false
    Fr Viz=false
    Ot Viz=false
   )
   > RigVMVariableNode(Value=@"Thigh Readers_1".Fr) {variable=R Thigh Fr}
   > RigVMVariableNode(Value=@"Thigh Readers_1".Fr Pre) {variable=R Thigh Fr Pre}
   > RigVMVariableNode(Value=@"Thigh Readers_1".Fr Post) {variable=R Thigh Fr Post}
   > RigVMVariableNode(Value=@"Thigh Readers_1".Ba) {variable=R Thigh Ba}
   > RigVMVariableNode(Value=@"Thigh Readers_1".In) {variable=R Thigh In}
   > RigVMVariableNode(Value=@"Thigh Readers_1".Ot) {variable=R Thigh Ot}
   > RigVMVariableNode(Value=@"Thigh Readers_1".Crouch) {variable=R Thigh Crouch}
   > RigVMVariableNode(Value=@"Thigh Readers_1".Crouch Ot) {variable=R Thigh Crouch Ot}
   > RigVMVariableNode(Value=@"Thigh Readers_1".Crouch Pre Ot) {variable=R Thigh Crouch Pre Ot}
   > RigVMVariableNode(Value=@"Thigh Readers_1".Crouch Post Ot) {variable=R Thigh Crouch Ot Post}
   FN_DeformRig_FNC_1 = RigVMFunctionReferenceNode(
    DeformRig_ Curve="(deformRig_l_upperarm_up_alpha_72_#1=False,deformRig_r_upperarm_up_alpha_73_#2=False,deformRig_l_upperarm_fr_alpha_74_#3=False,deformRig_r_upperarm_fr_alpha_75_CB51B8 ...(truncated)"
    DeformRig_Bool="(upperarm_1_#4=True,elbow_3_#5=True,wrist_7_#6=True,groin_8_#7=True,glute_17_#8=True,knee_9_40C408324CA261 ...(truncated)"
    DeformRig_Debug=(upperarm_debug_18_#4=False,wrist_debug_20_#6=False,lat_debug_26_#9=False,groinGlute_debug_27_#7=False)
    DeformRig_Value="(upperarm_up_18_#10=(Rotation=(X=0,Y=0,Z=0,W=1),Translation=(X=0,Y=0,Z=8),Scale3D=(X=0.7,Y=1,Z=1)),upperarm_fwd_fr_19_#11=(Rotation=(X=0,Y=0,Z=0,W=1),Translation=(X=0,Y=10,Z=0),Scale3D=(X=0.5,Y=1,Z= ...(truncated)"
    ExecuteContext=@RigVMFunction_Sequence.B
   )
   RigVMFunction_ControlFlowBranch = &RigVMFunction_ControlFlowBranch(Condition=RigVMVariableNode() {variable=Debug Controls On}.Value, ExecuteContext=@RigVMFunction_Sequence.A)
   For_Each_2 = RigVMDispatchNode(ExecuteContext=@RigVMFunction_ControlFlowBranch.True, Array=RigVMVariableNode() {variable=Debug Ctrls}.Value) {template="DISPATCH_RigVMDispatch_ArrayIterator(in Array,out Element,out Index,out Count,out Ratio)"}
   > &RigUnit_SetTransform(
    Weight=1
    bInitial=False
    Item=RigVMDispatchNode(Array=RigVMVariableNode() {variable=Debug Bones}.Value, Index=@For_Each_2.Index) {template="DISPATCH_RigVMDispatch_ArrayGetAtIndex(in Array,in Index,out Element)"}.Element
    Value=&RigUnit_GetTransform(bInitial=False, Item=@For_Each_2.Element).Transform
   )
   &RigUnit_ModifyTransforms(
    ItemToModify="((Item=(Type=Bone,Name=\"skirt_fr_l\"),Transform=(Rotation=(X=0.0726,Y=0.0697,Z=-0.6545,W=0.7494),Translation=(X=0,Y=0,Z=0),Scale3D=(X=1,Y=1,Z=1))),(Item=(Type=Bone,Name=\"skirt_fr_c\"),Transform=(Rotation=(X=0,Y=0,Z=-0.5556,W=0.8315),Translation=(X=0,Y=0,Z=0) ...(truncated)"
    WeightMinimum=0
    Weight=RigVMVariableNode() {variable=L Thigh Fr}.Value
    ExecutePin=@RigVMFunction_Sequence.F
   )
   > &RigUnit_ModifyTransforms(
    ItemToModify=((Item=(Type=Bone,Name="skirt_fr_l"),Transform=(Rotation=(X=0,Y=0,Z=0.1125,W=0.9936),Translation=(X=0,Y=0,Z=0),Scale3D=(X=1,Y=1,Z=1))))
    WeightMinimum=0
    Weight=RigVMVariableNode() {variable=L Thigh Fr Pre}.Value
   )
   > &RigUnit_ModifyTransforms(
    ItemToModify="((Item=(Type=Bone,Name=\"skirt_fr_l\"),Transform=(Rotation=(X=0.1088,Y=0.1253,Z=-0.7447,W=0.6464),Translation=(X=3.4878,Y=4.6857,Z=-0.0001),Scale3D=(X=1,Y=1,Z=1))),(Item=(Type=Bone,Name=\"skirt_fr_c\"),Transform=(Rotation=(X=0,Y=0,Z=-0.5556,W=0.8315),Translati ...(truncated)"
    WeightMinimum=0
    Weight=RigVMVariableNode() {variable=L Thigh Fr Post}.Value
   )
   > &RigUnit_ModifyTransforms(
    ItemToModify="((Item=(Type=Bone,Name=\"skirt_bck_l\"),Transform=(Rotation=(X=0,Y=0,Z=0.6344,W=0.773),Translation=(X=0,Y=0,Z=0),Scale3D=(X=1,Y=1,Z=1))),(Item=(Type=Bone,Name=\"skirt_bck_c\"),Transform=(Rotation=(X=0,Y=0,Z=0.3827,W=0.9239),Translation=(X=0,Y=0,Z=0),Scale3D=(X ...(truncated)"
    WeightMinimum=0
    Weight=RigVMVariableNode() {variable=L Thigh Ba}.Value
   )
   > &RigUnit_ModifyTransforms(
    ItemToModify="((Item=(Type=Bone,Name=\"skirt_side_l\"),Transform=(Rotation=(X=0,Y=-0.5593,Z=0,W=0.829),Translation=(X=0,Y=0,Z=0),Scale3D=(X=1,Y=1,Z=1))),(Item=(Type=Bone,Name=\"skirt_fr_l\"),Transform=(Rotation=(X=0.0202,Y=0.454,Z=-0.0063,W=-0.8907),Translation=(X=0,Y=0,Z=0 ...(truncated)"
    WeightMinimum=0
    Weight=RigVMVariableNode() {variable=L Thigh Ot}.Value
   )
   > &RigUnit_ModifyTransforms(
    ItemToModify="((Item=(Type=Bone,Name=\"skirt_fr_l\"),Transform=(Rotation=(X=-0.0239,Y=-0.1466,Z=-0.1588,W=0.9761),Translation=(X=0,Y=0,Z=0),Scale3D=(X=1,Y=1,Z=1))),(Item=(Type=Bone,Name=\"skirt_side_l\"),Transform=(Rotation=(X=0.009,Y=-0.0823,Z=-0.0442,W=0.9956),Translation ...(truncated)"
    WeightMinimum=0
    Weight=RigVMVariableNode() {variable=L Thigh Crouch Ot}.Value
   )
   > &RigUnit_ModifyTransforms(
    ItemToModify="((Item=(Type=Bone,Name=\"skirt_side_l\"),Transform=(Rotation=(X=0.2746,Y=0.1925,Z=0.7348,W=-0.5897),Translation=(X=0,Y=0,Z=0),Scale3D=(X=1,Y=1,Z=1))),(Item=(Type=Bone,Name=\"dyn_skirt_side_l_02\"),Transform=(Rotation=(X=0,Y=0,Z=0,W=1),Translation=(X=-2,Y=0,Z=0 ...(truncated)"
    WeightMinimum=0
    Weight=RigVMVariableNode() {variable=L Thigh Crouch Ot Post}.Value
   )
   > &RigUnit_ModifyTransforms(
    ItemToModify=((Item=(Type=Bone,Name="skirt_fr_r"),Transform=(Rotation=(X=0,Y=0,Z=0.0407,W=-0.9992),Translation=(X=0,Y=0,Z=0),Scale3D=(X=1,Y=1,Z=1))))
    WeightMinimum=0
    Weight=RigVMVariableNode() {variable=R Thigh Crouch}.Value
   )
   > &RigUnit_ModifyTransforms(
    ItemToModify="((Item=(Type=Bone,Name=\"skirt_fr_r\"),Transform=(Rotation=(X=-0.3522,Y=-0.5928,Z=0.2063,W=-0.6943),Translation=(X=0,Y=0,Z=0),Scale3D=(X=1,Y=1,Z=1))),(Item=(Type=Bone,Name=\"skirt_fr_c\"),Transform=(Rotation=(X=-0.4173,Y=-0.3369,Z=-0.2273,W=0.8128),Translation ...(truncated)"
    WeightMinimum=0
    Weight=RigVMVariableNode() {variable=R Thigh In}.Value
   )
   > &RigUnit_ModifyTransforms(
    ItemToModify="((Item=(Type=Bone,Name=\"skirt_side_l\"),Transform=(Rotation=(X=0,Y=-0.1233,Z=0,W=0.9924),Translation=(X=0,Y=0,Z=0),Scale3D=(X=1,Y=1,Z=1))),(Item=(Type=Bone,Name=\"skirt_fr_l\"),Transform=(Rotation=(X=0.0079,Y=0.1837,Z=0.0847,W=-0.9793),Translation=(X=0,Y=0,Z= ...(truncated)"
    WeightMinimum=0
    Weight=/Script/RigVM.RigVMFunction_MathDoubleMin(B=RigVMVariableNode() {variable=L Thigh Ot}.Value, A=RigVMVariableNode() {variable=L Thigh Fr}.Value) {is=struct RigVMUnitNode, template="Minimum::Execute(in A,in B,out Result)"}.Result
   )
   &RigUnit_ParentConstraint(
    AdvancedSettings=(InterpolationType=Average,RotationOrderForFilter=XZY)
    Child=(Type=Bone,Name="chest_wing_a_l")
    Parents=((Item=(Type=Bone,Name="spine_04"),Weight=0.213),(Item=(Type=Bone,Name="spine_05"),Weight=0.2373),(Item=(Type=Bone,Name="clavicle_l"),Weight=0.3137),(Item=(Type=Bone,Name="deform_pec_l"),Weight=0.2361))
    Weight=1
    ExecutePin=@RigVMFunction_Sequence.I
   )
   > &RigUnit_ParentConstraint(
    AdvancedSettings=(InterpolationType=Average,RotationOrderForFilter=XZY)
    Child=(Type=Bone,Name="chest_wing_b_l")
    Parents="((Item=(Type=Bone,Name=\"spine_04\"),Weight=0.3229),(Item=(Type=Bone,Name=\"spine_05\"),Weight=0.1427),(Item=(Type=Bone,Name=\"clavicle_l\"),Weight=0.2553),(Item=(Type=Bone,Name=\"deform_pec_l\"),Weight=0.2753),(Item=(Type=Bone,Name=\"spine_02\"),Weight=0.0019),(Ite ...(truncated)"
    Weight=1
   )
   > &RigUnit_ParentConstraint(
    AdvancedSettings=(InterpolationType=Average,RotationOrderForFilter=XZY)
    Child=(Type=Bone,Name="chest_wing_c_l")
    Parents="((Item=(Type=Bone,Name=\"spine_04\"),Weight=0.4708),(Item=(Type=Bone,Name=\"spine_05\"),Weight=0.1246),(Item=(Type=Bone,Name=\"clavicle_l\"),Weight=0.2135),(Item=(Type=Bone,Name=\"deform_pec_l\"),Weight=0.1444),(Item=(Type=Bone,Name=\"spine_02\"),Weight=0.0242),(Ite ...(truncated)"
    Weight=1
   )
   > &RigUnit_ParentConstraint(
    AdvancedSettings=(InterpolationType=Average,RotationOrderForFilter=XZY)
    Child=(Type=Bone,Name="chest_wing_a_r")
    Parents=((Item=(Type=Bone,Name="spine_04"),Weight=0.213),(Item=(Type=Bone,Name="spine_05"),Weight=0.2373),(Item=(Type=Bone,Name="clavicle_r"),Weight=0.3137),(Item=(Type=Bone,Name="deform_pec_r"),Weight=0.2361))
    Weight=1
   )
   > &RigUnit_ParentConstraint(
    AdvancedSettings=(InterpolationType=Average,RotationOrderForFilter=XZY)
    Child=(Type=Bone,Name="chest_wing_b_r")
    Parents="((Item=(Type=Bone,Name=\"spine_04\"),Weight=0.3229),(Item=(Type=Bone,Name=\"spine_05\"),Weight=0.1427),(Item=(Type=Bone,Name=\"clavicle_r\"),Weight=0.2553),(Item=(Type=Bone,Name=\"deform_pec_r\"),Weight=0.2753),(Item=(Type=Bone,Name=\"spine_02\"),Weight=0.0019),(Ite ...(truncated)"
    Weight=1
   )
   > &RigUnit_ParentConstraint(
    AdvancedSettings=(InterpolationType=Average,RotationOrderForFilter=XZY)
    Child=(Type=Bone,Name="chest_wing_c_r")
    Parents="((Item=(Type=Bone,Name=\"spine_04\"),Weight=0.4708),(Item=(Type=Bone,Name=\"spine_05\"),Weight=0.1246),(Item=(Type=Bone,Name=\"clavicle_r\"),Weight=0.2135),(Item=(Type=Bone,Name=\"deform_pec_r\"),Weight=0.1444),(Item=(Type=Bone,Name=\"spine_02\"),Weight=0.0242),(Ite ...(truncated)"
    Weight=1
   )
   > &RigUnit_ParentConstraint(
    AdvancedSettings=(InterpolationType=Average,RotationOrderForFilter=XZY)
    Child=(Type=Bone,Name="necklace")
    Parents=((Item=(Type=Bone,Name="spine_05"),Weight=0.17),(Item=(Type=Bone,Name="neck_01"),Weight=0.35),(Item=(Type=Bone,Name="neck_02"),Weight=0.32),(Item=(Type=Bone,Name="head"),Weight=0.16))
    Weight=1
   )
   > n5 = &RigUnit_SphericalPoseReader(
    ActiveRegionScaleFactors=(PositiveWidth=1,NegativeWidth=1,PositiveHeight=1,NegativeHeight=1)
    ActiveRegionSize=0.001
    Debug=(bDrawDebug=false,bDraw2D=False,bDrawLocalAxes=False,DebugScale=10,DebugSegments=20,DebugThickness=0.1)
    DriverAxis=(X=-1,Y=0,Z=0)
    DriverItem=(Type=Bone,Name="clavicle_l")
    FalloffRegionScaleFactors=(PositiveWidth=1,NegativeWidth=1,PositiveHeight=1,NegativeHeight=1)
    FalloffSize=0.5
    FlipHeightScaling=False
    FlipWidthScaling=False
    OptionalParentItem=(Type=Bone,Name="None")
    RotationOffset=(X=-90,Y=0,Z=0)
   )
   > &RigUnit_ModifyTransforms(
    ItemToModify="((Item=(Type=Bone,Name=\"chest_wing_a_l\"),Transform=(Rotation=(X=-0.1123,Y=-0.0538,Z=0.5148,W=0.8482),Translation=(X=-2.5432,Y=2.6424,Z=0),Scale3D=(X=1,Y=1,Z=1))),(Item=(Type=Bone,Name=\"chest_wing_b_l\"),Transform=(Rotation=(X=-0.1532,Y=-0.0829,Z=0.7717,W=0. ...(truncated)"
    WeightMinimum=0
    Weight=@n5.OutputParam
   )
   > n6 = &RigUnit_SphericalPoseReader(
    ActiveRegionScaleFactors=(PositiveWidth=1,NegativeWidth=1,PositiveHeight=1,NegativeHeight=1)
    ActiveRegionSize=0.001
    Debug=(bDrawDebug=false,bDraw2D=False,bDrawLocalAxes=False,DebugScale=10,DebugSegments=20,DebugThickness=0.1)
    DriverAxis=(X=-1,Y=0,Z=0)
    DriverItem=(Type=Bone,Name="clavicle_r")
    FalloffRegionScaleFactors=(PositiveWidth=1,NegativeWidth=1,PositiveHeight=1,NegativeHeight=1)
    FalloffSize=0.5
    FlipHeightScaling=False
    FlipWidthScaling=False
    OptionalParentItem=(Type=Bone,Name="None")
    RotationOffset=(X=-90,Y=0,Z=0)
   )
   > &RigUnit_ModifyTransforms(
    ItemToModify="((Item=(Type=Bone,Name=\"chest_wing_a_r\"),Transform=(Rotation=(X=-0.1123,Y=-0.0538,Z=0.5148,W=0.8482),Translation=(X=2.5432,Y=-2.6424,Z=0),Scale3D=(X=1,Y=1,Z=1))),(Item=(Type=Bone,Name=\"chest_wing_b_r\"),Transform=(Rotation=(X=-0.1532,Y=-0.0829,Z=0.7717,W=0. ...(truncated)"
    WeightMinimum=0
    Weight=@n6.OutputParam
   )
   > n7 = &RigUnit_SphericalPoseReader(
    ActiveRegionScaleFactors=(PositiveWidth=1,NegativeWidth=1,PositiveHeight=1,NegativeHeight=1)
    ActiveRegionSize=0.001
    Debug=(bDrawDebug=false,bDraw2D=False,bDrawLocalAxes=False,DebugScale=10,DebugSegments=20,DebugThickness=0.1)
    DriverAxis=(X=-1,Y=0,Z=0)
    DriverItem=(Type=Bone,Name="clavicle_l")
    FalloffRegionScaleFactors=(PositiveWidth=1,NegativeWidth=1,PositiveHeight=1,NegativeHeight=1)
    FalloffSize=0.5
    FlipHeightScaling=False
    FlipWidthScaling=False
    OptionalParentItem=(Type=Bone,Name="None")
    RotationOffset=(X=180,Y=0,Z=0)
   )
   > &RigUnit_ModifyTransforms(
    ItemToModify="((Item=(Type=Bone,Name=\"chest_wing_a_l\"),Transform=(Rotation=(X=0,Y=0,Z=0,W=-1),Translation=(X=0,Y=1.0928,Z=0),Scale3D=(X=1,Y=1,Z=1))),(Item=(Type=Bone,Name=\"chest_wing_b_l\"),Transform=(Rotation=(X=0,Y=0,Z=0,W=-1),Translation=(X=0,Y=2.5012,Z=0),Scale3D=(X= ...(truncated)"
    WeightMinimum=0
    Weight=@n7.OutputParam
   )
   > n8 = &RigUnit_SphericalPoseReader(
    ActiveRegionScaleFactors=(PositiveWidth=1,NegativeWidth=1,PositiveHeight=1,NegativeHeight=1)
    ActiveRegionSize=0.001
    Debug=(bDrawDebug=false,bDraw2D=False,bDrawLocalAxes=False,DebugScale=10,DebugSegments=20,DebugThickness=0.1)
    DriverAxis=(X=-1,Y=0,Z=0)
    DriverItem=(Type=Bone,Name="clavicle_r")
    FalloffRegionScaleFactors=(PositiveWidth=1,NegativeWidth=1,PositiveHeight=1,NegativeHeight=1)
    FalloffSize=0.5
    FlipHeightScaling=False
    FlipWidthScaling=False
    OptionalParentItem=(Type=Bone,Name="None")
    RotationOffset=(X=180,Y=0,Z=0)
   )
   > &RigUnit_ModifyTransforms(
    ItemToModify="((Item=(Type=Bone,Name=\"chest_wing_a_r\"),Transform=(Rotation=(X=0,Y=0,Z=0,W=-1),Translation=(X=0,Y=-0.4972,Z=0),Scale3D=(X=1,Y=1,Z=1))),(Item=(Type=Bone,Name=\"chest_wing_b_r\"),Transform=(Rotation=(X=0,Y=0,Z=0.1045,W=0.9945),Translation=(X=0,Y=-1.1843,Z=0), ...(truncated)"
    WeightMinimum=0
    Weight=@n8.OutputParam
   )
   RigUnit_GetTransform = &RigUnit_GetTransform(Item=(Type=Bone,Name="hand_l"), Space=LocalSpace, bInitial=False)
   > RigUnit_SetTransform = &RigUnit_SetTransform(Item=(Type=Bone,Name="sleeve_l"), Space=LocalSpace, Weight=1, bInitial=False, Value=@RigUnit_GetTransform.Transform)
   > &RigUnit_SetTransform(
    Item=(Type=Bone,Name="sleeve_r")
    Space=LocalSpace
    Weight=1
    bInitial=False
    Value=&RigUnit_GetTransform(Item=(Type=Bone,Name="hand_r"), Space=LocalSpace, bInitial=False).Transform
   )
   > n9 = &RigUnit_SphericalPoseReader(
    ActiveRegionScaleFactors=(PositiveWidth=1,NegativeWidth=1,PositiveHeight=1,NegativeHeight=1)
    ActiveRegionSize=0.001
    Debug=(bDrawDebug=false,bDraw2D=False,bDrawLocalAxes=False,DebugScale=10,DebugSegments=20,DebugThickness=0.1)
    DriverAxis=(X=-1,Y=0,Z=0)
    DriverItem=(Type=Bone,Name="hand_l")
    FalloffRegionScaleFactors=(PositiveWidth=1,NegativeWidth=1,PositiveHeight=1,NegativeHeight=1)
    FalloffSize=0.5
    FlipHeightScaling=False
    FlipWidthScaling=False
    OptionalParentItem=(Type=Bone,Name="None")
    RotationOffset=(X=90,Y=0,Z=0)
   )
   > &RigUnit_ModifyTransforms(
    ItemToModify=((Item=(Type=Bone,Name="sleeve_l"),Transform=(Rotation=(X=0,Y=0,Z=0.3827,W=0.9239),Translation=(X=0,Y=0,Z=0),Scale3D=(X=1,Y=1,Z=1))))
    WeightMinimum=0
    Weight=@n9.OutputParam
   )
   > n10 = &RigUnit_SphericalPoseReader(
    ActiveRegionScaleFactors=(PositiveWidth=1,NegativeWidth=1,PositiveHeight=1,NegativeHeight=1)
    ActiveRegionSize=0.001
    Debug=(bDrawDebug=false,bDraw2D=False,bDrawLocalAxes=False,DebugScale=10,DebugSegments=20,DebugThickness=0.1)
    DriverAxis=(X=-1,Y=0,Z=0)
    DriverItem=(Type=Bone,Name="hand_r")
    FalloffRegionScaleFactors=(PositiveWidth=1,NegativeWidth=1,PositiveHeight=1,NegativeHeight=1)
    FalloffSize=0.5
    FlipHeightScaling=False
    FlipWidthScaling=False
    OptionalParentItem=(Type=Bone,Name="None")
    RotationOffset=(X=90,Y=0,Z=0)
   )
   > &RigUnit_ModifyTransforms(
    ItemToModify=((Item=(Type=Bone,Name="sleeve_r"),Transform=(Rotation=(X=0,Y=0,Z=0.3827,W=0.9239),Translation=(X=0,Y=0,Z=0),Scale3D=(X=1,Y=1,Z=1))))
    WeightMinimum=0
    Weight=@n10.OutputParam
   )
   > n11 = &RigUnit_SphericalPoseReader(
    ActiveRegionScaleFactors=(PositiveWidth=1,NegativeWidth=1,PositiveHeight=1,NegativeHeight=1)
    ActiveRegionSize=0.001
    Debug=(bDrawDebug=false,bDraw2D=False,bDrawLocalAxes=False,DebugScale=10,DebugSegments=20,DebugThickness=0.1)
    DriverAxis=(X=-1,Y=0,Z=0)
    DriverItem=(Type=Bone,Name="hand_l")
    FalloffRegionScaleFactors=(PositiveWidth=1,NegativeWidth=1,PositiveHeight=1,NegativeHeight=1)
    FalloffSize=0.5
    FlipHeightScaling=False
    FlipWidthScaling=False
    OptionalParentItem=(Type=Bone,Name="None")
    RotationOffset=(X=0,Y=0,Z=0)
   )
   > &RigUnit_ModifyTransforms(
    ItemToModify=((Item=(Type=Bone,Name="sleeve_l"),Transform=(Rotation=(X=0,Y=-0.3827,Z=0,W=0.9239),Translation=(X=0,Y=0,Z=0),Scale3D=(X=1,Y=1,Z=1))))
    WeightMinimum=0
    Weight=@n11.OutputParam
   )
   > n12 = &RigUnit_SphericalPoseReader(
    ActiveRegionScaleFactors=(PositiveWidth=1,NegativeWidth=1,PositiveHeight=1,NegativeHeight=1)
    ActiveRegionSize=0.001
    Debug=(bDrawDebug=false,bDraw2D=False,bDrawLocalAxes=False,DebugScale=10,DebugSegments=20,DebugThickness=0.1)
    DriverAxis=(X=-1,Y=0,Z=0)
    DriverItem=(Type=Bone,Name="hand_l")
    FalloffRegionScaleFactors=(PositiveWidth=1,NegativeWidth=1,PositiveHeight=1,NegativeHeight=1)
    FalloffSize=0.5
    FlipHeightScaling=False
    FlipWidthScaling=False
    OptionalParentItem=(Type=Bone,Name="None")
    RotationOffset=(X=180,Y=0,Z=0)
   )
   > &RigUnit_ModifyTransforms(
    ItemToModify=((Item=(Type=Bone,Name="sleeve_l"),Transform=(Rotation=(X=0,Y=0.3827,Z=0,W=0.9239),Translation=(X=0,Y=0,Z=0),Scale3D=(X=1,Y=1,Z=1))))
    WeightMinimum=0
    Weight=@n12.OutputParam
   )
   > n13 = &RigUnit_SphericalPoseReader(
    ActiveRegionScaleFactors=(PositiveWidth=1,NegativeWidth=1,PositiveHeight=1,NegativeHeight=1)
    ActiveRegionSize=0.001
    Debug=(bDrawDebug=false,bDraw2D=False,bDrawLocalAxes=False,DebugScale=10,DebugSegments=20,DebugThickness=0.1)
    DriverAxis=(X=-1,Y=0,Z=0)
    DriverItem=(Type=Bone,Name="hand_r")
    FalloffRegionScaleFactors=(PositiveWidth=1,NegativeWidth=1,PositiveHeight=1,NegativeHeight=1)
    FalloffSize=0.5
    FlipHeightScaling=False
    FlipWidthScaling=False
    OptionalParentItem=(Type=Bone,Name="None")
    RotationOffset=(X=0,Y=0,Z=0)
   )
   > &RigUnit_ModifyTransforms(
    ItemToModify=((Item=(Type=Bone,Name="sleeve_r"),Transform=(Rotation=(X=0,Y=-0.3827,Z=0,W=0.9239),Translation=(X=0,Y=0,Z=0),Scale3D=(X=1,Y=1,Z=1))))
    WeightMinimum=0
    Weight=@n13.OutputParam
   )
   > n14 = &RigUnit_SphericalPoseReader(
    ActiveRegionScaleFactors=(PositiveWidth=1,NegativeWidth=1,PositiveHeight=1,NegativeHeight=1)
    ActiveRegionSize=0.001
    Debug=(bDrawDebug=false,bDraw2D=False,bDrawLocalAxes=False,DebugScale=10,DebugSegments=20,DebugThickness=0.1)
    DriverAxis=(X=-1,Y=0,Z=0)
    DriverItem=(Type=Bone,Name="hand_r")
    FalloffRegionScaleFactors=(PositiveWidth=1,NegativeWidth=1,PositiveHeight=1,NegativeHeight=1)
    FalloffSize=0.5
    FlipHeightScaling=False
    FlipWidthScaling=False
    OptionalParentItem=(Type=Bone,Name="None")
    RotationOffset=(X=180,Y=0,Z=0)
   )
   > &RigUnit_ModifyTransforms(
    ItemToModify=((Item=(Type=Bone,Name="sleeve_r"),Transform=(Rotation=(X=0,Y=0.3827,Z=0,W=0.9239),Translation=(X=0,Y=0,Z=0),Scale3D=(X=1,Y=1,Z=1))))
    WeightMinimum=0
    Weight=@n14.OutputParam
   )
   n15 = &RigUnit_SphericalPoseReader(
    ActiveRegionScaleFactors=(PositiveWidth=1,NegativeWidth=1,PositiveHeight=1,NegativeHeight=1)
    ActiveRegionSize=0.001
    Debug=(bDrawDebug=false,bDraw2D=False,bDrawLocalAxes=False,DebugScale=10,DebugSegments=20,DebugThickness=0.01)
    DriverAxis=(X=-1,Y=0,Z=0)
    DriverItem=(Type=Bone,Name="thigh_r")
    FalloffRegionScaleFactors=(PositiveWidth=1,NegativeWidth=1,PositiveHeight=1,NegativeHeight=1)
    FalloffSize=0.5
    FlipHeightScaling=False
    FlipWidthScaling=False
    OptionalParentItem=(Type=Bone,Name="None")
    RotationOffset=(X=0,Y=-180,Z=0)
    ExecutePin=@RigVMFunction_Sequence.E
   )
   > &RigUnit_ModifyTransforms(
    ItemToModify=((Item=(Type=Bone,Name="thigh_r_twist_up"),Transform=(Rotation=(X=0,Y=-0.7071,Z=0,W=0.7071),Translation=(X=0,Y=0,Z=0),Scale3D=(X=1,Y=1,Z=1))))
    WeightMinimum=0
    Weight=@n15.OutputParam
   )
   > n16 = &RigUnit_SphericalPoseReader(
    ActiveRegionScaleFactors=(PositiveWidth=1,NegativeWidth=1,PositiveHeight=1,NegativeHeight=1)
    ActiveRegionSize=0.001
    Debug=(bDrawDebug=false,bDraw2D=False,bDrawLocalAxes=False,DebugScale=10,DebugSegments=20,DebugThickness=0.01)
    DriverAxis=(X=-1,Y=0,Z=0)
    DriverItem=(Type=Bone,Name="thigh_r")
    FalloffRegionScaleFactors=(PositiveWidth=0.5,NegativeWidth=0.5,PositiveHeight=1,NegativeHeight=1)
    FalloffSize=0.5
    FlipHeightScaling=False
    FlipWidthScaling=False
    OptionalParentItem=(Type=Bone,Name="None")
    RotationOffset=(X=0,Y=-135,Z=0)
   )
   > &RigUnit_ModifyTransforms(
    ItemToModify=((Item=(Type=Bone,Name="thigh_r_twist_up"),Transform=(Rotation=(X=0,Y=-0.7071,Z=0,W=0.7071),Translation=(X=0,Y=0,Z=0),Scale3D=(X=1,Y=1,Z=1))))
    WeightMinimum=0
    Weight=@n16.OutputParam
   )
   > n17 = &RigUnit_SphericalPoseReader(
    ActiveRegionScaleFactors=(PositiveWidth=1,NegativeWidth=1,PositiveHeight=1,NegativeHeight=1)
    ActiveRegionSize=0.001
    Debug=(bDrawDebug=false,bDraw2D=False,bDrawLocalAxes=False,DebugScale=10,DebugSegments=20,DebugThickness=0.01)
    DriverAxis=(X=-1,Y=0,Z=0)
    DriverItem=(Type=Bone,Name="thigh_r")
    FalloffRegionScaleFactors=(PositiveWidth=1,NegativeWidth=1,PositiveHeight=1,NegativeHeight=1)
    FalloffSize=0.5
    FlipHeightScaling=False
    FlipWidthScaling=False
    OptionalParentItem=(Type=Bone,Name="None")
    RotationOffset=(X=0,Y=0,Z=0)
   )
   > &RigUnit_ModifyTransforms(
    ItemToModify=((Item=(Type=Bone,Name="thigh_r_twist_up"),Transform=(Rotation=(X=0,Y=0.7071,Z=0,W=0.7071),Translation=(X=0,Y=0,Z=0),Scale3D=(X=1,Y=1,Z=1))))
    WeightMinimum=0
    Weight=@n17.OutputParam
   )
   > n18 = &RigUnit_SphericalPoseReader(
    ActiveRegionScaleFactors=(PositiveWidth=1,NegativeWidth=1,PositiveHeight=1,NegativeHeight=1)
    ActiveRegionSize=0.001
    Debug=(bDrawDebug=false,bDraw2D=False,bDrawLocalAxes=False,DebugScale=10,DebugSegments=20,DebugThickness=0.01)
    DriverAxis=(X=-1,Y=0,Z=0)
    DriverItem=(Type=Bone,Name="thigh_r")
    FalloffRegionScaleFactors=(PositiveWidth=0.5,NegativeWidth=0.5,PositiveHeight=1,NegativeHeight=1)
    FalloffSize=0.5
    FlipHeightScaling=False
    FlipWidthScaling=False
    OptionalParentItem=(Type=Bone,Name="None")
    RotationOffset=(X=0,Y=-45,Z=0)
   )
   > &RigUnit_ModifyTransforms(
    ItemToModify=((Item=(Type=Bone,Name="thigh_r_twist_up"),Transform=(Rotation=(X=0,Y=0.7071,Z=0,W=0.7071),Translation=(X=0,Y=0,Z=0),Scale3D=(X=1,Y=1,Z=1))))
    WeightMinimum=0
    Weight=@n18.OutputParam
   )
   > RigVMVariableNode(Value=&RigVMFunction_MathTransformRotateVector(Vector=(X=0,Y=0,Z=1), Transform=&RigUnit_GetTransform(Item=(Type=Bone,Name="thigh_r_twist_up"), bInitial=False).Transform) {template="Rotate Vector::Execute(in Transform,in Vector,out Result)"}.Result) {variable=Skirt R Sec Vector}
   &RigUnit_ModifyTransforms(
    ItemToModify="((Item=(Type=Bone,Name=\"skirt_fr_r\"),Transform=(Rotation=(X=-0.0207,Y=-0.0179,Z=0.6825,W=-0.7304),Translation=(X=0,Y=0,Z=0),Scale3D=(X=1,Y=1,Z=1))),(Item=(Type=Bone,Name=\"skirt_fr_c\"),Transform=(Rotation=(X=-0.0503,Y=-0.0597,Z=-0.6428,W=0.7621),Translation ...(truncated)"
    WeightMinimum=0
    Weight=RigVMVariableNode() {variable=R Thigh Fr}.Value
    ExecutePin=@RigVMFunction_Sequence.G
   )
   > &RigUnit_ModifyTransforms(
    ItemToModify=((Item=(Type=Bone,Name="skirt_fr_r"),Transform=(Rotation=(X=0,Y=0,Z=0.0523,W=0.9986),Translation=(X=0,Y=0,Z=0),Scale3D=(X=1,Y=1,Z=1))))
    WeightMinimum=0
    Weight=RigVMVariableNode() {variable=R Thigh Fr Pre}.Value
   )
   > &RigUnit_ModifyTransforms(
    ItemToModify="((Item=(Type=Bone,Name=\"skirt_fr_r\"),Transform=(Rotation=(X=-0.1157,Y=-0.119,Z=0.7071,W=-0.6874),Translation=(X=-3.4876,Y=-4.6858,Z=-0.0001),Scale3D=(X=1,Y=1,Z=1))),(Item=(Type=Bone,Name=\"skirt_fr_c\"),Transform=(Rotation=(X=-0.0613,Y=-0.1318,Z=-0.6693,W=0. ...(truncated)"
    WeightMinimum=0
    Weight=RigVMVariableNode() {variable=R Thigh Fr Post}.Value
   )
   > &RigUnit_ModifyTransforms(
    ItemToModify="((Item=(Type=Bone,Name=\"skirt_bck_r\"),Transform=(Rotation=(X=0,Y=0,Z=0.6344,W=0.773),Translation=(X=0,Y=0,Z=0),Scale3D=(X=1,Y=1,Z=1))),(Item=(Type=Bone,Name=\"skirt_bck_c\"),Transform=(Rotation=(X=0,Y=0,Z=0.5265,W=0.8502),Translation=(X=0,Y=0,Z=0),Scale3D=(X ...(truncated)"
    WeightMinimum=0
    Weight=RigVMVariableNode() {variable=R Thigh Ba}.Value
   )
   > &RigUnit_ModifyTransforms(
    ItemToModify="((Item=(Type=Bone,Name=\"skirt_side_r\"),Transform=(Rotation=(X=0,Y=-0.5593,Z=0,W=0.829),Translation=(X=0,Y=0,Z=0),Scale3D=(X=1,Y=1,Z=1))),(Item=(Type=Bone,Name=\"skirt_fr_r\"),Transform=(Rotation=(X=0.0202,Y=0.454,Z=-0.0063,W=-0.8907),Translation=(X=0,Y=0,Z=0 ...(truncated)"
    WeightMinimum=0
    Weight=RigVMVariableNode() {variable=R Thigh Ot}.Value
   )
   > &RigUnit_ModifyTransforms(
    ItemToModify="((Item=(Type=Bone,Name=\"skirt_fr_r\"),Transform=(Rotation=(X=0.0569,Y=0.2533,Z=0.182,W=-0.9484),Translation=(X=0,Y=0,Z=0),Scale3D=(X=1,Y=1,Z=1))),(Item=(Type=Bone,Name=\"skirt_side_r\"),Transform=(Rotation=(X=0,Y=0,Z=0,W=1),Translation=(X=0,Y=0,Z=0),Scale3D=( ...(truncated)"
    WeightMinimum=0
    Weight=RigVMVariableNode() {variable=R Thigh Crouch Ot}.Value
   )
   > &RigUnit_ModifyTransforms(
    ItemToModify="((Item=(Type=Bone,Name=\"skirt_side_r\"),Transform=(Rotation=(X=-0.2988,Y=-0.0558,Z=-0.517,W=0.8002),Translation=(X=0,Y=0,Z=0),Scale3D=(X=1,Y=1,Z=1))),(Item=(Type=Bone,Name=\"dyn_skirt_side_r_02\"),Transform=(Rotation=(X=0,Y=0,Z=0,W=1),Translation=(X=-2,Y=0,Z= ...(truncated)"
    WeightMinimum=0
    Weight=RigVMVariableNode() {variable=R Thigh Crouch Ot Post}.Value
   )
   > &RigUnit_ModifyTransforms(
    ItemToModify="((Item=(Type=Bone,Name=\"skirt_fr_l\"),Transform=(Rotation=(X=0.2061,Y=0.5324,Z=-0.3057,W=0.762),Translation=(X=0,Y=0,Z=0),Scale3D=(X=1,Y=1,Z=1))),(Item=(Type=Bone,Name=\"skirt_fr_c\"),Transform=(Rotation=(X=0.2439,Y=0.5135,Z=-0.2016,W=0.7976),Translation=(X=0 ...(truncated)"
    WeightMinimum=0
    Weight=RigVMVariableNode() {variable=L Thigh In}.Value
   )
   > &RigUnit_ModifyTransforms(
    ItemToModify="((Item=(Type=Bone,Name=\"skirt_side_r\"),Transform=(Rotation=(X=0,Y=-0.1233,Z=0,W=0.9924),Translation=(X=0,Y=0,Z=0),Scale3D=(X=1,Y=1,Z=1))),(Item=(Type=Bone,Name=\"skirt_fr_r\"),Transform=(Rotation=(X=0.0079,Y=0.1837,Z=0.0847,W=-0.9793),Translation=(X=0,Y=0,Z= ...(truncated)"
    WeightMinimum=0
    Weight=&RigVMFunction_MathDoubleMin(A=RigVMVariableNode() {variable=R Thigh Fr}.Value, B=RigVMVariableNode() {variable=R Thigh Ot}.Value) {template="Minimum::Execute(in A,in B,out Result)"}.Result
   )
   Minimum = &RigVMFunction_MathDoubleMin(A=RigVMVariableNode() {variable=L Thigh Fr}.Value, B=RigVMVariableNode() {variable=R Thigh Fr}.Value) {template="Minimum::Execute(in A,in B,out Result)"}
   &RigUnit_ModifyTransforms(
    ItemToModify="((Item=(Type=Bone,Name=\"skirt_fr_c\"),Transform=(Rotation=(X=0.0194,Y=-0.0319,Z=0.4771,W=0.8781),Translation=(X=0,Y=0,Z=0),Scale3D=(X=1,Y=1,Z=1))),(Item=(Type=Bone,Name=\"skirt_fr_r\"),Transform=(Rotation=(X=0,Y=0,Z=0.1132,W=0.9936),Translation=(X=0,Y=0,Z=0), ...(truncated)"
    WeightMinimum=0
    Weight=@Minimum.Result
    ExecutePin=@RigVMFunction_Sequence.H
   )
   > &RigUnit_ModifyTransforms(
    ItemToModify=((Item=(Type=Bone,Name="skirt_fr_c"),Transform=(Rotation=(X=0.0705,Y=-0.119,Z=0.5839,W=0.7999),Translation=(X=0,Y=0,Z=0),Scale3D=(X=1,Y=1,Z=1))))
    WeightMinimum=0
    Weight=&RigVMFunction_MathDoubleMin(B=RigVMVariableNode() {variable=R Thigh Fr Post}.Value, A=RigVMVariableNode() {variable=L Thigh Fr Post}.Value) {template="Minimum::Execute(in A,in B,out Result)"}.Result
   )
   > &RigUnit_ParentConstraint(
    AdvancedSettings=(InterpolationType=Average,RotationOrderForFilter=XZY)
    Child=(Type=Bone,Name="emblem")
    Parents=((Item=(Type=Bone,Name="pelvis"),Weight=0.15),(Item=(Type=Bone,Name="spine_01"),Weight=0.03),(Item=(Type=Bone,Name="spine_02"),Weight=0.31),(Item=(Type=Bone,Name="spine_03"),Weight=0.12),(Item=(Type=Bone,Name="spine_04"),Weight=0.39))
    Weight=1
   )
  ]}
 RigVMModel Rig {nodes [
   n1 = &RigUnit_PrepareForExecution()
   RigVMFunction_Sequence = RigVMAggregateNode(ExecuteContext=@n1.ExecutePin)
   ItemArray = RigVMDispatchNode(Value=((Type=Bone,Name="belt_fr_l"),(Type=Bone,Name="belt_side_l"),(Type=Bone,Name="belt_bck_l"),(Type=Bone,Name="belt_fr_c"),(Type=Bone,Name="belt_bck_c"))) {template="DISPATCH_RigVMDispatch_Constant(io Value)"}
   For_Each_1 = RigVMDispatchNode(Array=@ItemArray.Value, ExecuteContext=@RigVMFunction_Sequence.B) {template="DISPATCH_RigVMDispatch_ArrayIterator(in Array,out Element,out Index,out Count,out Ratio)"}
   Replace = /Script/RigVM.RigVMFunction_NameReplace(New=n_belt, Old=belt, Name=@For_Each_1.Element.Name) {is=struct RigVMUnitNode, template="Replace::Execute(in Name,in Old,in New,out Result)"}
   > &RigUnit_HierarchyAddNull(
    Parent=(Type=Bone,Name="pelvis")
    Space=GlobalSpace
    Transform=(Rotation=(X=0,Y=0,Z=0,W=1),Translation=(X=0,Y=0,Z=0),Scale3D=(X=1,Y=1,Z=1))
    Transform.Translation=&RigUnit_GetTransform(bInitial=true, Item=@For_Each_1.Element).Transform.Translation
    Transform.Rotation=&RigUnit_GetTransform(Item=(Type=Bone,Name="thigh_l"), bInitial=true).Transform.Rotation
    Name=@Replace.Result
   )
   ItemArray_2 = RigVMDispatchNode(Value=((Type=Bone,Name="belt_fr_r"),(Type=Bone,Name="belt_side_r"),(Type=Bone,Name="belt_bck_r"))) {template="DISPATCH_RigVMDispatch_Constant(io Value)"}
   For_Each_2 = RigVMDispatchNode(Array=@ItemArray_2.Value, ExecuteContext=@For_Each_1.Completed) {template="DISPATCH_RigVMDispatch_ArrayIterator(in Array,out Element,out Index,out Count,out Ratio)"}
   > &RigUnit_HierarchyAddNull(
    Parent=(Type=Bone,Name="pelvis")
    Space=GlobalSpace
    Transform=(Rotation=(X=0,Y=0,Z=0,W=1),Translation=(X=0,Y=0,Z=0),Scale3D=(X=1,Y=1,Z=1))
    Transform.Translation=&RigUnit_GetTransform(bInitial=true, Item=@For_Each_2.Element).Transform.Translation
    Transform.Rotation=&RigUnit_GetTransform(Item=(Type=Bone,Name="thigh_r"), bInitial=true).Transform.Rotation
    Name=&RigVMFunction_NameReplace(New=n_belt, Old=belt, Name=@For_Each_2.Element.Name) {template="Replace::Execute(in Name,in Old,in New,out Result)"}.Result
   )
   RigUnit_GetTransform = &RigUnit_GetTransform(Item=(Type=Bone,Name="thigh_l"), bInitial=true)
   n2 = &RigUnit_HierarchyAddNull(
    Name=Twist Null
    Parent=(Type=Bone,Name="thigh_l")
    Space=GlobalSpace
    Transform=@RigUnit_GetTransform.Transform
    ExecutePin=@For_Each_2.Completed
   )
   > RigVMVariableNode(Value=@n2.Item) {variable=Thigh L Null}
   > RigVMVariableNode(Value=&RigVMFunction_MathTransformMakeRelative(
     Global=@RigUnit_GetTransform.Transform
     Parent=&RigUnit_GetTransform(Item=(Type=Bone,Name="pelvis"), bInitial=true).Transform
    ).Local) {variable=Relative Original Thigh L}
   n3 = &RigUnit_GetTransform(Item=(Type=Bone,Name="thigh_r"), bInitial=true)
   > n4 = &RigUnit_HierarchyAddNull(Name=Thigh R Null, Parent=(Type=Bone,Name="pelvis"), Space=GlobalSpace, Transform=@n3.Transform)
   > RigVMVariableNode(Value=@n4.Item) {variable=Thigh R Null}
   > RigVMVariableNode(Value=&RigVMFunction_MathTransformMakeRelative(Global=@n3.Transform, Parent=&RigUnit_GetTransform(Item=(Type=Bone,Name="pelvis"), bInitial=true).Transform).Local) {variable=Relative Original Thigh R}
   RigVMFunction_ControlFlowBranch = &RigVMFunction_ControlFlowBranch(Condition=RigVMVariableNode() {variable=Debug Controls On}.Value, ExecuteContext=@RigVMFunction_Sequence.A)
   RigVMDispatchNode(ExecuteContext=@RigVMFunction_ControlFlowBranch.True, Array=RigVMVariableNode() {variable=Debug Ctrls}.Value) {template="DISPATCH_RigVMDispatch_ArrayReset(io Array)"}
   ItemArray_3 = RigVMDispatchNode(Value="((Type=Bone,Name=\"thigh_l\"),(Type=Bone,Name=\"thigh_r\"),(Type=Bone,Name=\"pelvis\"),(Type=Bone,Name=\"hand_l\"),(Type=Bone,Name=\"hand_r\"),(Type=Bone,Name=\"clavicle_l\"),(Type=Bone,Name=\"clavicle_r\"),(Type=Bone,Name=\"lowerarm_l\"),(Type=Bone,Name=\"lowerarm_r\"),(Ty ...(truncated)") {template="DISPATCH_RigVMDispatch_Constant(io Value)"}
   > RigVMVariableNode(Value=@ItemArray_3.Value) {variable=Debug Bones}
   > For_Each = RigVMDispatchNode(Array=@ItemArray_3.Value) {template="DISPATCH_RigVMDispatch_ArrayIterator(in Array,out Element,out Index,out Count,out Ratio)"}
   Concat = /Script/RigVM.RigVMFunction_NameConcat(B=" Ctrl", A=@For_Each.Element.Name) {is=struct RigVMUnitNode, template="Concat::Execute(in A,in B,out Result)"}
   > n5 = /Script/ControlRig.RigUnit_HierarchyAddControlTransform(
    InitialValue=(Rotation=(X=0,Y=0,Z=0,W=1),Translation=(X=0,Y=0,Z=0),Scale3D=(X=1,Y=1,Z=1))
    OffsetSpace=GlobalSpace
    Settings="(InitialSpace=LocalSpace,bUsePreferredRotationOrder=False,PreferredRotationOrder=YZX,Limits=(LimitTranslationX=(bMinimum=False,bMaximum=False),LimitTranslationY=(bMinimum=False,bMaximum=False),LimitTranslationZ=(bMinimum=False,bMaximum=False),LimitPitch=(b ...(truncated)"
    Name=@Concat.Result
    Parent=&RigUnit_HierarchyGetParent(bDefaultParent=True, Child=@For_Each.Element).Parent
    OffsetTransform=&RigUnit_GetTransform(bInitial=False, Item=@For_Each.Element).Transform
   ) {
    is=struct RigVMUnitNode
    template="SpawnControl::Execute(in InitialValue,in Settings,in OffsetTransform,in OffsetSpace,in Parent,in Name,out Item)"
   }
   > RigVMDispatchNode(Element=@n5.Item, Array=RigVMVariableNode() {variable=Debug Ctrls}.Value) {template="DISPATCH_RigVMDispatch_ArrayAdd(io Array,in Element,out Index)"}
  ]}
]
hierarchy [
 root {loc=<0 0 0>, rot=<0 0 0 1>}
 pelvis {parent=root, loc=<0 -0.5301 98.0571>, rot=<0.0127 0.707 -0.0127 -0.707>}
 spine_01 {parent=pelvis, loc=<6.1208 -0.9328 0>, rot=<0 0 0.034 -0.9994>}
 spine_02 {parent=spine_01, loc=<6.5997 -0.2095 0>, rot=<0 0 0.0238 -0.9997>}
 spine_03 {parent=spine_02, loc=<6.4914 -0.1237 0>, rot=<0 0 0.0149 -0.9999>}
 spine_04 {parent=spine_03, loc=<6.9598 0.0794 0>, rot=<0 0 -0.0132 -0.9999>}
 spine_05 {parent=spine_04, loc=<6.9987 0 0>, rot=<0 0 -0.0015 -1>}
 neck_01 {parent=spine_05, loc=<11.3138 0.1529 0>, rot=<0 0 -0.2003 -0.9797>}
 neck_02 {parent=neck_01, loc=<4.5627 0 0>, rot=<0 0 0.0504 -0.9987>}
 head {parent=neck_02, loc=<5.3914 0 0>, rot=<0 0 0.115 -0.9934>}
 necklace {parent=neck_01, loc=<2.6962 0.6416 0>, rot=<0 0 0 1>}
 clavicle_l {parent=spine_05, loc=<5.4831 -0.1053 -3.2946>, rot=<0.0465 -0.7289 -0.009 -0.6829>}
 upperarm_l {parent=clavicle_l, loc=<9.8658 -1.2136 -0.4105>, rot=<0.0164 -0.3245 -0.0269 -0.9454>}
 lowerarm_l {parent=upperarm_l, loc=<26.465 -0.8345 0.259>, rot=<0.0036 0.0161 -0.2298 -0.9731>}
 hand_l {parent=lowerarm_l, loc=<22.7355 -0.3203 -0.6105>, rot=<-0.6784 -0.0223 0.0735 -0.7307>}
 thumb_01_l {parent=hand_l, loc=<1.3118 -0.3191 -2.459>, rot=<0.6546 -0.4042 0.0664 -0.6354>}
 thumb_02_l {parent=thumb_01_l, loc=<5.9336 -0.0448 0.0401>, rot=<-0.0037 0.0213 0.298 -0.9543>}
 thumb_03_l {parent=thumb_02_l, loc=<2.2223 0.3682 -0.0907>, rot=<0.0118 -0.013 0.0066 -0.9998>}
 index_metacarpal_l {parent=hand_l, loc=<1.6535 0.0183 -2.96>, rot=<0.0158 -0.0124 -0.0173 -0.9996>}
 index_01_l {parent=index_metacarpal_l, loc=<8.3008 0.0357 -0.0524>, rot=<-0.0044 0.0485 0.0924 -0.9945>}
 index_02_l {parent=index_01_l, loc=<3.3587 -0.3004 0.0452>, rot=<-0.002 0.0027 0.3225 -0.9466>}
 index_03_l {parent=index_02_l, loc=<2.5578 -0.123 -0.0252>, rot=<-0.002 0.0067 0.2558 -0.9667>}
 middle_metacarpal_l {parent=hand_l, loc=<2.1856 -0.1725 -0.5984>, rot=<0.006 0.0168 -0.0329 -0.9993>}
 middle_01_l {parent=middle_metacarpal_l, loc=<7.3275 -0.2447 0.0129>, rot=<-0.0208 0.0042 0.0858 -0.9961>}
 middle_02_l {parent=middle_01_l, loc=<4.0123 0.1401 0.1175>, rot=<-0.0218 0.0258 0.3512 -0.9357>}
 middle_03_l {parent=middle_02_l, loc=<2.6077 0.0534 0.1373>, rot=<0.0082 0.0188 0.1708 -0.9851>}
 ring_metacarpal_l {parent=hand_l, loc=<2.2074 -0.3625 1.3638>, rot=<-0.0284 0.0722 -0.0458 -0.9959>}
 ring_01_l {parent=ring_metacarpal_l, loc=<7.0295 -0.6197 -0.1446>, rot=<-0.0489 -0.0252 0.1892 -0.9804>}
 ring_02_l {parent=ring_01_l, loc=<3.6174 0.4647 -0.1027>, rot=<0 0 0.2806 -0.9598>}
 ring_03_l {parent=ring_02_l, loc=<2.4152 0.114 0.0199>, rot=<0 0 0.1189 -0.9929>}
 pinky_metacarpal_l {parent=hand_l, loc=<1.7616 -0.9753 3.3814>, rot=<-0.0397 0.0814 -0.069 -0.9935>}
 pinky_01_l {parent=pinky_metacarpal_l, loc=<7.1709 -0.4831 -0.1094>, rot=<-0.0654 -0.0035 0.2242 -0.9723>}
 pinky_02_l {parent=pinky_01_l, loc=<2.8274 0.0893 -0.1043>, rot=<0.0037 0.0011 0.2854 -0.9584>}
 pinky_03_l {parent=pinky_02_l, loc=<2.2046 0.0307 -0.0078>, rot=<-0.0002 0.0098 0.1445 -0.9895>}
 deform_wrist_dn_base_l {parent=hand_l, loc=<0 -1.5326 0>, rot=<0 0 0 -1>}
 deform_wrist_dn_l {parent=deform_wrist_dn_base_l, loc=<0 0 0>, rot=<0 0 0 1>}
 deform_wrist_up_base_l {parent=hand_l, loc=<0 2.3009 0>, rot=<0 0 0 -1>}
 deform_wrist_up_l {parent=deform_wrist_up_base_l, loc=<0 0 0>, rot=<0 0 0 1>}
 lowerarm_twist_02_l {parent=lowerarm_l, loc=<10.6707 0 0>, rot=<0 0 0 1>}
 lowerarm_twist_01_l {parent=lowerarm_l, loc=<18.316 0 0>, rot=<0 0 0 1>}
 deform_elbow_in_base_l {parent=lowerarm_l, loc=<0.1834 3.6062 -0.2572>, rot=<0 0 0 -1>}
 deform_elbow_in_l {parent=deform_elbow_in_base_l, loc=<0 0 0>, rot=<0 0 0 1>}
 deform_elbow_out_base_l {parent=lowerarm_l, loc=<0 -4 0>, rot=<0 0 0 1>}
 deform_elbow_out_l {parent=deform_elbow_out_base_l, loc=<0 0 0>, rot=<0 0 0 1>}
 sleeve_l {parent=lowerarm_l, loc=<22.7355 -0.3203 -0.6105>, rot=<-0.6784 -0.0223 0.0735 -0.7307>}
 upperarm_twist_01_l {parent=upperarm_l, loc=<7.2135 -0.1134 0>, rot=<0 0 0 1>}
 deform_upperarm_bk_base_l {parent=upperarm_twist_01_l, loc=<-6.4053 -4.3746 0.314>, rot=<0 0 0 -1>}
 deform_upperarm_bk_l {parent=deform_upperarm_bk_base_l, loc=<0 0 0>, rot=<0 0 0 1>}
 deform_upperarm_fr_base_l {parent=upperarm_twist_01_l, loc=<-5.6312 5.092 0.3354>, rot=<0 0 0 -1>}
 deform_upperarm_fr_l {parent=deform_upperarm_fr_base_l, loc=<0 0 0>, rot=<0 0 0 1>}
 deform_upperarm_up_base_l {parent=upperarm_twist_01_l, loc=<-8.891 0.4825 3.9517>, rot=<0 0 0 -1>}
 deform_upperarm_up_l {parent=deform_upperarm_up_base_l, loc=<0 0 0>, rot=<0 0 0 1>}
 upperarm_twist_02_l {parent=upperarm_l, loc=<19.7998 -0.1134 0>, rot=<0 0 0 1>}
 clavicle_r {parent=spine_05, loc=<5.4831 -0.1053 3.2946>, rot=<0.7289 0.0465 -0.6829 0.009>}
 upperarm_r {parent=clavicle_r, loc=<-9.8658 1.2136 0.4105>, rot=<0.0164 -0.3245 -0.0269 -0.9454>}
 lowerarm_r {parent=upperarm_r, loc=<-26.465 0.8345 -0.259>, rot=<0.0036 0.0161 -0.2298 -0.9731>}
 hand_r {parent=lowerarm_r, loc=<-22.7355 0.3203 0.6106>, rot=<-0.6784 -0.0223 0.0735 -0.7307>}
 thumb_01_r {parent=hand_r, loc=<-1.3118 0.3191 2.459>, rot=<0.6546 -0.4042 0.0664 -0.6354>}
 thumb_02_r {parent=thumb_01_r, loc=<-5.9336 0.0448 -0.0401>, rot=<-0.0037 0.0213 0.298 -0.9543>}
 thumb_03_r {parent=thumb_02_r, loc=<-2.2223 -0.3682 0.0907>, rot=<0.0118 -0.013 0.0066 -0.9998>}
 index_metacarpal_r {parent=hand_r, loc=<-1.6535 -0.0183 2.96>, rot=<0.0158 -0.0124 -0.0173 -0.9996>}
 index_01_r {parent=index_metacarpal_r, loc=<-8.3008 -0.0357 0.0524>, rot=<-0.0044 0.0485 0.0924 -0.9945>}
 index_02_r {parent=index_01_r, loc=<-3.3587 0.3004 -0.0452>, rot=<-0.002 0.0027 0.3225 -0.9466>}
 index_03_r {parent=index_02_r, loc=<-2.5578 0.123 0.0252>, rot=<-0.002 0.0067 0.2558 -0.9667>}
 middle_metacarpal_r {parent=hand_r, loc=<-2.1856 0.1725 0.5984>, rot=<0.006 0.0168 -0.0329 -0.9993>}
 middle_01_r {parent=middle_metacarpal_r, loc=<-7.3275 0.2447 -0.0129>, rot=<-0.0208 0.0042 0.0858 -0.9961>}
 middle_02_r {parent=middle_01_r, loc=<-4.0123 -0.1401 -0.1175>, rot=<-0.0218 0.0258 0.3512 -0.9357>}
 middle_03_r {parent=middle_02_r, loc=<-2.6077 -0.0534 -0.1373>, rot=<0.0082 0.0188 0.1708 -0.9851>}
 ring_metacarpal_r {parent=hand_r, loc=<-2.2074 0.3625 -1.3638>, rot=<-0.0284 0.0722 -0.0458 -0.9959>}
 ring_01_r {parent=ring_metacarpal_r, loc=<-7.0295 0.6197 0.1446>, rot=<-0.0489 -0.0252 0.1892 -0.9804>}
 ring_02_r {parent=ring_01_r, loc=<-3.6174 -0.4647 0.1027>, rot=<0 0 0.2806 -0.9598>}
 ring_03_r {parent=ring_02_r, loc=<-2.4152 -0.114 -0.0199>, rot=<0 0 0.1189 -0.9929>}
 pinky_metacarpal_r {parent=hand_r, loc=<-1.7616 0.9753 -3.3814>, rot=<-0.0397 0.0814 -0.069 -0.9935>}
 pinky_01_r {parent=pinky_metacarpal_r, loc=<-7.1709 0.4831 0.1094>, rot=<-0.0654 -0.0035 0.2242 -0.9723>}
 pinky_02_r {parent=pinky_01_r, loc=<-2.8274 -0.0893 0.1043>, rot=<0.0037 0.0011 0.2854 -0.9584>}
 pinky_03_r {parent=pinky_02_r, loc=<-2.2046 -0.0307 0.0078>, rot=<-0.0002 0.0098 0.1445 -0.9895>}
 deform_wrist_dn_base_r {parent=hand_r, loc=<-0.0001 1.5328 0>, rot=<0 0 0 -1>}
 deform_wrist_dn_r {parent=deform_wrist_dn_base_r, loc=<0 0 0>, rot=<0 0 0 1>}
 deform_wrist_up_base_r {parent=hand_r, loc=<-0.0002 -2.3007 -0.0001>, rot=<0 0 0 -1>}
 deform_wrist_up_r {parent=deform_wrist_up_base_r, loc=<0 0 0>, rot=<0 0 0 1>}
 lowerarm_twist_02_r {parent=lowerarm_r, loc=<-10.6707 0 0>, rot=<0 0 0 -1>}
 lowerarm_twist_01_r {parent=lowerarm_r, loc=<-18.316 0 0>, rot=<0 0 0 -1>}
 deform_elbow_in_base_r {parent=lowerarm_r, loc=<-0.1834 -3.6062 0.2572>, rot=<0 0 0 -1>}
 deform_elbow_in_r {parent=deform_elbow_in_base_r, loc=<0 0 0>, rot=<0 0 0 1>}
 deform_elbow_out_base_r {parent=lowerarm_r, loc=<0 4 -0.0001>, rot=<0 0 0 -1>}
 deform_elbow_out_r {parent=deform_elbow_out_base_r, loc=<0 0 0>, rot=<0 0 0 1>}
 sleeve_r {parent=lowerarm_r, loc=<-22.7355 0.3203 0.6106>, rot=<-0.6784 -0.0223 0.0735 -0.7307>}
 upperarm_twist_01_r {parent=upperarm_r, loc=<-7.2135 0.1134 0>, rot=<0 0 0 -1>}
 deform_upperarm_bk_base_r {parent=upperarm_twist_01_r, loc=<6.4053 4.3746 -0.314>, rot=<0 0 0 -1>}
 deform_upperarm_bk_r {parent=deform_upperarm_bk_base_r, loc=<0 0 0>, rot=<0 0 0 1>}
 deform_upperarm_fr_base_r {parent=upperarm_twist_01_r, loc=<5.6312 -5.092 -0.3354>, rot=<0 0 0 -1>}
 deform_upperarm_fr_r {parent=deform_upperarm_fr_base_r, loc=<0 0 0>, rot=<0 0 0 1>}
 deform_upperarm_up_base_r {parent=upperarm_twist_01_r, loc=<8.891 -0.4825 -3.9517>, rot=<0 0 0 -1>}
 deform_upperarm_up_r {parent=deform_upperarm_up_base_r, loc=<0 0 0>, rot=<0 0 0 1>}
 upperarm_twist_02_r {parent=upperarm_r, loc=<-19.7998 0.1134 0>, rot=<0 0 0 -1>}
 attach_backpack {parent=spine_05, loc=<-0.123 -13.473 0>, rot=<-0.0093 -0.707 0.0093 -0.707>}
 Backpack_BR {parent=attach_backpack, type=Null, loc=<0 0 0>, rot=<0 0 0 1>}
 attach_cape {parent=spine_05, loc=<9.017 -11.714 0>, rot=<0.0616 -0.7044 -0.0616 -0.7044>}
 cape {parent=attach_cape, type=Null, loc=<0 0 0>, rot=<0 0 0 1>}
 deform_pec_aim_base_l {parent=spine_04, loc=<3.2456 9.3309 -1.2237>, rot=<0.0603 -0.4033 0.3318 -0.8507>}
 deform_pec_aim_l {parent=deform_pec_aim_base_l, loc=<0 0 0>, rot=<0 0 0 1>}
 deform_pec_base_l {parent=deform_pec_aim_l, loc=<9.3665 0 0>, rot=<0 0 0 1>}
 deform_pec_l {parent=deform_pec_base_l, loc=<0 0 0>, rot=<0 0 0 1>}
 chest_wing_a_l {parent=deform_pec_l, loc=<-0.667 4.9409 -2.1545>, rot=<0.2452 -0.3282 -0.2599 -0.8744>}
 chest_wing_b_l {parent=deform_pec_l, loc=<-1.3917 3.9343 -5.1255>, rot=<0.2452 -0.3282 -0.2599 -0.8744>}
 chest_wing_c_l {parent=deform_pec_l, loc=<-2.4936 2.4568 -7.2909>, rot=<0.2452 -0.3282 -0.2599 -0.8744>}
 deform_pec_aim_base_r {parent=spine_04, loc=<3.2452 9.3309 1.2237>, rot=<-0.4033 -0.0603 0.8507 0.3318>}
 deform_pec_aim_r {parent=deform_pec_aim_base_r, loc=<0 0 0>, rot=<0 0 0 1>}
 deform_pec_base_r {parent=deform_pec_aim_r, loc=<-9.3667 -0.0003 -0.0004>, rot=<0 0 0 1>}
 deform_pec_r {parent=deform_pec_base_r, loc=<0 0 0>, rot=<0 0 0 1>}
 chest_wing_a_r {parent=deform_pec_r, loc=<0.667 -4.9409 2.1546>, rot=<0.2452 -0.3282 -0.2599 -0.8744>}
 chest_wing_b_r {parent=deform_pec_r, loc=<1.3917 -3.9343 5.1256>, rot=<0.2452 -0.3282 -0.2599 -0.8744>}
 chest_wing_c_r {parent=deform_pec_r, loc=<2.4936 -2.4567 7.291>, rot=<0.2452 -0.3282 -0.2599 -0.8744>}
 thigh_l {parent=pelvis, loc=<-1.0855 -0.2109 -8.7053>, rot=<0.0009 0.0355 0.0155 -0.9993>}
 calf_l {parent=thigh_l, loc=<-42.5284 -0.0019 -0.7297>, rot=<0.0006 -0.0161 -0.0391 -0.9991>}
 foot_l {parent=calf_l, loc=<-46.7401 -2.5014 -0.592>, rot=<0.0427 -0.0238 0.0353 -0.9982>}
 ball_l {parent=foot_l, loc=<-3.5658 10.8066 -0.1544>, rot=<0.0001 -0.0001 -0.7063 -0.7079>}
 calf_twist_02_l {parent=calf_l, loc=<-19.1546 -1.7943 -0.2638>, rot=<0.0237 0.0095 -0.0035 -0.9997>}
 calf_twist_01_l {parent=calf_l, loc=<-35.8483 -1.9197 -0.5794>, rot=<0.0237 0.0095 -0.0035 -0.9997>}
 deform_knee_in_base_l {parent=calf_l, loc=<0.8031 -5.0231 0.2456>, rot=<0 0 0 -1>}
 deform_knee_in_l {parent=deform_knee_in_base_l, loc=<0 0 0>, rot=<0 0 0 1>}
 deform_knee_out_base_l {parent=calf_l, loc=<0 4 0>, rot=<0 0 0 -1>}
 deform_knee_out_l {parent=deform_knee_out_base_l, loc=<0 0 0>, rot=<0 0 0 1>}
 thigh_twist_01_l {parent=thigh_l, loc=<-27.952 0 0>, rot=<0.0013 0 0 -1>}
 deform_groin_base_l {parent=thigh_l, loc=<0.4801 5.6776 -0.2486>, rot=<0 0 0 -1>}
 deform_groin_l {parent=deform_groin_base_l, loc=<0 0 0>, rot=<0 0 0 1>}
 thigh_r {parent=pelvis, loc=<-1.0855 -0.2109 8.7053>, rot=<0.0355 -0.0009 0.9993 0.0155>}
 calf_r {parent=thigh_r, loc=<42.5284 0.0019 0.7297>, rot=<0.0006 -0.0161 -0.0391 -0.9991>}
 foot_r {parent=calf_r, loc=<46.7401 2.5014 0.592>, rot=<0.0427 -0.0238 0.0353 -0.9982>}
 ball_r {parent=foot_r, loc=<3.5658 -10.8066 0.1544>, rot=<0.0001 -0.0001 -0.7063 -0.7079>}
 calf_twist_02_r {parent=calf_r, loc=<19.1546 1.7943 0.2638>, rot=<0.0237 0.0095 -0.0035 -0.9997>}
 calf_twist_01_r {parent=calf_r, loc=<35.8483 1.9197 0.5794>, rot=<0.0237 0.0095 -0.0035 -0.9997>}
 deform_knee_in_base_r {parent=calf_r, loc=<-0.8031 5.0231 -0.2457>, rot=<0 0 0 -1>}
 deform_knee_in_r {parent=deform_knee_in_base_r, loc=<0 0 0>, rot=<0 0 0 1>}
 deform_knee_out_base_r {parent=calf_r, loc=<0 -4 0>, rot=<0 0 0 -1>}
 deform_knee_out_r {parent=deform_knee_out_base_r, loc=<0 0 0>, rot=<0 0 0 1>}
 thigh_twist_01_r {parent=thigh_r, loc=<27.952 0 0>, rot=<0.0013 0 0 -1>}
 deform_groin_base_r {parent=thigh_r, loc=<-0.4801 -5.6776 0.2486>, rot=<0 0 0 -1>}
 deform_groin_r {parent=deform_groin_base_r, loc=<0 0 0>, rot=<0 0 0 1>}
 deform_glute_base_l {parent=pelvis, loc=<-7.3466 -9.0213 -9.1245>, rot=<0.0172 0.031 -0.4477 -0.8935>}
 deform_glute_l {parent=deform_glute_base_l, loc=<0 0 0>, rot=<0 0 0 1>}
 deform_glute_base_r {parent=pelvis, loc=<-7.3464 -9.0213 9.1245>, rot=<0.031 -0.0172 0.8935 -0.4477>}
 deform_glute_r {parent=deform_glute_base_r, loc=<0 0 0>, rot=<0 0 0 1>}
 belt_fr_l {parent=pelvis, loc=<8.3469 6.5283 -7.5626>, rot=<0.0009 0.0355 0.0155 -0.9993>}
 belt_side_l {parent=pelvis, loc=<10.1473 -0.5196 -10.7905>, rot=<0.0009 0.0355 0.0155 -0.9993>}
 belt_bck_l {parent=pelvis, loc=<9.9749 -7.4967 -7.4907>, rot=<0.0009 0.0355 0.0155 -0.9993>}
 belt_bck_c {parent=pelvis, loc=<9.7813 -8.7897 0>, rot=<0.0009 0.0355 0.0155 -0.9993>}
 belt_fr_c {parent=pelvis, loc=<1.5112 8.3961 0>, rot=<0.0009 0.0355 0.0155 -0.9993>}
 belt_fr_r {parent=pelvis, loc=<8.1425 6.752 7.269>, rot=<0.0355 -0.0009 0.9993 0.0155>}
 belt_side_r {parent=pelvis, loc=<10.2186 -0.4139 10.8793>, rot=<0.0355 -0.0009 0.9993 0.0155>}
 belt_bck_r {parent=pelvis, loc=<9.9749 -7.4967 7.4907>, rot=<0.0355 -0.0009 0.9993 0.0155>}
 skirt_fr_l {parent=pelvis, loc=<8.3469 6.5283 -7.5626>, rot=<0.0009 0.0355 0.0155 -0.9993>}
 dyn_skirt_fr_l_base {parent=skirt_fr_l, loc=<0 0 0>, rot=<0.0078 -0.797 -0.5984 -0.082>}
 dyn_skirt_fr_l_01 {parent=dyn_skirt_fr_l_base, loc=<0 0 0>, rot=<0 0 0 1>}
 dyn_skirt_fr_l_02 {parent=dyn_skirt_fr_l_01, loc=<7.0778 0 0>, rot=<0 0 0 1>}
 dyn_skirt_fr_l_03 {parent=dyn_skirt_fr_l_02, loc=<7.0778 0 0>, rot=<0 0 0 1>}
 dyn_skirt_fr_l_04 {parent=dyn_skirt_fr_l_03, loc=<7.0778 0 0>, rot=<0 0 0 1>}
 skirt_side_l {parent=pelvis, loc=<10.1473 -0.5196 -10.7905>, rot=<0.0009 0.0355 0.0155 -0.9993>}
 dyn_skirt_side_l_base {parent=skirt_side_l, loc=<0 0 0>, rot=<0.0577 -0.9856 -0.0139 -0.1586>}
 dyn_skirt_side_l_01 {parent=dyn_skirt_side_l_base, loc=<0 0 0>, rot=<0 0 0 1>}
 dyn_skirt_side_l_02 {parent=dyn_skirt_side_l_01, loc=<8.7609 0 0>, rot=<0 0 0 1>}
 dyn_skirt_side_l_03 {parent=dyn_skirt_side_l_02, loc=<7.2043 0 0>, rot=<0 0 0 1>}
 dyn_skirt_side_l_04 {parent=dyn_skirt_side_l_03, loc=<7.4992 0 0>, rot=<0 0 0 1>}
 skirt_bck_l {parent=pelvis, loc=<9.9749 -7.4967 -7.4907>, rot=<0.0009 0.0355 0.0155 -0.9993>}
 dyn_skirt_bck_l_base {parent=skirt_bck_l, loc=<0 0 0>, rot=<0.1125 -0.8503 0.4669 -0.2152>}
 dyn_skirt_bck_l_01 {parent=dyn_skirt_bck_l_base, loc=<0 0 0>, rot=<0 0 0 1>}
 dyn_skirt_bck_l_02 {parent=dyn_skirt_bck_l_01, loc=<7.8457 0 0>, rot=<0 -0.1352 0 -0.9908>}
 dyn_skirt_bck_l_03 {parent=dyn_skirt_bck_l_02, loc=<7.8457 0 0>, rot=<0 -0.0676 0 -0.9977>}
 dyn_skirt_bck_l_04 {parent=dyn_skirt_bck_l_03, loc=<7.8457 0 0>, rot=<0 0 0 1>}
 skirt_bck_c {parent=pelvis, loc=<9.7813 -8.7897 0>, rot=<0.0009 0.0355 0.0155 -0.9993>}
 dyn_skirt_bck_c_base {parent=skirt_bck_c, loc=<0 0 0>, rot=<0.2087 -0.6889 0.6753 -0.1604>}
 dyn_skirt_bck_c_01 {parent=dyn_skirt_bck_c_base, loc=<0 0 0>, rot=<0 0 0 1>}
 dyn_skirt_bck_c_02 {parent=dyn_skirt_bck_c_01, loc=<7.9493 0 0>, rot=<0 -0.1168 0 -0.9932>}
 dyn_skirt_bck_c_03 {parent=dyn_skirt_bck_c_02, loc=<7.9493 0 0>, rot=<0 -0.0409 0 -0.9992>}
 dyn_skirt_bck_c_04 {parent=dyn_skirt_bck_c_03, loc=<7.9493 0 0>, rot=<0 0 0 1>}
 skirt_fr_c {parent=pelvis, loc=<1.5112 8.3961 0>, rot=<0.0009 0.0355 0.0155 -0.9993>}
 dyn_skirt_fr_c_base {parent=skirt_fr_c, loc=<0 0 0>, rot=<-0.0401 -0.7068 -0.7062 0.0101>}
 dyn_skirt_fr_c_01 {parent=dyn_skirt_fr_c_base, loc=<0 0 0>, rot=<0 0 0 1>}
 dyn_skirt_fr_c_02 {parent=dyn_skirt_fr_c_01, loc=<4.7059 0 0>, rot=<0 0.0523 0 -0.9986>}
 dyn_skirt_fr_c_03 {parent=dyn_skirt_fr_c_02, loc=<4.7059 0 0>, rot=<0 0.0172 0 -0.9999>}
 dyn_skirt_fr_c_04 {parent=dyn_skirt_fr_c_03, loc=<4.7059 0 0>, rot=<0 0 0 1>}
 skirt_fr_r {parent=pelvis, loc=<8.1425 6.752 7.269>, rot=<0.0355 -0.0009 0.9993 0.0155>}
 dyn_skirt_fr_r_base {parent=skirt_fr_r, loc=<0 0 0>, rot=<-0.6085 0.0832 -0.0513 -0.7875>}
 dyn_skirt_fr_r_01 {parent=dyn_skirt_fr_r_base, loc=<0 0 0>, rot=<0 0 0 1>}
 dyn_skirt_fr_r_02 {parent=dyn_skirt_fr_r_01, loc=<7.0189 0 0>, rot=<0 0.0423 0 -0.9991>}
 dyn_skirt_fr_r_03 {parent=dyn_skirt_fr_r_02, loc=<7.0189 0 0>, rot=<0 0.0308 0 -0.9995>}
 dyn_skirt_fr_r_04 {parent=dyn_skirt_fr_r_03, loc=<7.0189 0 0>, rot=<0 0 0 1>}
 skirt_side_r {parent=pelvis, loc=<10.2186 -0.4139 10.8793>, rot=<0.0355 -0.0009 0.9993 0.0155>}
 dyn_skirt_side_r_base {parent=skirt_side_r, loc=<0 0 0>, rot=<0.0032 0.1996 -0.0212 -0.9796>}
 dyn_skirt_side_r_01 {parent=dyn_skirt_side_r_base, loc=<0 0 0>, rot=<0 0 0 1>}
 dyn_skirt_side_r_02 {parent=dyn_skirt_side_r_01, loc=<7.5544 0 0>, rot=<0 -0.0348 0 -0.9994>}
 dyn_skirt_side_r_03 {parent=dyn_skirt_side_r_02, loc=<7.5544 0 0>, rot=<0 0.059 0 -0.9983>}
 dyn_skirt_side_r_04 {parent=dyn_skirt_side_r_03, loc=<7.5544 0 0>, rot=<0 0 0 1>}
 skirt_bck_r {parent=pelvis, loc=<9.9749 -7.4967 7.4907>, rot=<0.0355 -0.0009 0.9993 0.0155>}
 dyn_skirt_bck_r_base {parent=skirt_bck_r, loc=<0 0 0>, rot=<0.4678 0.2257 -0.0941 -0.8493>}
 dyn_skirt_bck_r_01 {parent=dyn_skirt_bck_r_base, loc=<0 0 0>, rot=<0 0 0 1>}
 dyn_skirt_bck_r_02 {parent=dyn_skirt_bck_r_01, loc=<7.9213 0 0>, rot=<0 -0.1148 0 -0.9934>}
 dyn_skirt_bck_r_03 {parent=dyn_skirt_bck_r_02, loc=<7.9213 0 0>, rot=<0 -0.0163 0 -0.9999>}
 dyn_skirt_bck_r_04 {parent=dyn_skirt_bck_r_03, loc=<7.9213 0 0>, rot=<0 0 0 1>}
 thigh_l_twist_up {parent=pelvis, loc=<-1.0855 -0.2109 -8.7053>, rot=<0.0009 0.0355 0.0155 -0.9993>}
 thigh_l_twist_end {parent=thigh_l_twist_up, loc=<0 0 -7.25>, rot=<0 0 0 1>}
 thigh_r_twist_up {parent=pelvis, loc=<-1.0855 -0.2109 8.7053>, rot=<0.0355 -0.0009 0.9993 0.0155>}
 thigh_r_twist_end {parent=thigh_r_twist_up, loc=<0 0 7.25>, rot=<0 0 0 1>}
 emblem {parent=pelvis, loc=<17.6731 8.8909 0>, rot=<0 0 0.0578 -0.9983>}
]
variables [
 Debug Bones {type="TArray<FRigElementKey>", default=((Type=Bone,Name="calf_l"),(Type=Bone,Name="calf_r"))}
 Debug Controls On {type=bool, default=False}
 Debug Ctrls {type="TArray<FRigElementKey>", default=(())}
 Debug Shape {
  type=FRigUnit_HierarchyAddControl_ShapeSettings
  default=(bVisible=True,Name="Circle_Thick",Color=(R=1,G=0.9718,B=0,A=1),Transform=(Rotation=(X=0,Y=-0.7071,Z=0,W=0.7071),Translation=(X=0,Y=0,Z=0),Scale3D=(X=2,Y=2,Z=2)))
 }
 L Thigh Ba {type=double, default=0}
 L Thigh Crouch Ot {type=double, default=0}
 L Thigh Crouch Ot Post {type=double, default=0}
 L Thigh Crouch Pre Ot {type=double, default=0}
 L Thigh Fr {type=double, default=0}
 L Thigh Fr Post {type=double, default=0}
 L Thigh Fr Pre {type=double, default=0}
 L Thigh In {type=double, default=0}
 L Thigh Ot {type=double, default=0}
 R Thigh Ba {type=double, default=0}
 R Thigh Crouch {type=double, default=0}
 R Thigh Crouch Ot {type=double, default=0}
 R Thigh Crouch Ot Post {type=double, default=0}
 R Thigh Crouch Pre Ot {type=double, default=0}
 R Thigh Fr {type=double, default=0}
 R Thigh Fr Post {type=double, default=0}
 R Thigh Fr Pre {type=double, default=0}
 R Thigh In {type=double, default=0}
 R Thigh Ot {type=double, default=0}
 Relative Original Thigh L {type=FTransform, default=(Rotation=(X=0,Y=0,Z=0,W=1),Translation=(X=0,Y=0,Z=0),Scale3D=(X=1,Y=1,Z=1))}
 Relative Original Thigh R {type=FTransform, default=(Rotation=(X=0,Y=0,Z=0,W=1),Translation=(X=0,Y=0,Z=0),Scale3D=(X=1,Y=1,Z=1))}
 Skirt L Ba Aim Vector {type=FVector, default=(X=0,Y=0,Z=0)}
 Skirt L Ba Sec Vector {type=FVector, default=(X=0,Y=0,Z=0)}
 Skirt L Fr Aim Vector {type=FVector, default=(X=0,Y=0,Z=0)}
 Skirt L Fr Sec Vector {type=FVector, default=(X=0,Y=0,Z=0)}
 Skirt L IO Aim Vector {type=FVector, default=(X=0,Y=0,Z=0)}
 Skirt L IO Sec Vector {type=FVector, default=(X=0,Y=0,Z=0)}
 Skirt L Sec Vector {type=FVector, default=(X=0,Y=0,Z=0)}
 Skirt R Ba Aim Vector {type=FVector, default=(X=0,Y=0,Z=0)}
 Skirt R Fr Aim Vector {type=FVector, default=(X=0,Y=0,Z=0)}
 Skirt R IO Aim Vector {type=FVector, default=(X=0,Y=0,Z=0)}
 Skirt R Sec Vector {type=FVector, default=(X=0,Y=0,Z=0)}
 Thigh L Null {type=FRigElementKey, default=(Type=Bone,Name="")}
 Thigh R Null {type=FRigElementKey, default=(Type=Bone,Name="")}
]
