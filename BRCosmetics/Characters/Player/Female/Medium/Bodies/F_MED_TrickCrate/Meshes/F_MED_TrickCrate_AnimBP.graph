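# Post-copy Anim Blueprint for the female TrickCrate character body, adding secondary motion to its skirt.
# It copies the parent mesh pose, runs a Control Rig, then simulates eight four-bone skirt chains with sprung AnimDynamics into a cached pose.
# When IsTheCape is true, the output also partially constrains three back skirt bones to the thighs.
# On customization load, that flag is set from a TrickCrate backpack meta-tag check.
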AnimBlueprint /BRCosmetics/Characters/Player/Female/Medium/Bodies/F_MED_TrickCrate/Meshes/F_MED_TrickCrate_AnimBP
target_skeleton=&/F_MED_TrickCrate_Skeleton
graphs [
 AnimGraph {nodes [
   n1 = "Constraint - dyn_skirt_bck_l_03 (P)"(ComponentPose=Local To Component(LocalPose=Use cached pose 'Dynamics_cache'()), Alpha=0.3, ConstraintWeights_0=1) {boneToModify {boneName=dyn_skirt_bck_l_03}, constraintSetup [{targetBone {boneName=thigh_l}}], constraintWeights [1]}
   n2 = "Constraint - dyn_skirt_bck_c_03 (PP)"!(ComponentPose=@n1, Alpha=0.3, ConstraintWeights_0=0.5, ConstraintWeights_1:float=0.5) -> (Pose) {
    boneToModify {boneName=dyn_skirt_bck_c_03}
    constraintSetup [{targetBone {boneName=thigh_l}}, {targetBone {boneName=thigh_r}}]
    constraintWeights [1,1]
   }
   n3 = "Constraint - dyn_skirt_bck_r_03 (P)"(ComponentPose=@n2, Alpha=0.3, ConstraintWeights_0=1) {boneToModify {boneName=dyn_skirt_bck_r_03}, constraintSetup [{targetBone {boneName=thigh_r}}], constraintWeights [1]}
   Output Pose(Result=Blend Poses by bool(
     BlendPose_0=Component To Local(ComponentPose=@n3)
     BlendPose_1=Use cached pose 'Dynamics_cache'()
     bActiveValue=Get IsTheCape() -> (IsTheCape:bool)
    )) {layerGroup=None}
   n4 = Control Rig(Source=Copy Pose From Mesh() {bUseAttachedParent=true}) {controlRigAssetReference {blueprintRigClass=&/F_MED_TrickCrate_Flip_Rig.F_MED_TrickCrate_Flip_Rig_C}}
   n5 = Anim Dynamics - Bone: dyn_skirt_fr_c_01!(SimulationSpace:AnimPhysSimSpaceType Enum=Component, ComponentPose=Local To Component(LocalPose=@n4), Alpha=1, ExternalForce="0,0,0") -> (Pose) {
    angularDampingOverride=0.99
    angularSpringConstant=3000
    bAngularSpring=true
    bOverrideAngularDamping=true
    bUseGravityOverride=true
    gravityScale=0.35
    simulationSpace=BoneRelative
    alphaBoolBlend {blendOption=Linear}
    boundBone {boneName=dyn_skirt_fr_c_01}
    componentAppliedLinearAccClamp=<1000 1000 1000>
    componentLinearAccScale=<0.7 0.7 0.7>
    componentLinearVelScale=<0.7 0.7 0.7>
    gravityOverride {z=-980}
    physicsBodyDefinitions [{
      boundBone {boneName=dyn_skirt_fr_c_01}
      boxExtents=<12 12 12>
      constraintSetup {angularLimitsMax=<0 0 0>, angularLimitsMin=<0 -10 0>, angularTarget=<1 0 0>}
      localJointOffset=<2 0 0>
     }]
    relativeSpaceBone {boneName=root}
   }
   n6 = Anim Dynamics - Bone: dyn_skirt_fr_c_02!(SimulationSpace:AnimPhysSimSpaceType Enum=BoneRelative, ComponentPose=@n5, Alpha=1, ExternalForce="0,0,0") -> (Pose) {
    angularDampingOverride=0.99
    angularSpringConstant=3000
    bAngularSpring=true
    bOverrideAngularDamping=true
    bUseGravityOverride=true
    gravityScale=0.35
    simulationSpace=BoneRelative
    alphaBoolBlend {blendOption=Linear}
    boundBone {boneName=dyn_skirt_fr_c_02}
    componentAppliedLinearAccClamp=<1000 1000 1000>
    componentLinearAccScale=<0.7 0.7 0.7>
    componentLinearVelScale=<0.7 0.7 0.7>
    gravityOverride {z=-980}
    physicsBodyDefinitions [{
      boundBone {boneName=dyn_skirt_fr_c_02}
      boxExtents=<12 12 12>
      constraintSetup {angularLimitsMax=<0 0 0>, angularLimitsMin=<0 -20 0>, angularTarget=<1 0 0>}
      localJointOffset=<2 0 0>
     }]
    relativeSpaceBone {boneName=root}
   }
   n7 = Anim Dynamics - Bone: dyn_skirt_fr_c_03!(SimulationSpace:AnimPhysSimSpaceType Enum=BoneRelative, ComponentPose=@n6, Alpha=1, ExternalForce="0,0,0") -> (Pose) {
    angularDampingOverride=0.99
    angularSpringConstant=3000
    bAngularSpring=true
    bOverrideAngularDamping=true
    bUseGravityOverride=true
    gravityScale=0.35
    simulationSpace=BoneRelative
    alphaBoolBlend {blendOption=Linear}
    boundBone {boneName=dyn_skirt_fr_c_03}
    componentAppliedLinearAccClamp=<1000 1000 1000>
    componentLinearAccScale=<0.7 0.7 0.7>
    componentLinearVelScale=<0.7 0.7 0.7>
    gravityOverride {z=-980}
    physicsBodyDefinitions [{
      boundBone {boneName=dyn_skirt_fr_c_03}
      boxExtents=<12 12 12>
      constraintSetup {angularLimitsMax=<0 0 0>, angularLimitsMin=<0 -50 0>, angularTarget=<1 0 0>}
      localJointOffset=<2 0 0>
     }]
    relativeSpaceBone {boneName=root}
   }
   n8 = Anim Dynamics - Bone: dyn_skirt_fr_c_04!(SimulationSpace:AnimPhysSimSpaceType Enum=BoneRelative, ComponentPose=@n7, Alpha=1, ExternalForce="0,0,0") -> (Pose) {
    angularDampingOverride=0.99
    angularSpringConstant=1000
    bAngularSpring=true
    bOverrideAngularDamping=true
    bUseGravityOverride=true
    gravityScale=0.35
    simulationSpace=BoneRelative
    alphaBoolBlend {blendOption=Linear}
    boundBone {boneName=dyn_skirt_fr_c_04}
    componentAppliedLinearAccClamp=<1000 1000 1000>
    componentLinearAccScale=<0.7 0.7 0.7>
    componentLinearVelScale=<0.7 0.7 0.7>
    gravityOverride {z=-980}
    physicsBodyDefinitions [{
      boundBone {boneName=dyn_skirt_fr_c_04}
      boxExtents=<12 12 12>
      constraintSetup {angularLimitsMax=<0 10 0>, angularLimitsMin=<0 -30 0>, angularTarget=<1 0 0>}
      localJointOffset=<1 0 0>
     }]
    relativeSpaceBone {boneName=root}
   }
   n9 = Anim Dynamics - Bone: dyn_skirt_fr_l_01!(SimulationSpace:AnimPhysSimSpaceType Enum=BoneRelative, ComponentPose=@n8, Alpha=1, ExternalForce="0,0,0") -> (Pose) {
    angularDampingOverride=0.99
    angularSpringConstant=3000
    bAngularSpring=true
    bOverrideAngularDamping=true
    bUseGravityOverride=true
    gravityScale=0.35
    simulationSpace=BoneRelative
    alphaBoolBlend {blendOption=Linear}
    boundBone {boneName=dyn_skirt_fr_l_01}
    componentAppliedLinearAccClamp=<1000 1000 1000>
    componentLinearAccScale=<0.7 0.7 0.7>
    componentLinearVelScale=<0.7 0.7 0.7>
    gravityOverride {z=-980}
    physicsBodyDefinitions [{
      boundBone {boneName=dyn_skirt_fr_l_01}
      boxExtents=<12 12 12>
      constraintSetup {angularLimitsMax=<0 0 0>, angularLimitsMin=<0 -10 0>, angularTarget=<1 0 0>}
      localJointOffset=<2 0 0>
     }]
    relativeSpaceBone {boneName=root}
   }
   n10 = Anim Dynamics - Bone: dyn_skirt_fr_l_02!(SimulationSpace:AnimPhysSimSpaceType Enum=BoneRelative, ComponentPose=@n9, Alpha=1, ExternalForce="0,0,0") -> (Pose) {
    angularDampingOverride=0.99
    angularSpringConstant=3000
    bAngularSpring=true
    bOverrideAngularDamping=true
    bUseGravityOverride=true
    gravityScale=0.35
    simulationSpace=BoneRelative
    alphaBoolBlend {blendOption=Linear}
    boundBone {boneName=dyn_skirt_fr_l_02}
    componentAppliedLinearAccClamp=<1000 1000 1000>
    componentLinearAccScale=<0.7 0.7 0.7>
    componentLinearVelScale=<0.7 0.7 0.7>
    gravityOverride {z=-980}
    physicsBodyDefinitions [{
      boundBone {boneName=dyn_skirt_fr_l_02}
      boxExtents=<12 12 12>
      constraintSetup {angularLimitsMax=<0 0 0>, angularLimitsMin=<0 -20 0>, angularTarget=<1 0 0>}
      localJointOffset=<2 0 0>
     }]
    relativeSpaceBone {boneName=root}
   }
   n11 = Anim Dynamics - Bone: dyn_skirt_fr_l_03!(SimulationSpace:AnimPhysSimSpaceType Enum=BoneRelative, ComponentPose=@n10, Alpha=1, ExternalForce="0,0,0") -> (Pose) {
    angularDampingOverride=0.99
    angularSpringConstant=3000
    bAngularSpring=true
    bOverrideAngularDamping=true
    bUseGravityOverride=true
    gravityScale=0.35
    simulationSpace=BoneRelative
    alphaBoolBlend {blendOption=Linear}
    boundBone {boneName=dyn_skirt_fr_l_03}
    componentAppliedLinearAccClamp=<1000 1000 1000>
    componentLinearAccScale=<0.7 0.7 0.7>
    componentLinearVelScale=<0.7 0.7 0.7>
    gravityOverride {z=-980}
    physicsBodyDefinitions [{
      boundBone {boneName=dyn_skirt_fr_l_03}
      boxExtents=<12 12 12>
      constraintSetup {angularLimitsMax=<0 0 0>, angularLimitsMin=<0 -50 0>, angularTarget=<1 0 0>}
      localJointOffset=<2 0 0>
     }]
    relativeSpaceBone {boneName=root}
   }
   n12 = Anim Dynamics - Bone: dyn_skirt_fr_l_04!(SimulationSpace:AnimPhysSimSpaceType Enum=BoneRelative, ComponentPose=@n11, Alpha=1, ExternalForce="0,0,0") -> (Pose) {
    angularDampingOverride=0.99
    angularSpringConstant=1000
    bAngularSpring=true
    bOverrideAngularDamping=true
    bUseGravityOverride=true
    gravityScale=0.35
    simulationSpace=BoneRelative
    alphaBoolBlend {blendOption=Linear}
    boundBone {boneName=dyn_skirt_fr_l_04}
    componentAppliedLinearAccClamp=<1000 1000 1000>
    componentLinearAccScale=<0.7 0.7 0.7>
    componentLinearVelScale=<0.7 0.7 0.7>
    gravityOverride {z=-980}
    physicsBodyDefinitions [{
      boundBone {boneName=dyn_skirt_fr_l_04}
      boxExtents=<12 12 12>
      constraintSetup {angularLimitsMax=<0 0 0>, angularLimitsMin=<0 -30 0>, angularTarget=<1 0 0>}
      localJointOffset=<1 0 0>
     }]
    relativeSpaceBone {boneName=root}
   }
   n13 = Anim Dynamics - Bone: dyn_skirt_side_l_01!(SimulationSpace:AnimPhysSimSpaceType Enum=BoneRelative, ComponentPose=@n12, Alpha=1, ExternalForce="0,0,0") -> (Pose) {
    angularDampingOverride=0.99
    angularSpringConstant=3000
    bAngularSpring=true
    bOverrideAngularDamping=true
    bUseGravityOverride=true
    gravityScale=0.35
    simulationSpace=BoneRelative
    alphaBoolBlend {blendOption=Linear}
    boundBone {boneName=dyn_skirt_side_l_01}
    componentAppliedLinearAccClamp=<1000 1000 1000>
    componentLinearAccScale=<0.7 0.7 0.7>
    componentLinearVelScale=<0.7 0.7 0.7>
    gravityOverride {z=-980}
    physicsBodyDefinitions [{
      boundBone {boneName=dyn_skirt_side_l_01}
      boxExtents=<12 12 12>
      constraintSetup {angularLimitsMax=<0 0 0>, angularLimitsMin=<0 -10 0>, angularTarget=<1 0 0>}
      localJointOffset=<2 0 0>
     }]
    relativeSpaceBone {boneName=root}
   }
   n14 = Anim Dynamics - Bone: dyn_skirt_side_l_02!(SimulationSpace:AnimPhysSimSpaceType Enum=BoneRelative, ComponentPose=@n13, Alpha=1, ExternalForce="0,0,0") -> (Pose) {
    angularDampingOverride=0.99
    angularSpringConstant=3000
    bAngularSpring=true
    bOverrideAngularDamping=true
    bUseGravityOverride=true
    gravityScale=0.35
    simulationSpace=BoneRelative
    alphaBoolBlend {blendOption=Linear}
    boundBone {boneName=dyn_skirt_side_l_02}
    componentAppliedLinearAccClamp=<1000 1000 1000>
    componentLinearAccScale=<0.7 0.7 0.7>
    componentLinearVelScale=<0.7 0.7 0.7>
    gravityOverride {z=-980}
    physicsBodyDefinitions [{
      boundBone {boneName=dyn_skirt_side_l_02}
      boxExtents=<12 12 12>
      constraintSetup {angularLimitsMax=<0 0 0>, angularLimitsMin=<0 -20 0>, angularTarget=<1 0 0>}
      localJointOffset=<2 0 0>
     }]
    relativeSpaceBone {boneName=root}
   }
   n15 = Anim Dynamics - Bone: dyn_skirt_side_l_03!(SimulationSpace:AnimPhysSimSpaceType Enum=BoneRelative, ComponentPose=@n14, Alpha=1, ExternalForce="0,0,0") -> (Pose) {
    angularDampingOverride=0.99
    angularSpringConstant=3000
    bAngularSpring=true
    bOverrideAngularDamping=true
    bUseGravityOverride=true
    gravityScale=0.35
    simulationSpace=BoneRelative
    alphaBoolBlend {blendOption=Linear}
    boundBone {boneName=dyn_skirt_side_l_03}
    componentAppliedLinearAccClamp=<1000 1000 1000>
    componentLinearAccScale=<0.7 0.7 0.7>
    componentLinearVelScale=<0.7 0.7 0.7>
    gravityOverride {z=-980}
    physicsBodyDefinitions [{
      boundBone {boneName=dyn_skirt_side_l_03}
      boxExtents=<12 12 12>
      constraintSetup {angularLimitsMax=<0 0 0>, angularLimitsMin=<0 -50 0>, angularTarget=<1 0 0>}
      localJointOffset=<2 0 0>
     }]
    relativeSpaceBone {boneName=root}
   }
   n16 = Anim Dynamics - Bone: dyn_skirt_side_l_04!(SimulationSpace:AnimPhysSimSpaceType Enum=BoneRelative, ComponentPose=@n15, Alpha=1, ExternalForce="0,0,0") -> (Pose) {
    angularDampingOverride=0.99
    angularSpringConstant=1000
    bAngularSpring=true
    bOverrideAngularDamping=true
    bUseGravityOverride=true
    gravityScale=0.35
    simulationSpace=BoneRelative
    alphaBoolBlend {blendOption=Linear}
    boundBone {boneName=dyn_skirt_side_l_04}
    componentAppliedLinearAccClamp=<1000 1000 1000>
    componentLinearAccScale=<0.7 0.7 0.7>
    componentLinearVelScale=<0.7 0.7 0.7>
    gravityOverride {z=-980}
    physicsBodyDefinitions [{
      boundBone {boneName=dyn_skirt_side_l_04}
      boxExtents=<12 12 12>
      constraintSetup {angularLimitsMax=<0 10 0>, angularLimitsMin=<0 -30 0>, angularTarget=<1 0 0>}
      localJointOffset=<1 0 0>
     }]
    relativeSpaceBone {boneName=root}
   }
   n17 = Anim Dynamics - Bone: dyn_skirt_bck_l_01!(SimulationSpace:AnimPhysSimSpaceType Enum=BoneRelative, ComponentPose=@n16, Alpha=1, ExternalForce="0,0,0") -> (Pose) {
    angularDampingOverride=0.99
    angularSpringConstant=3000
    bAngularSpring=true
    bOverrideAngularDamping=true
    bUseGravityOverride=true
    gravityScale=0.35
    simulationSpace=BoneRelative
    alphaBoolBlend {blendOption=Linear}
    boundBone {boneName=dyn_skirt_bck_l_01}
    componentAppliedLinearAccClamp=<1000 1000 1000>
    componentLinearAccScale=<0.7 0.7 0.7>
    componentLinearVelScale=<0.7 0.7 0.7>
    gravityOverride {z=-980}
    physicsBodyDefinitions [{
      boundBone {boneName=dyn_skirt_bck_l_01}
      boxExtents=<12 12 12>
      constraintSetup {angularLimitsMax=<0 0 0>, angularLimitsMin=<0 -10 0>, angularTarget=<1 0 0>}
      localJointOffset=<2 0 0>
     }]
    relativeSpaceBone {boneName=root}
   }
   n18 = Anim Dynamics - Bone: dyn_skirt_bck_l_02!(SimulationSpace:AnimPhysSimSpaceType Enum=BoneRelative, ComponentPose=@n17, Alpha=1, ExternalForce="0,0,0") -> (Pose) {
    angularDampingOverride=0.99
    angularSpringConstant=3000
    bAngularSpring=true
    bOverrideAngularDamping=true
    bUseGravityOverride=true
    gravityScale=0.35
    simulationSpace=BoneRelative
    alphaBoolBlend {blendOption=Linear}
    boundBone {boneName=dyn_skirt_bck_l_02}
    componentAppliedLinearAccClamp=<1000 1000 1000>
    componentLinearAccScale=<0.7 0.7 0.7>
    componentLinearVelScale=<0.7 0.7 0.7>
    gravityOverride {z=-980}
    physicsBodyDefinitions [{
      boundBone {boneName=dyn_skirt_bck_l_02}
      boxExtents=<12 12 12>
      constraintSetup {angularLimitsMax=<0 0 0>, angularLimitsMin=<0 -20 0>, angularTarget=<1 0 0>}
      localJointOffset=<2 0 0>
     }]
    relativeSpaceBone {boneName=root}
   }
   n19 = Anim Dynamics - Bone: dyn_skirt_bck_l_03!(SimulationSpace:AnimPhysSimSpaceType Enum=BoneRelative, ComponentPose=@n18, Alpha=1, ExternalForce="0,0,0") -> (Pose) {
    angularDampingOverride=0.99
    angularSpringConstant=3000
    bAngularSpring=true
    bOverrideAngularDamping=true
    bUseGravityOverride=true
    gravityScale=0.35
    simulationSpace=BoneRelative
    alphaBoolBlend {blendOption=Linear}
    boundBone {boneName=dyn_skirt_bck_l_03}
    componentAppliedLinearAccClamp=<1000 1000 1000>
    componentLinearAccScale=<0.7 0.7 0.7>
    componentLinearVelScale=<0.7 0.7 0.7>
    gravityOverride {z=-980}
    physicsBodyDefinitions [{
      boundBone {boneName=dyn_skirt_bck_l_03}
      boxExtents=<12 12 12>
      constraintSetup {angularLimitsMax=<0 0 0>, angularLimitsMin=<0 -50 0>, angularTarget=<1 0 0>}
      localJointOffset=<2 0 0>
     }]
    relativeSpaceBone {boneName=root}
   }
   n20 = Anim Dynamics - Bone: dyn_skirt_bck_l_04!(SimulationSpace:AnimPhysSimSpaceType Enum=BoneRelative, ComponentPose=@n19, Alpha=1, ExternalForce="0,0,0") -> (Pose) {
    angularDampingOverride=0.99
    angularSpringConstant=1000
    bAngularSpring=true
    bOverrideAngularDamping=true
    bUseGravityOverride=true
    gravityScale=0.35
    simulationSpace=BoneRelative
    alphaBoolBlend {blendOption=Linear}
    boundBone {boneName=dyn_skirt_bck_l_04}
    componentAppliedLinearAccClamp=<1000 1000 1000>
    componentLinearAccScale=<0.7 0.7 0.7>
    componentLinearVelScale=<0.7 0.7 0.7>
    gravityOverride {z=-980}
    physicsBodyDefinitions [{
      boundBone {boneName=dyn_skirt_bck_l_04}
      boxExtents=<12 12 12>
      constraintSetup {angularLimitsMax=<0 10 0>, angularLimitsMin=<0 -35 0>, angularTarget=<1 0 0>}
      localJointOffset=<1 0 0>
     }]
    relativeSpaceBone {boneName=root}
   }
   n21 = Anim Dynamics - Bone: dyn_skirt_bck_c_01!(SimulationSpace:AnimPhysSimSpaceType Enum=BoneRelative, ComponentPose=@n20, Alpha=1, ExternalForce="0,0,0") -> (Pose) {
    angularDampingOverride=0.99
    angularSpringConstant=3000
    bAngularSpring=true
    bOverrideAngularDamping=true
    bUseGravityOverride=true
    gravityScale=0.35
    simulationSpace=BoneRelative
    alphaBoolBlend {blendOption=Linear}
    boundBone {boneName=dyn_skirt_bck_c_01}
    componentAppliedLinearAccClamp=<1000 1000 1000>
    componentLinearAccScale=<0.7 0.7 0.7>
    componentLinearVelScale=<0.7 0.7 0.7>
    gravityOverride {z=-980}
    physicsBodyDefinitions [{
      boundBone {boneName=dyn_skirt_bck_c_01}
      boxExtents=<12 12 12>
      constraintSetup {angularLimitsMax=<0 0 0>, angularLimitsMin=<0 -10 0>, angularTarget=<1 0 0>}
      localJointOffset=<2 0 0>
     }]
    relativeSpaceBone {boneName=root}
   }
   n22 = Anim Dynamics - Bone: dyn_skirt_bck_c_02!(SimulationSpace:AnimPhysSimSpaceType Enum=BoneRelative, ComponentPose=@n21, Alpha=1, ExternalForce="0,0,0") -> (Pose) {
    angularDampingOverride=0.99
    angularSpringConstant=3000
    bAngularSpring=true
    bOverrideAngularDamping=true
    bUseGravityOverride=true
    gravityScale=0.35
    simulationSpace=BoneRelative
    alphaBoolBlend {blendOption=Linear}
    boundBone {boneName=dyn_skirt_bck_c_02}
    componentAppliedLinearAccClamp=<1000 1000 1000>
    componentLinearAccScale=<0.7 0.7 0.7>
    componentLinearVelScale=<0.7 0.7 0.7>
    gravityOverride {z=-980}
    physicsBodyDefinitions [{
      boundBone {boneName=dyn_skirt_bck_c_02}
      boxExtents=<12 12 12>
      constraintSetup {angularLimitsMax=<0 0 0>, angularLimitsMin=<0 -20 0>, angularTarget=<1 0 0>}
      localJointOffset=<2 0 0>
     }]
    relativeSpaceBone {boneName=root}
   }
   n23 = Anim Dynamics - Bone: dyn_skirt_bck_c_03!(SimulationSpace:AnimPhysSimSpaceType Enum=BoneRelative, ComponentPose=@n22, Alpha=1, ExternalForce="0,0,0") -> (Pose) {
    angularDampingOverride=0.99
    angularSpringConstant=3000
    bAngularSpring=true
    bOverrideAngularDamping=true
    bUseGravityOverride=true
    gravityScale=0.35
    simulationSpace=BoneRelative
    alphaBoolBlend {blendOption=Linear}
    boundBone {boneName=dyn_skirt_bck_c_03}
    componentAppliedLinearAccClamp=<1000 1000 1000>
    componentLinearAccScale=<0.7 0.7 0.7>
    componentLinearVelScale=<0.7 0.7 0.7>
    gravityOverride {z=-980}
    physicsBodyDefinitions [{
      boundBone {boneName=dyn_skirt_bck_c_03}
      boxExtents=<12 12 12>
      constraintSetup {angularLimitsMax=<0 0 0>, angularLimitsMin=<0 -50 0>, angularTarget=<1 0 0>}
      localJointOffset=<2 0 0>
     }]
    relativeSpaceBone {boneName=root}
   }
   n24 = Anim Dynamics - Bone: dyn_skirt_bck_c_04!(SimulationSpace:AnimPhysSimSpaceType Enum=BoneRelative, ComponentPose=@n23, Alpha=1, ExternalForce="0,0,0") -> (Pose) {
    angularDampingOverride=0.99
    angularSpringConstant=1000
    bAngularSpring=true
    bOverrideAngularDamping=true
    bUseGravityOverride=true
    gravityScale=0.35
    simulationSpace=BoneRelative
    alphaBoolBlend {blendOption=Linear}
    boundBone {boneName=dyn_skirt_bck_c_04}
    componentAppliedLinearAccClamp=<1000 1000 1000>
    componentLinearAccScale=<0.5 0.5 0.5>
    componentLinearVelScale=<0.5 0.5 0.5>
    gravityOverride {z=-980}
    physicsBodyDefinitions [{
      boundBone {boneName=dyn_skirt_bck_c_04}
      boxExtents=<12 12 12>
      constraintSetup {angularLimitsMax=<0 10 0>, angularLimitsMin=<0 -35 0>, angularTarget=<1 0 0>}
      localJointOffset=<1 0 0>
     }]
    relativeSpaceBone {boneName=root}
   }
   n25 = Anim Dynamics - Bone: dyn_skirt_bck_r_01!(SimulationSpace:AnimPhysSimSpaceType Enum=BoneRelative, ComponentPose=@n24, Alpha=1, ExternalForce="0,0,0") -> (Pose) {
    angularDampingOverride=0.99
    angularSpringConstant=3000
    bAngularSpring=true
    bOverrideAngularDamping=true
    bUseGravityOverride=true
    gravityScale=0.35
    simulationSpace=BoneRelative
    alphaBoolBlend {blendOption=Linear}
    boundBone {boneName=dyn_skirt_bck_r_01}
    componentAppliedLinearAccClamp=<1000 1000 1000>
    componentLinearAccScale=<0.7 0.7 0.7>
    componentLinearVelScale=<0.7 0.7 0.7>
    gravityOverride {z=-980}
    physicsBodyDefinitions [{
      boundBone {boneName=dyn_skirt_bck_r_01}
      boxExtents=<12 12 12>
      constraintSetup {angularLimitsMax=<0 0 0>, angularLimitsMin=<0 -10 0>, angularTarget=<1 0 0>}
      localJointOffset=<2 0 0>
     }]
    relativeSpaceBone {boneName=root}
   }
   n26 = Anim Dynamics - Bone: dyn_skirt_bck_r_02!(SimulationSpace:AnimPhysSimSpaceType Enum=BoneRelative, ComponentPose=@n25, Alpha=1, ExternalForce="0,0,0") -> (Pose) {
    angularDampingOverride=0.99
    angularSpringConstant=3000
    bAngularSpring=true
    bOverrideAngularDamping=true
    bUseGravityOverride=true
    gravityScale=0.35
    simulationSpace=BoneRelative
    alphaBoolBlend {blendOption=Linear}
    boundBone {boneName=dyn_skirt_bck_r_02}
    componentAppliedLinearAccClamp=<1000 1000 1000>
    componentLinearAccScale=<0.7 0.7 0.7>
    componentLinearVelScale=<0.7 0.7 0.7>
    gravityOverride {z=-980}
    physicsBodyDefinitions [{
      boundBone {boneName=dyn_skirt_bck_r_02}
      boxExtents=<12 12 12>
      constraintSetup {angularLimitsMax=<0 0 0>, angularLimitsMin=<0 -20 0>, angularTarget=<1 0 0>}
      localJointOffset=<2 0 0>
     }]
    relativeSpaceBone {boneName=root}
   }
   n27 = Anim Dynamics - Bone: dyn_skirt_bck_r_03!(SimulationSpace:AnimPhysSimSpaceType Enum=BoneRelative, ComponentPose=@n26, Alpha=1, ExternalForce="0,0,0") -> (Pose) {
    angularDampingOverride=0.99
    angularSpringConstant=3000
    bAngularSpring=true
    bOverrideAngularDamping=true
    bUseGravityOverride=true
    gravityScale=0.35
    simulationSpace=BoneRelative
    alphaBoolBlend {blendOption=Linear}
    boundBone {boneName=dyn_skirt_bck_r_03}
    componentAppliedLinearAccClamp=<1000 1000 1000>
    componentLinearAccScale=<0.7 0.7 0.7>
    componentLinearVelScale=<0.7 0.7 0.7>
    gravityOverride {z=-980}
    physicsBodyDefinitions [{
      boundBone {boneName=dyn_skirt_bck_r_03}
      boxExtents=<12 12 12>
      constraintSetup {angularLimitsMax=<0 0 0>, angularLimitsMin=<0 -50 0>, angularTarget=<1 0 0>}
      localJointOffset=<2 0 0>
     }]
    relativeSpaceBone {boneName=root}
   }
   n28 = Anim Dynamics - Bone: dyn_skirt_bck_r_04!(SimulationSpace:AnimPhysSimSpaceType Enum=BoneRelative, ComponentPose=@n27, Alpha=1, ExternalForce="0,0,0") -> (Pose) {
    angularDampingOverride=0.99
    angularSpringConstant=1000
    bAngularSpring=true
    bOverrideAngularDamping=true
    bUseGravityOverride=true
    gravityScale=0.35
    simulationSpace=BoneRelative
    alphaBoolBlend {blendOption=Linear}
    boundBone {boneName=dyn_skirt_bck_r_04}
    componentAppliedLinearAccClamp=<1000 1000 1000>
    componentLinearAccScale=<0.5 0.5 0.5>
    componentLinearVelScale=<0.5 0.5 0.5>
    gravityOverride {z=-980}
    physicsBodyDefinitions [{
      boundBone {boneName=dyn_skirt_bck_r_04}
      boxExtents=<12 12 12>
      constraintSetup {angularLimitsMax=<0 10 0>, angularLimitsMin=<0 -35 0>, angularTarget=<1 0 0>}
      localJointOffset=<1 0 0>
     }]
    relativeSpaceBone {boneName=root}
   }
   n29 = Anim Dynamics - Bone: dyn_skirt_side_r_01!(SimulationSpace:AnimPhysSimSpaceType Enum=BoneRelative, ComponentPose=@n28, Alpha=1, ExternalForce="0,0,0") -> (Pose) {
    angularDampingOverride=0.99
    angularSpringConstant=3000
    bAngularSpring=true
    bOverrideAngularDamping=true
    bUseGravityOverride=true
    gravityScale=0.35
    simulationSpace=BoneRelative
    alphaBoolBlend {blendOption=Linear}
    boundBone {boneName=dyn_skirt_side_r_01}
    componentAppliedLinearAccClamp=<1000 1000 1000>
    componentLinearAccScale=<0.7 0.7 0.7>
    componentLinearVelScale=<0.7 0.7 0.7>
    gravityOverride {z=-980}
    physicsBodyDefinitions [{
      boundBone {boneName=dyn_skirt_side_r_01}
      boxExtents=<12 12 12>
      constraintSetup {angularLimitsMax=<0 0 0>, angularLimitsMin=<0 -10 0>, angularTarget=<1 0 0>}
      localJointOffset=<2 0 0>
     }]
    relativeSpaceBone {boneName=root}
   }
   n30 = Anim Dynamics - Bone: dyn_skirt_side_r_02!(SimulationSpace:AnimPhysSimSpaceType Enum=BoneRelative, ComponentPose=@n29, Alpha=1, ExternalForce="0,0,0") -> (Pose) {
    angularDampingOverride=0.99
    angularSpringConstant=3000
    bAngularSpring=true
    bOverrideAngularDamping=true
    bUseGravityOverride=true
    gravityScale=0.35
    simulationSpace=BoneRelative
    alphaBoolBlend {blendOption=Linear}
    boundBone {boneName=dyn_skirt_side_r_02}
    componentAppliedLinearAccClamp=<1000 1000 1000>
    componentLinearAccScale=<0.7 0.7 0.7>
    componentLinearVelScale=<0.7 0.7 0.7>
    gravityOverride {z=-980}
    physicsBodyDefinitions [{
      boundBone {boneName=dyn_skirt_side_r_02}
      boxExtents=<12 12 12>
      constraintSetup {angularLimitsMax=<0 0 0>, angularLimitsMin=<0 -20 0>, angularTarget=<1 0 0>}
      localJointOffset=<2 0 0>
     }]
    relativeSpaceBone {boneName=root}
   }
   n31 = Anim Dynamics - Bone: dyn_skirt_side_r_03!(SimulationSpace:AnimPhysSimSpaceType Enum=BoneRelative, ComponentPose=@n30, Alpha=1, ExternalForce="0,0,0") -> (Pose) {
    angularDampingOverride=0.99
    angularSpringConstant=3000
    bAngularSpring=true
    bOverrideAngularDamping=true
    bUseGravityOverride=true
    gravityScale=0.35
    simulationSpace=BoneRelative
    alphaBoolBlend {blendOption=Linear}
    boundBone {boneName=dyn_skirt_side_r_03}
    componentAppliedLinearAccClamp=<1000 1000 1000>
    componentLinearAccScale=<0.7 0.7 0.7>
    componentLinearVelScale=<0.7 0.7 0.7>
    gravityOverride {z=-980}
    physicsBodyDefinitions [{
      boundBone {boneName=dyn_skirt_side_r_03}
      boxExtents=<12 12 12>
      constraintSetup {angularLimitsMax=<0 0 0>, angularLimitsMin=<0 -50 0>, angularTarget=<1 0 0>}
      localJointOffset=<2 0 0>
     }]
    relativeSpaceBone {boneName=root}
   }
   n32 = Anim Dynamics - Bone: dyn_skirt_side_r_04!(SimulationSpace:AnimPhysSimSpaceType Enum=BoneRelative, ComponentPose=@n31, Alpha=1, ExternalForce="0,0,0") -> (Pose) {
    angularDampingOverride=0.99
    angularSpringConstant=1000
    bAngularSpring=true
    bOverrideAngularDamping=true
    bUseGravityOverride=true
    gravityScale=0.35
    simulationSpace=BoneRelative
    alphaBoolBlend {blendOption=Linear}
    boundBone {boneName=dyn_skirt_side_r_04}
    componentAppliedLinearAccClamp=<1000 1000 1000>
    componentLinearAccScale=<0.5 0.5 0.5>
    componentLinearVelScale=<0.5 0.5 0.5>
    gravityOverride {z=-980}
    physicsBodyDefinitions [{
      boundBone {boneName=dyn_skirt_side_r_04}
      boxExtents=<12 12 12>
      constraintSetup {angularLimitsMax=<0 10 0>, angularLimitsMin=<0 -30 0>, angularTarget=<1 0 0>}
      localJointOffset=<1 0 0>
     }]
    relativeSpaceBone {boneName=root}
   }
   n33 = Anim Dynamics - Bone: dyn_skirt_fr_r_01!(SimulationSpace:AnimPhysSimSpaceType Enum=BoneRelative, ComponentPose=@n32, Alpha=1, ExternalForce="0,0,0") -> (Pose) {
    angularDampingOverride=0.99
    angularSpringConstant=3000
    bAngularSpring=true
    bOverrideAngularDamping=true
    bUseGravityOverride=true
    gravityScale=0.35
    simulationSpace=BoneRelative
    alphaBoolBlend {blendOption=Linear}
    boundBone {boneName=dyn_skirt_fr_r_01}
    componentAppliedLinearAccClamp=<1000 1000 1000>
    componentLinearAccScale=<0.7 0.7 0.7>
    componentLinearVelScale=<0.7 0.7 0.7>
    gravityOverride {z=-980}
    physicsBodyDefinitions [{
      boundBone {boneName=dyn_skirt_fr_r_01}
      boxExtents=<12 12 12>
      constraintSetup {angularLimitsMax=<0 0 0>, angularLimitsMin=<0 -10 0>, angularTarget=<1 0 0>}
      localJointOffset=<2 0 0>
     }]
    relativeSpaceBone {boneName=root}
   }
   n34 = Anim Dynamics - Bone: dyn_skirt_fr_r_02!(SimulationSpace:AnimPhysSimSpaceType Enum=BoneRelative, ComponentPose=@n33, Alpha=1, ExternalForce="0,0,0") -> (Pose) {
    angularDampingOverride=0.99
    angularSpringConstant=3000
    bAngularSpring=true
    bOverrideAngularDamping=true
    bUseGravityOverride=true
    gravityScale=0.35
    simulationSpace=BoneRelative
    alphaBoolBlend {blendOption=Linear}
    boundBone {boneName=dyn_skirt_fr_r_02}
    componentAppliedLinearAccClamp=<1000 1000 1000>
    componentLinearAccScale=<0.7 0.7 0.7>
    componentLinearVelScale=<0.7 0.7 0.7>
    gravityOverride {z=-980}
    physicsBodyDefinitions [{
      boundBone {boneName=dyn_skirt_fr_r_02}
      boxExtents=<12 12 12>
      constraintSetup {angularLimitsMax=<0 0 0>, angularLimitsMin=<0 -20 0>, angularTarget=<1 0 0>}
      localJointOffset=<2 0 0>
     }]
    relativeSpaceBone {boneName=root}
   }
   n35 = Anim Dynamics - Bone: dyn_skirt_fr_r_03!(SimulationSpace:AnimPhysSimSpaceType Enum=BoneRelative, ComponentPose=@n34, Alpha=1, ExternalForce="0,0,0") -> (Pose) {
    angularDampingOverride=0.99
    angularSpringConstant=3000
    bAngularSpring=true
    bOverrideAngularDamping=true
    bUseGravityOverride=true
    gravityScale=0.35
    simulationSpace=BoneRelative
    alphaBoolBlend {blendOption=Linear}
    boundBone {boneName=dyn_skirt_fr_r_03}
    componentAppliedLinearAccClamp=<1000 1000 1000>
    componentLinearAccScale=<0.7 0.7 0.7>
    componentLinearVelScale=<0.7 0.7 0.7>
    gravityOverride {z=-980}
    physicsBodyDefinitions [{
      boundBone {boneName=dyn_skirt_fr_r_03}
      boxExtents=<12 12 12>
      constraintSetup {angularLimitsMax=<0 0 0>, angularLimitsMin=<0 -50 0>, angularTarget=<1 0 0>}
      localJointOffset=<2 0 0>
     }]
    relativeSpaceBone {boneName=root}
   }
   n36 = Anim Dynamics - Bone: dyn_skirt_fr_r_04!(SimulationSpace:AnimPhysSimSpaceType Enum=BoneRelative, ComponentPose=@n35, Alpha=1, ExternalForce="0,0,0") -> (Pose) {
    angularDampingOverride=0.99
    angularSpringConstant=1000
    bAngularSpring=true
    bOverrideAngularDamping=true
    bUseGravityOverride=true
    gravityScale=0.35
    simulationSpace=BoneRelative
    alphaBoolBlend {blendOption=Linear}
    boundBone {boneName=dyn_skirt_fr_r_04}
    componentAppliedLinearAccClamp=<1000 1000 1000>
    componentLinearAccScale=<0.5 0.5 0.5>
    componentLinearVelScale=<0.5 0.5 0.5>
    gravityOverride {z=-980}
    physicsBodyDefinitions [{
      boundBone {boneName=dyn_skirt_fr_r_04}
      boxExtents=<12 12 12>
      constraintSetup {angularLimitsMax=<0 0 0>, angularLimitsMin=<0 -30 0>, angularTarget=<1 0 0>}
      localJointOffset=<1 0 0>
     }]
    relativeSpaceBone {boneName=root}
   }
   Save cached pose 'Dynamics_cache'(Pose=Component To Local(ComponentPose=@n36)) {cacheName=Dynamics_cache}
  ]}
 EventGraph {nodes [
   Event OnCharacterCustomizationLoaded()
   > Set Is Wearing Backbling(Is Wearing Backbling=CheckForBackbling())
   > Set Cape Socket Child(Cape Socket Child:Skeletal Mesh Component Object Reference=FindChildCommponentFromAttachSocket!(
     self:Custom Character Part Anim Instance Object Reference
     Parent:Skeletal Mesh Component Object Reference=GetPartSkeletalMeshComponent()
     bIncludeAllDescendants:bool=false
     AttachSocket:Name=cap
    ) -> (ReturnValue:Skeletal Mesh Component Object Reference)) -> (Output_Get:Skeletal Mesh Component Object Reference)
   > Set IsTheCape(IsTheCape:bool=BackpackItemDefinitionHasAnyMetaTag!(Tags:Array of Gameplay Tag Structures=Make Array("[0]"=(TagName="Cosmetics.Metatags.TrickCrate")), bRequiresEquipped:bool=false) -> (ReturnValue:bool)) -> (Output_Get:bool)
   Event BlueprintUpdateAnimation()
  ]}
]
variables [
 Cape Socket Child {type=SkeletalMeshComponent}
 Is Wearing Backbling {type=boolean}
 IsTheCape {type=boolean}
 Simulation Space {type=AnimPhysSimSpaceType, default=BoneRelative}
 SkirtAlpha {type=number, default=1}
]
