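# Physics asset for the female medium TrickCrate no-cape FaceAcc mesh, giving physics bodies and joint limits to its hair.
# Left, right and center hair chains and two front locks hang from base bodies constrained to the head, with limited swing and tight or locked twist.
# Free-rotation constraints link the center chain near its tip to both side chains.
# Capsules on head, neck, spine, clavicles and upper arms form the body volumes around the hair.
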PhysicsAsset /BRCosmetics/Characters/Player/Female/Medium/Bodies/F_MED_TrickCrate/Meshes/Parts/F_MED_NoCape_TrickCrate_FaceAcc_Physics
bodies {
 base_hair_02_cr [{shapeType=Capsule, length=0.505, radius=0.505, shapeName=base_hair_02_cr_capsule, rotation=<rot 0 -90 0>}]
 base_hair_02_lf [{shapeType=Capsule, length=0.505, radius=0.505, shapeName=base_hair_02_lf_capsule, rotation=<rot 0 -90 0>}]
 base_hair_02_rt [{shapeType=Capsule, length=0.505, radius=0.505, shapeName=base_hair_02_rt_capsule, rotation=<rot 0 -90 0>}]
 base_lock_fr_l_03 [{shapeType=Capsule, length=0.505, radius=0.505, shapeName=base_lock_fr_l_03_capsule, rotation=<rot 0 -90 0>}]
 base_lock_fr_r_03 [{shapeType=Capsule, length=0.505, radius=0.505, shapeName=base_lock_fr_r_03_capsule, rotation=<rot 0 -90 0>}]
 clavicle_l [{shapeType=Capsule, length=5.6736, radius=6.1054, shapeName=clavicle_l_capsule, center=<6.9869 -1.3802 -0.8969>, rotation=<rot 90 0 0>}]
 clavicle_r [{shapeType=Capsule, length=3.5389, radius=5.6622, shapeName=clavicle_l_capsule, center=<-9.3665 0.3902 0.7413>, rotation=<rot 90 0 0>}]
 dyn_hair_02_cr [{shapeType=Capsule, length=4, radius=3.5, shapeName=dyn_hair_02_rt_capsule, center=<3.8266 0 0.8136>, rotation=<rot 72.5739 0 0>}]
 dyn_hair_02_lf [{shapeType=Capsule, length=12.107, radius=3.5, shapeName=dyn_hair_02_rt_capsule, center=<6.0535 0 3.2819>, rotation=<rot 90 0 0>}]
 dyn_hair_02_rt [{shapeType=Capsule, length=7.724, radius=3.5, shapeName=dyn_hair_02_rt_capsule, center=<3.862 -0.694 2.7831>, rotation=<rot 90 0 0>}]
 dyn_hair_03_cr [{shapeType=Capsule, length=7.384, radius=3.5, shapeName=dyn_hair_02_rt_capsule, center=<3.692 0 0>, rotation=<rot 90 0 0>}]
 dyn_hair_03_lf [{shapeType=Capsule, length=8.177, radius=3.5, shapeName=dyn_hair_02_rt_capsule, center=<4.0885 1.5656 3.4931>, rotation=<rot 90 0 0>}]
 dyn_hair_03_rt [{shapeType=Capsule, length=6.352, radius=3.5, shapeName=dyn_hair_02_rt_capsule, center=<3.176 -1.6763 2.6982>, rotation=<rot 90 0 0>}]
 dyn_hair_04_cr [{shapeType=Capsule, length=7.485, radius=3.5, shapeName=dyn_hair_04_cr_capsule, center=<3.7425 0 0>, rotation=<rot 90 0 0>}]
 dyn_hair_04_lf [{shapeType=Capsule, length=6.385, radius=3.5, shapeName=dyn_hair_04_cr_capsule, center=<3.1925 1.5272 3.4737>, rotation=<rot 90 0 0>}]
 dyn_hair_04_rt [{shapeType=Capsule, length=6.846, radius=3.5, shapeName=dyn_hair_04_cr_capsule, center=<3.423 -1.9687 2.7283>, rotation=<rot 90 0 0>}]
 dyn_hair_05_cr [{shapeType=Capsule, length=7.13, radius=3.5, shapeName=dyn_hair_05_cr_capsule, center=<3.565 0 0>, rotation=<rot 90 0 0>}]
 dyn_hair_05_lf [{shapeType=Capsule, length=7.286, radius=3.5, shapeName=dyn_hair_05_cr_capsule, center=<3.643 -0.7161 3.5065>, rotation=<rot 90 0 0>}]
 dyn_hair_05_rt [{shapeType=Capsule, length=7.264, radius=3.5, shapeName=dyn_hair_05_cr_capsule, center=<1.816 2.0712 2.7855>, rotation=<rot 90 0 0>}]
 dyn_hair_06_cr [{shapeType=Capsule, length=6.59, radius=3.5, shapeName=dyn_hair_05_cr_capsule, center=<3.295 0 0>, rotation=<rot 90 0 0>}]
 dyn_hair_06_lf [{shapeType=Capsule, length=3.349, radius=3.5, shapeName=dyn_hair_05_cr_capsule, center=<1.6745 -2.6041 3.4746>, rotation=<rot 90 0 0>}]
 dyn_hair_06_rt [{shapeType=Capsule, length=4.535, radius=3.5, shapeName=dyn_hair_05_cr_capsule, center=<2.2675 3.1347 3.0021>, rotation=<rot 90 0 0>}]
 dyn_hair_07_cr [{shapeType=Capsule, length=6.0665, radius=4.4335, shapeName=dyn_hair_05_cr_capsule, center=<5.3202 0 2.4691>, rotation=<rot 90 0 0>}]
 dyn_hair_07_lf [{shapeType=Capsule, length=3, radius=3.5, shapeName=dyn_hair_05_cr_capsule, center=<1.5 -2.9979 3.6947>, rotation=<rot 90 0 0>}]
 dyn_hair_07_rt [{shapeType=Capsule, length=3, radius=3.5, shapeName=dyn_hair_05_cr_capsule, center=<1.5 4.2529 2.631>, rotation=<rot 90 0 0>}]
 dyn_hair_08_cr [{shapeType=Capsule, length=5.9294, radius=4.5706, shapeName=dyn_hair_05_cr_capsule, center=<7.8064 0 2.9342>, rotation=<rot 90 0 0>}]
 dyn_lock_fr_l_03 [{shapeType=Capsule, length=2.8305, radius=0.7737, shapeName=dyn_hair_06_cr_capsule, center=<1.4 0.0365 0>, rotation=<rot 90 0 0>}]
 dyn_lock_fr_l_04 [{shapeType=Capsule, length=2.7215, radius=0.9027, shapeName=dyn_hair_06_cr_capsule, center=<1.41 0 0>, rotation=<rot 90 0 0>}]
 dyn_lock_fr_l_05 [{shapeType=Capsule, length=2.254, radius=0.9502, shapeName=dyn_hair_06_cr_capsule, center=<1.2 0 0>, rotation=<rot 90 0 0>}]
 dyn_lock_fr_l_06 [{shapeType=Capsule, length=2.6767, radius=1.0855, shapeName=dyn_hair_06_cr_capsule, center=<1.479 0 0>, rotation=<rot 90 0 0>}]
 dyn_lock_fr_l_07 [{shapeType=Capsule, length=2.6767, radius=1.0855, shapeName=dyn_hair_06_cr_capsule, center=<1.479 0 0>, rotation=<rot 90 0 0>}]
 dyn_lock_fr_r_03 [{shapeType=Capsule, length=2.2089, radius=1.3953, shapeName=dyn_hair_06_cr_capsule, center=<1.3092 -0.2297 0.0728>, rotation=<rot 90 0 0>}]
 dyn_lock_fr_r_04 [{shapeType=Capsule, length=2.0424, radius=1.5617, shapeName=dyn_hair_06_cr_capsule, center=<1.3092 0.0543 0.0728>, rotation=<rot 90 5.8696 0>}]
 dyn_lock_fr_r_05 [{shapeType=Capsule, length=2.3201, radius=1.2841, shapeName=dyn_hair_06_cr_capsule, center=<1.3092 0.0543 0.1422>, rotation=<rot 90 0 0>}]
 dyn_lock_fr_r_06 [{shapeType=Capsule, length=2.3009, radius=1.3032, shapeName=dyn_hair_06_cr_capsule, center=<1.3092 0.0543 0.0728>, rotation=<rot 90 0 0>}]
 dyn_lock_fr_r_07 [{shapeType=Capsule, length=2.5001, radius=1.1041, shapeName=dyn_hair_06_cr_capsule, center=<1.3092 0.0543 0.0728>, rotation=<rot 90 0 0>}]
 head [
  {shapeType=Capsule, length=6.0812, radius=4.9092, shapeName=head_capsule, center=<2.5333 1.3001 0.2337>, rotation=<rot 0 -74.2468 90>}
  {
   shapeType=Capsule
   length=13.1164
   radius=1.8976
   shapeName=head_capsule
   center=<-4.2936 -1.8136 10.6455>
   rotation=<rot -11.2175 9.7203 120.1539>
  }
  {
   shapeType=Capsule
   length=13.0078
   radius=1.8964
   shapeName=head_capsule
   center=<-1.0936 -1.0635 10.8532>
   rotation=<rot -10.9505 11.4194 119.676>
  }
  {
   shapeType=Capsule
   length=13.0078
   radius=1.8964
   shapeName=head_capsule
   center=<2.3036 0.1069 11.0124>
   rotation=<rot -10.9505 11.4194 119.676>
  }
  {
   shapeType=Capsule
   length=13.0078
   radius=1.8964
   shapeName=head_capsule
   center=<2.3036 0.1069 -11.0124>
   rotation=<rot 10.9505 11.4194 -119.676>
  }
  {
   shapeType=Capsule
   length=13.0078
   radius=1.8964
   shapeName=head_capsule
   center=<-1.0936 -1.0635 -10.8532>
   rotation=<rot 10.9505 11.4194 -119.676>
  }
  {
   shapeType=Capsule
   length=13.1164
   radius=1.8976
   shapeName=head_capsule
   center=<-4.2936 -1.8136 -10.6455>
   rotation=<rot 11.2175 9.7203 -120.1539>
  }
  {shapeType=Capsule, length=4.8798, radius=6.1106, shapeName=head_capsule, center=<2.5855 5.4787 0.2337>, rotation=<rot 0 -74.2468 90>}
 ]
 neck_01 [{shapeType=Capsule, length=4.3636, radius=3.6364, shapeName=neck_01_capsule, center=<3 0 0>, rotation=<rot 90 0 0>}]
 neck_02 [{shapeType=Capsule, length=2.7128, radius=5.2872, shapeName=neck_01_capsule, center=<6.8712 0 0.139>, rotation=<rot 90 0 0>}]
 spine_01 [{shapeType=Capsule, length=1.7054, radius=13.4393, shapeName=spine_01_capsule, center=<-3.9777 -0.1407 0>, rotation=<rot 90 23.6731 0>}]
 spine_02 [{shapeType=Capsule, length=0.5807, radius=10.0802, shapeName=spine_02_capsule, center=<-1.5174 1.2345 0>, rotation=<rot 90 11.5144 0>}]
 spine_03 [{
   shapeType=Capsule
   length=13.5311
   radius=14.8787
   shapeName=spine_02_capsule
   center=<2.7003 1.5259 -0.0268>
   rotation=<rot -0.4577 -176.4345 -179.9362>
  }]
 spine_04 [{
   shapeType=Capsule
   length=15.4198
   radius=12.9901
   shapeName=spine_02_capsule
   center=<1.4957 -0.1194 -0.0389>
   rotation=<rot -0.4595 -174.6641 -179.9504>
  }]
 spine_05 [{shapeType=Capsule, length=22.6277, radius=7.0204, shapeName=spine_05_capsule, center=<3.8919 -2.4682 0>, rotation=<rot 0 -90 0>}]
 upperarm_l [{shapeType=Capsule, length=24.4316, radius=5.8184, shapeName=upperarm_l_capsule, center=<13 0 0>, rotation=<rot 90 0 0>}]
 upperarm_r [{shapeType=Capsule, length=23.8794, radius=6.3705, shapeName=upperarm_r_capsule, center=<-13 0 0>, rotation=<rot 90 0 0>}]
}
constraints [
 {
  bone1Name=neck_02
  bone2Name=neck_01
  swing1LimitDegrees=45
  swing1Motion=Free
  swing2LimitDegrees=45
  swing2Motion=Free
  twistLimitDegrees=45
  twistMotion=Free
 }
 {
  bone1Name=neck_01
  bone2Name=spine_05
  swing1LimitDegrees=45
  swing1Motion=Free
  swing2LimitDegrees=45
  swing2Motion=Free
  twistLimitDegrees=45
  twistMotion=Free
 }
 {
  bone1Name=head
  bone2Name=neck_02
  swing1LimitDegrees=45
  swing1Motion=Free
  swing2LimitDegrees=45
  swing2Motion=Free
  twistLimitDegrees=45
  twistMotion=Free
 }
 {
  bone1Name=dyn_lock_fr_l_03
  bone2Name=base_lock_fr_l_03
  swing1LimitDegrees=3
  swing2LimitDegrees=15
  twistLimitDegrees=3
  twistMotion=Locked
 }
 {bone1Name=dyn_lock_fr_l_04, bone2Name=dyn_lock_fr_l_03, swing1LimitDegrees=3, swing2LimitDegrees=15, twistLimitDegrees=3, twistMotion=Locked}
 {bone1Name=dyn_lock_fr_l_05, bone2Name=dyn_lock_fr_l_04, swing1LimitDegrees=3, swing2LimitDegrees=15, twistLimitDegrees=3, twistMotion=Locked}
 {bone1Name=dyn_lock_fr_l_06, bone2Name=dyn_lock_fr_l_05, swing1LimitDegrees=3, swing2LimitDegrees=15, twistLimitDegrees=3, twistMotion=Locked}
 {bone1Name=dyn_lock_fr_l_07, bone2Name=dyn_lock_fr_l_06, swing1LimitDegrees=3, swing2LimitDegrees=15, twistLimitDegrees=3, twistMotion=Locked}
 {
  bone1Name=dyn_lock_fr_r_03
  bone2Name=base_lock_fr_r_03
  swing1LimitDegrees=3
  swing2LimitDegrees=15
  twistLimitDegrees=3
  twistMotion=Locked
 }
 {bone1Name=dyn_lock_fr_r_04, bone2Name=dyn_lock_fr_r_03, swing1LimitDegrees=3, swing2LimitDegrees=15, twistLimitDegrees=3, twistMotion=Locked}
 {bone1Name=dyn_lock_fr_r_05, bone2Name=dyn_lock_fr_r_04, swing1LimitDegrees=3, swing2LimitDegrees=15, twistLimitDegrees=3, twistMotion=Locked}
 {bone1Name=dyn_lock_fr_r_06, bone2Name=dyn_lock_fr_r_05, swing1LimitDegrees=3, swing2LimitDegrees=15, twistLimitDegrees=3, twistMotion=Locked}
 {bone1Name=dyn_lock_fr_r_07, bone2Name=dyn_lock_fr_r_06, swing1LimitDegrees=3, swing2LimitDegrees=15, twistLimitDegrees=3, twistMotion=Locked}
 {bone1Name=dyn_hair_02_lf, bone2Name=base_hair_02_lf, swing1LimitDegrees=36, swing2LimitDegrees=72, twistLimitDegrees=3}
 {bone1Name=dyn_hair_03_lf, bone2Name=dyn_hair_02_lf, swing1LimitDegrees=15, swing2LimitDegrees=45, twistLimitDegrees=0}
 {bone1Name=dyn_hair_04_lf, bone2Name=dyn_hair_03_lf, swing1LimitDegrees=15, swing2LimitDegrees=45, twistLimitDegrees=0}
 {bone1Name=dyn_hair_05_lf, bone2Name=dyn_hair_04_lf, swing1LimitDegrees=15, swing2LimitDegrees=45, twistLimitDegrees=0}
 {bone1Name=dyn_hair_06_lf, bone2Name=dyn_hair_05_lf, swing1LimitDegrees=15, swing2LimitDegrees=45, twistLimitDegrees=0}
 {bone1Name=dyn_hair_07_lf, bone2Name=dyn_hair_06_lf, swing1LimitDegrees=15, swing2LimitDegrees=45, twistLimitDegrees=0}
 {bone1Name=dyn_hair_02_rt, bone2Name=base_hair_02_rt, swing1LimitDegrees=36, swing2LimitDegrees=72, twistLimitDegrees=3}
 {bone1Name=dyn_hair_03_rt, bone2Name=dyn_hair_02_rt, swing1LimitDegrees=15, swing2LimitDegrees=45, twistLimitDegrees=0}
 {bone1Name=dyn_hair_04_rt, bone2Name=dyn_hair_03_rt, swing1LimitDegrees=15, swing2LimitDegrees=45, twistLimitDegrees=0}
 {bone1Name=dyn_hair_05_rt, bone2Name=dyn_hair_04_rt, swing1LimitDegrees=15, swing2LimitDegrees=45, twistLimitDegrees=0}
 {bone1Name=dyn_hair_06_rt, bone2Name=dyn_hair_05_rt, swing1LimitDegrees=15, swing2LimitDegrees=45, twistLimitDegrees=0}
 {bone1Name=dyn_hair_07_rt, bone2Name=dyn_hair_06_rt, swing1LimitDegrees=15, swing2LimitDegrees=45, twistLimitDegrees=0}
 {bone1Name=base_hair_02_lf, bone2Name=head, swing1LimitDegrees=9, swing2LimitDegrees=45, twistLimitDegrees=6}
 {bone1Name=base_hair_02_rt, bone2Name=head, swing1LimitDegrees=9, swing2LimitDegrees=45, twistLimitDegrees=6}
 {bone1Name=base_lock_fr_r_03, bone2Name=head, swing1LimitDegrees=3, swing2LimitDegrees=15, twistLimitDegrees=3, twistMotion=Locked}
 {bone1Name=base_lock_fr_l_03, bone2Name=head, swing1LimitDegrees=3, swing2LimitDegrees=15, twistLimitDegrees=3, twistMotion=Locked}
 {bone1Name=base_hair_02_cr, bone2Name=head, swing1LimitDegrees=3, swing2LimitDegrees=15, twistLimitDegrees=3, twistMotion=Locked}
 {bone1Name=dyn_hair_02_cr, bone2Name=base_hair_02_cr, swing1LimitDegrees=36, swing2LimitDegrees=72, twistLimitDegrees=3}
 {bone1Name=dyn_hair_03_cr, bone2Name=dyn_hair_02_cr, swing1LimitDegrees=15, swing2LimitDegrees=60, twistLimitDegrees=0}
 {bone1Name=dyn_hair_04_cr, bone2Name=dyn_hair_03_cr, swing1LimitDegrees=15, swing2LimitDegrees=60, twistLimitDegrees=0}
 {bone1Name=dyn_hair_05_cr, bone2Name=dyn_hair_04_cr, swing1LimitDegrees=15, swing2LimitDegrees=60, twistLimitDegrees=0}
 {bone1Name=dyn_hair_06_cr, bone2Name=dyn_hair_05_cr, swing1LimitDegrees=15, swing2LimitDegrees=60, twistLimitDegrees=0}
 {bone1Name=dyn_hair_07_cr, bone2Name=dyn_hair_06_cr, swing1LimitDegrees=15, swing2LimitDegrees=60, twistLimitDegrees=0}
 {bone1Name=dyn_hair_08_cr, bone2Name=dyn_hair_07_cr, swing1LimitDegrees=15, swing2LimitDegrees=60, twistLimitDegrees=0}
 {
  bone1Name=dyn_hair_07_cr
  bone2Name=dyn_hair_07_lf
  swing1LimitDegrees=45
  swing1Motion=Free
  swing2LimitDegrees=45
  swing2Motion=Free
  twistLimitDegrees=45
  twistMotion=Free
 }
 {
  bone1Name=dyn_hair_07_cr
  bone2Name=dyn_hair_07_rt
  swing1LimitDegrees=45
  swing1Motion=Free
  swing2LimitDegrees=45
  swing2Motion=Free
  twistLimitDegrees=45
  twistMotion=Free
 }
]
