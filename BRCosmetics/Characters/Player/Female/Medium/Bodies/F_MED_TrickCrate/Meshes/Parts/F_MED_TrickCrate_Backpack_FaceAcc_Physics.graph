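# Physics asset for the face accessory part of the female TrickCrate character cosmetic, set up for two dangling front lock chains, left and right.
# Each chain hangs from a base body constrained to the head and runs through five capsule segments with twist locked and tight swing limits.
# Capsules on the spine, clavicles, upper arms, neck and a multi-capsule head form the body volume around the locks, with free-moving constraints linking upper spine, neck and head.
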PhysicsAsset /BRCosmetics/Characters/Player/Female/Medium/Bodies/F_MED_TrickCrate/Meshes/Parts/F_MED_TrickCrate_Backpack_FaceAcc_Physics
bodies {
 base_lock_fr_l_03 [{shapeType=Capsule, length=0.505, radius=0.505, shapeName=base_lock_fr_l_03_capsule, rotation=<rot 0 -90 0>}]
 base_lock_fr_r_03 [{shapeType=Capsule, length=0.505, radius=0.505, shapeName=base_lock_fr_r_03_capsule, rotation=<rot 0 -90 0>}]
 clavicle_l [{shapeType=Capsule, length=4.4362, radius=5.5436, shapeName=clavicle_l_capsule, center=<8.8633 -0.4363 -0.656>, rotation=<rot 90 0 0>}]
 clavicle_r [{shapeType=Capsule, length=3.5389, radius=5.6622, shapeName=clavicle_l_capsule, center=<-9.3665 0.3902 0.7413>, rotation=<rot 90 0 0>}]
 dyn_lock_fr_l_03 [{shapeType=Capsule, length=2.3695, radius=1.2346, shapeName=dyn_hair_06_cr_capsule, center=<1.4 0.0365 0>, rotation=<rot 90 0 0>}]
 dyn_lock_fr_l_04 [{shapeType=Capsule, length=2.3595, radius=1.2647, shapeName=dyn_hair_06_cr_capsule, center=<1.41 0 0>, rotation=<rot 90 0 0>}]
 dyn_lock_fr_l_05 [{shapeType=Capsule, length=1.9742, radius=1.2299, shapeName=dyn_hair_06_cr_capsule, center=<1.2 0 0>, rotation=<rot 90 0 0>}]
 dyn_lock_fr_l_06 [{shapeType=Capsule, length=2.6767, radius=1.0855, shapeName=dyn_hair_06_cr_capsule, center=<1.479 0 0>, rotation=<rot 90 0 0>}]
 dyn_lock_fr_l_07 [{shapeType=Capsule, length=2.6767, radius=1.0855, shapeName=dyn_hair_06_cr_capsule, center=<1.479 0 0>, rotation=<rot 90 0 0>}]
 dyn_lock_fr_r_03 [{shapeType=Capsule, length=2.2089, radius=1.3953, shapeName=dyn_hair_06_cr_capsule, center=<1.3092 -0.2297 0.0728>, rotation=<rot 90 0 0>}]
 dyn_lock_fr_r_04 [{shapeType=Capsule, length=2.0424, radius=1.5617, shapeName=dyn_hair_06_cr_capsule, center=<1.3092 0.0543 0.0728>, rotation=<rot 90 5.8696 0>}]
 dyn_lock_fr_r_05 [{shapeType=Capsule, length=2.3201, radius=1.2841, shapeName=dyn_hair_06_cr_capsule, center=<1.3092 0.0543 0.1422>, rotation=<rot 90 0 0>}]
 dyn_lock_fr_r_06 [{shapeType=Capsule, length=2.3009, radius=1.3032, shapeName=dyn_hair_06_cr_capsule, center=<1.3092 0.0543 0.0728>, rotation=<rot 90 0 0>}]
 dyn_lock_fr_r_07 [{shapeType=Capsule, length=2.5001, radius=1.1041, shapeName=dyn_hair_06_cr_capsule, center=<1.3092 0.0543 0.0728>, rotation=<rot 90 0 0>}]
 head [
  {shapeType=Capsule, length=6.0812, radius=4.9092, shapeName=head_capsule, center=<2.5333 1.3001 0.2337>, rotation=<rot 0 -74.2468 90>}
  {
   shapeType=Capsule
   length=13.1164
   radius=1.8976
   shapeName=head_capsule
   center=<-4.2936 -1.8136 10.6455>
   rotation=<rot -11.2175 9.7203 120.1539>
  }
  {
   shapeType=Capsule
   length=13.0078
   radius=1.8964
   shapeName=head_capsule
   center=<-1.0936 -1.0635 10.8532>
   rotation=<rot -10.9505 11.4194 119.676>
  }
  {
   shapeType=Capsule
   length=13.0078
   radius=1.8964
   shapeName=head_capsule
   center=<2.3036 0.1069 11.0124>
   rotation=<rot -10.9505 11.4194 119.676>
  }
  {
   shapeType=Capsule
   length=13.0078
   radius=1.8964
   shapeName=head_capsule
   center=<2.3036 0.1069 -11.0124>
   rotation=<rot 10.9505 11.4194 -119.676>
  }
  {
   shapeType=Capsule
   length=13.0078
   radius=1.8964
   shapeName=head_capsule
   center=<-1.0936 -1.0635 -10.8532>
   rotation=<rot 10.9505 11.4194 -119.676>
  }
  {
   shapeType=Capsule
   length=13.1164
   radius=1.8976
   shapeName=head_capsule
   center=<-4.2936 -1.8136 -10.6455>
   rotation=<rot 11.2175 9.7203 -120.1539>
  }
  {shapeType=Capsule, length=4.8798, radius=6.1106, shapeName=head_capsule, center=<2.5855 5.4787 0.2337>, rotation=<rot 0 -74.2468 90>}
 ]
 neck_01 [{shapeType=Capsule, length=4.3636, radius=3.6364, shapeName=neck_01_capsule, center=<3 0 0>, rotation=<rot 90 0 0>}]
 neck_02 [{shapeType=Capsule, length=2.7128, radius=5.2872, shapeName=neck_01_capsule, center=<6.8712 0 0.139>, rotation=<rot 90 0 0>}]
 spine_01 [{shapeType=Capsule, length=2.3701, radius=12.7746, shapeName=spine_01_capsule, center=<-3.9777 -0.1407 0>, rotation=<rot 90 23.6731 0>}]
 spine_02 [{shapeType=Capsule, length=1.029, radius=9.4591, shapeName=spine_02_capsule, center=<-1.5174 1.2345 0>, rotation=<rot 90 11.5144 0>}]
 spine_03 [{shapeType=Capsule, length=2.0072, radius=9.4591, shapeName=spine_02_capsule, center=<-1.5174 1.2345 0>, rotation=<rot 90 11.5144 0>}]
 spine_04 [{
   shapeType=Capsule
   length=16.7976
   radius=11.6123
   shapeName=spine_02_capsule
   center=<1.9583 1.0733 -0.0334>
   rotation=<rot -0.4595 -174.6641 -179.9504>
  }]
 spine_05 [{shapeType=Capsule, length=23.1524, radius=6.4958, shapeName=spine_05_capsule, center=<3.3011 -1.6697 0>, rotation=<rot 0 -90 0>}]
 upperarm_l [{shapeType=Capsule, length=24.4316, radius=5.8184, shapeName=upperarm_l_capsule, center=<13 0 0>, rotation=<rot 90 0 0>}]
 upperarm_r [{shapeType=Capsule, length=24.4465, radius=5.8035, shapeName=upperarm_r_capsule, center=<-13 0 0>, rotation=<rot 90 0 0>}]
}
constraints [
 {
  bone1Name=neck_02
  bone2Name=neck_01
  swing1LimitDegrees=45
  swing1Motion=Free
  swing2LimitDegrees=45
  swing2Motion=Free
  twistLimitDegrees=45
  twistMotion=Free
 }
 {
  bone1Name=neck_01
  bone2Name=spine_05
  swing1LimitDegrees=45
  swing1Motion=Free
  swing2LimitDegrees=45
  swing2Motion=Free
  twistLimitDegrees=45
  twistMotion=Free
 }
 {
  bone1Name=head
  bone2Name=neck_02
  swing1LimitDegrees=45
  swing1Motion=Free
  swing2LimitDegrees=45
  swing2Motion=Free
  twistLimitDegrees=45
  twistMotion=Free
 }
 {
  bone1Name=dyn_lock_fr_l_03
  bone2Name=base_lock_fr_l_03
  swing1LimitDegrees=3
  swing2LimitDegrees=15
  twistLimitDegrees=3
  twistMotion=Locked
 }
 {bone1Name=dyn_lock_fr_l_04, bone2Name=dyn_lock_fr_l_03, swing1LimitDegrees=3, swing2LimitDegrees=15, twistLimitDegrees=3, twistMotion=Locked}
 {bone1Name=dyn_lock_fr_l_05, bone2Name=dyn_lock_fr_l_04, swing1LimitDegrees=3, swing2LimitDegrees=15, twistLimitDegrees=3, twistMotion=Locked}
 {bone1Name=dyn_lock_fr_l_06, bone2Name=dyn_lock_fr_l_05, swing1LimitDegrees=3, swing2LimitDegrees=15, twistLimitDegrees=3, twistMotion=Locked}
 {bone1Name=dyn_lock_fr_l_07, bone2Name=dyn_lock_fr_l_06, swing1LimitDegrees=3, swing2LimitDegrees=15, twistLimitDegrees=3, twistMotion=Locked}
 {
  bone1Name=dyn_lock_fr_r_03
  bone2Name=base_lock_fr_r_03
  swing1LimitDegrees=3
  swing2LimitDegrees=15
  twistLimitDegrees=3
  twistMotion=Locked
 }
 {bone1Name=dyn_lock_fr_r_04, bone2Name=dyn_lock_fr_r_03, swing1LimitDegrees=3, swing2LimitDegrees=15, twistLimitDegrees=3, twistMotion=Locked}
 {bone1Name=dyn_lock_fr_r_05, bone2Name=dyn_lock_fr_r_04, swing1LimitDegrees=3, swing2LimitDegrees=15, twistLimitDegrees=3, twistMotion=Locked}
 {bone1Name=dyn_lock_fr_r_06, bone2Name=dyn_lock_fr_r_05, swing1LimitDegrees=3, swing2LimitDegrees=15, twistLimitDegrees=3, twistMotion=Locked}
 {bone1Name=dyn_lock_fr_r_07, bone2Name=dyn_lock_fr_r_06, swing1LimitDegrees=3, swing2LimitDegrees=15, twistLimitDegrees=3, twistMotion=Locked}
 {bone1Name=base_lock_fr_r_03, bone2Name=head, swing1LimitDegrees=3, swing2LimitDegrees=15, twistLimitDegrees=3, twistMotion=Locked}
 {bone1Name=base_lock_fr_l_03, bone2Name=head, swing1LimitDegrees=3, swing2LimitDegrees=15, twistLimitDegrees=3, twistMotion=Locked}
]
