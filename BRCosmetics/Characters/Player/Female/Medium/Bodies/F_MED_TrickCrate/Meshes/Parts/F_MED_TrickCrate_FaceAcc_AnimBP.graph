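# Animation Blueprint for the TrickCrate female face-accessory mesh, driving hair motion that adapts to the worn back bling.
# Copies the parent mesh pose through a Control Rig, then runs RigidBody simulation with separate physics assets for cape, other backpack or none.
# The cape branch adds hand-tuned bone offsets and constraints.
# Each update detects the backpack by mesh name and selects the parameter set driving the simulation's gravity override.
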AnimBlueprint /BRCosmetics/Characters/Player/Female/Medium/Bodies/F_MED_TrickCrate/Meshes/Parts/F_MED_TrickCrate_FaceAcc_AnimBP
target_skeleton=&/F_MED_TrickCrate_FaceAcc_Skeleton
graphs [
 AnimGraph {nodes [
   n1 = Control Rig!(Source=Copy Pose From Mesh() {bUseAttachedParent=true}, Alpha, BackPack On:bool=true) -> (Pose) {controlRigAssetReference {blueprintRigClass=&/F_MED_TrickCrate_Spl_CtrlRig.F_MED_TrickCrate_Spl_CtrlRig_C}}
   n2 = RigidBody!(
    ComponentPose=Local To Component(LocalPose=@n1)
    Alpha
    OverrideWorldGravity:Vector=Get GravityOverride()
    ExternalForce
    SimSpaceSettings:Sim Space Settings Structure=(WorldAlpha=0.4,VelocityScaleZ=1,DampingAlpha=1,MaxLinearVelocity=10000,MaxAngularVelocity=10000,MaxLinearAcceleration=10000,MaxAngularAcceleration=10000,ExternalLinearDragV=(X=-0.5,Y=0,Z=0),ExternalLinearVelocity=(X=-0.5,Y=0,Z=0),ExternalAngularVelocity=(X=-0.5,Y=0,Z=0))
   ) -> (Pose) {
    bClampLinearTranslationLimitToRefPose=true
    bOverrideWorldGravity=true
    overridePhysicsAsset=&/F_MED_TrickCrate_Backpack_FaceAcc_Physics
    simulationSpace=BaseBoneSpace
    alphaBoolBlend {blendOption=Linear}
    baseBoneRef {boneName=spine_04}
    overrideWorldGravity {z=-980}
    simSpaceSettings {worldAlpha=0.4, externalAngularVelocity {x=-0.5}, externalLinearDragV {x=-0.5}, externalLinearVelocity {x=-0.5}}
   }
   n3 = Component To Local(ComponentPose=@n2)
   n4 = "Transform (Modify) Bone - Bone: dyn_hair_04_cr"!(
    ComponentPose=Local To Component(LocalPose=Use cached pose 'hair ctrl rig pose'())
    bAlphaBoolEnabled:bool=True
    Translation="0,0,0"
    Rotation="40,0,0"
    Scale="1,1,1"
   ) -> (Pose) {
    alphaInputType=Bool
    rotationMode=BMM_Additive
    rotationSpace=BCS_BoneSpace
    translationMode=BMM_Additive
    translationSpace=BCS_BoneSpace
    alphaBoolBlend {blendInTime=0.2, blendOutTime=0.2}
    boneToModify {boneName=dyn_hair_04_cr}
    rotation {pitch=40}
   }
   n5 = "Transform (Modify) Bone - Bone: dyn_hair_03_rt"!(ComponentPose=@n4, bAlphaBoolEnabled:bool=True, Translation="0,0,0", Rotation="30,-20,20", Scale="1,1,1") -> (Pose) {
    alphaInputType=Bool
    rotationMode=BMM_Additive
    rotationSpace=BCS_BoneSpace
    translationMode=BMM_Additive
    translationSpace=BCS_BoneSpace
    alphaBoolBlend {blendInTime=0.2, blendOutTime=0.2}
    boneToModify {boneName=dyn_hair_03_rt}
    rotation=<rot 30 -20 20>
   }
   n6 = "Transform (Modify) Bone - Bone: dyn_hair_03_lf"!(ComponentPose=@n5, bAlphaBoolEnabled:bool=True, Translation="0,0,0", Rotation="35,10,0", Scale="1,1,1") -> (Pose) {
    alphaInputType=Bool
    rotationMode=BMM_Additive
    rotationSpace=BCS_BoneSpace
    translationMode=BMM_Additive
    translationSpace=BCS_BoneSpace
    alphaBoolBlend {blendInTime=0.2, blendOutTime=0.2}
    boneToModify {boneName=dyn_hair_03_lf}
    rotation {pitch=35, yaw=10}
   }
   n7 = "Transform (Modify) Bone - Bone: dyn_hair_05_cr"!(ComponentPose=@n6, bAlphaBoolEnabled:bool=True, Translation="0,0,0", Rotation="30,0,0", Scale="1,1,1") -> (Pose) {
    alphaInputType=Bool
    rotationMode=BMM_Additive
    rotationSpace=BCS_BoneSpace
    translationMode=BMM_Additive
    translationSpace=BCS_BoneSpace
    alphaBoolBlend {blendInTime=0.2, blendOutTime=0.2}
    boneToModify {boneName=dyn_hair_05_cr}
    rotation {pitch=30}
   }
   n8 = "Transform (Modify) Bone - Bone: dyn_hair_05_rt"!(ComponentPose=@n7, bAlphaBoolEnabled:bool=True, Translation="0,0,0", Rotation="50,0,0", Scale="1,1,1") -> (Pose) {
    alphaInputType=Bool
    rotationMode=BMM_Additive
    rotationSpace=BCS_BoneSpace
    translationMode=BMM_Additive
    translationSpace=BCS_BoneSpace
    alphaBoolBlend {blendInTime=0.2, blendOutTime=0.2}
    boneToModify {boneName=dyn_hair_05_rt}
    rotation {pitch=50}
   }
   n9 = "Transform (Modify) Bone - Bone: dyn_hair_05_lf"!(ComponentPose=@n8, bAlphaBoolEnabled:bool=True, Translation="0,0,0", Rotation="30,0,0", Scale="1,1,1") -> (Pose) {
    alphaInputType=Bool
    rotationMode=BMM_Additive
    rotationSpace=BCS_BoneSpace
    translationMode=BMM_Additive
    translationSpace=BCS_BoneSpace
    alphaBoolBlend {blendInTime=0.2, blendOutTime=0.2}
    boneToModify {boneName=dyn_hair_05_lf}
    rotation {pitch=30}
   }
   n10 = "Constraint - dyn_hair_05_cr (P)"(ComponentPose=@n9, Alpha=1, ConstraintWeights_0=1) {boneToModify {boneName=dyn_hair_05_cr}, constraintSetup [{targetBone {boneName=dyn_hair_04_cr}}], constraintWeights [1]}
   n11 = "Constraint - dyn_hair_06_cr (P)"(ComponentPose=@n10, Alpha=1, ConstraintWeights_0=1) {boneToModify {boneName=dyn_hair_06_cr}, constraintSetup [{targetBone {boneName=dyn_hair_05_cr}}], constraintWeights [1]}
   n12 = "Constraint - dyn_hair_05_rt (P)"(ComponentPose=@n11, Alpha=1, ConstraintWeights_0=1) {boneToModify {boneName=dyn_hair_05_rt}, constraintSetup [{targetBone {boneName=dyn_hair_04_rt}}], constraintWeights [1]}
   n13 = "Constraint - dyn_hair_06_rt (P)"(ComponentPose=@n12, Alpha=1, ConstraintWeights_0=1) {boneToModify {boneName=dyn_hair_06_rt}, constraintSetup [{targetBone {boneName=dyn_hair_05_rt}}], constraintWeights [1]}
   n14 = Component To Local(ComponentPose=@n13)
   n15 = Use cached pose 'hair ctrl rig pose'()
   n16 = Get bRigidBodyEnabled()
   n17 = RigidBody(
    ComponentPose=Local To Component(LocalPose=Blend Poses by bool(BlendPose_0=@n15, BlendPose_1=@n14, bActiveValue=@n16))
    OverrideWorldGravity=Get GravityOverride()
    SimSpaceSettings=(WorldAlpha=0.4,VelocityScaleZ=1,DampingAlpha=1,MaxLinearVelocity=10000,MaxAngularVelocity=10000,MaxLinearAcceleration=10000,MaxAngularAcceleration=10000,ExternalLinearDragV=(X=-0.5,Y=0,Z=0),ExternalLinearVelocity=(X=-0.5,Y=0,Z=0),ExternalAngularVelocity=(X=-0.5,Y=0,Z=0))
   ) {
    bClampLinearTranslationLimitToRefPose=true
    bOverrideWorldGravity=true
    overridePhysicsAsset=&/F_MED_TrickCrate_FaceAcc_Physics
    simulationSpace=BaseBoneSpace
    alphaBoolBlend {blendOption=Linear}
    baseBoneRef {boneName=spine_04}
    componentAppliedLinearAccClamp=<700 700 700>
    overrideWorldGravity {z=-980}
    simSpaceSettings {worldAlpha=0.4, externalAngularVelocity {x=-0.5}, externalLinearDragV {x=-0.5}, externalLinearVelocity {x=-0.5}}
   }
   n18 = "Transform (Modify) Bone - Bone: dyn_hair_05_lf"(ComponentPose=@n17, bAlphaBoolEnabled=True, Translation="0,0,0", Rotation="10,0,0", Scale="1,1,1") {
    alphaInputType=Bool
    rotationMode=BMM_Additive
    rotationSpace=BCS_BoneSpace
    translationMode=BMM_Additive
    translationSpace=BCS_BoneSpace
    alphaBoolBlend {blendInTime=0.2, blendOutTime=0.2}
    boneToModify {boneName=dyn_hair_05_lf}
    rotation {pitch=10}
   }
   n19 = "Transform (Modify) Bone - Bone: dyn_hair_05_rt"(ComponentPose=@n18, bAlphaBoolEnabled=True, Translation="0,0,2", Rotation="15,0,0", Scale="1,1,1") {
    alphaInputType=Bool
    rotationMode=BMM_Additive
    rotationSpace=BCS_BoneSpace
    translationMode=BMM_Additive
    translationSpace=BCS_BoneSpace
    alphaBoolBlend {blendInTime=0.2, blendOutTime=0.2}
    boneToModify {boneName=dyn_hair_05_rt}
    rotation {pitch=15}
    translation {z=2}
   }
   n20 = "Transform (Modify) Bone - Bone: base_hair_01_lf"(ComponentPose=@n19, Alpha=Get HeadUp_Rot() -> (HeadUp_Rot:double), Translation="0,0,1", Rotation="10,0,0", Scale="1,1,1") {
    rotationMode=BMM_Additive
    rotationSpace=BCS_BoneSpace
    translationMode=BMM_Additive
    translationSpace=BCS_BoneSpace
    boneToModify {boneName=base_hair_01_lf}
    rotation {pitch=10}
    translation {z=1}
   }
   n21 = "Transform (Modify) Bone - Bone: dyn_hair_05_lf"(ComponentPose=@n20, bAlphaBoolEnabled=True, Translation="0,0,1", Rotation="0,0,0", Scale="1,1,1") {
    alphaInputType=Bool
    rotationMode=BMM_Additive
    rotationSpace=BCS_BoneSpace
    translationMode=BMM_Additive
    translationSpace=BCS_BoneSpace
    alphaBoolBlend {blendInTime=0.2, blendOutTime=0.2}
    boneToModify {boneName=dyn_hair_05_lf}
    translation {z=1}
   }
   n22 = Component To Local(ComponentPose=@n21)
   n23 = Get IsTheCape() -> (IsTheCape:bool)
   n24 = RigidBody(
    ComponentPose=Local To Component(LocalPose=Use cached pose 'hair ctrl rig pose'())
    OverrideWorldGravity=Get GravityOverride()
    SimSpaceSettings=(WorldAlpha=0.4,VelocityScaleZ=1,DampingAlpha=1,MaxLinearVelocity=10000,MaxAngularVelocity=10000,MaxLinearAcceleration=10000,MaxAngularAcceleration=10000,ExternalLinearDragV=(X=-0.5,Y=0,Z=0),ExternalLinearVelocity=(X=-0.5,Y=0,Z=0),ExternalAngularVelocity=(X=-0.5,Y=0,Z=0))
   ) {
    bClampLinearTranslationLimitToRefPose=true
    bOverrideWorldGravity=true
    overridePhysicsAsset=&/F_MED_NoCape_TrickCrate_FaceAcc_Physics
    simulationSpace=BaseBoneSpace
    alphaBoolBlend {blendOption=Linear}
    baseBoneRef {boneName=spine_04}
    overrideWorldGravity {z=-980}
    simSpaceSettings {worldAlpha=0.4, externalAngularVelocity {x=-0.5}, externalLinearDragV {x=-0.5}, externalLinearVelocity {x=-0.5}}
   }
   Output Pose(Result=Blend Poses by bool(
     BlendPose_0=Blend Poses by bool(BlendPose_0=@n22, BlendPose_1=@n3, bActiveValue=@n23)
     BlendPose_1=Component To Local(ComponentPose=@n24)
     bActiveValue=Get Is Wearing Backbling()
    )) {layerGroup=None}
   n25 = Control Rig!(Source=Copy Pose From Mesh() {bUseAttachedParent=true}, Alpha, IsTheCape:bool=Get IsTheCape()) -> (Pose) {controlRigAssetReference {blueprintRigClass=&F_MED_TrickCrate_Spl_CtrlRig_C}}
   Save cached pose 'hair ctrl rig pose'(Pose=@n25) {cacheName=hair ctrl rig pose}
  ]}
 EventGraph {nodes [
   Event BlueprintUpdateAnimation()
   > n1 = Sequence!(execute) -> (then_0, then_1, then_2:Exec, then_3:Exec)
   n2 = Branch(execute=@n1.then_1, Condition=Get IsTheCape())
   > n3 = "Gravity Override State Function (Parameter Asset) (Nativized)"(
    CharacterPartAnimInstance=Self-Reference()
    GameTimeInSeconds=GetGameTimeInSeconds()
    DebugJointName=neck_02
    Parameters=&/F_MED_TrickCrate_FaceAcc_Parameters
   )
   > Set GravityOverride(GravityOverride=@n3)
   > Set RBAN Settings Out(RBAN Settings Out=@n3.OutRigidBodyAnimNodeParameters)
   n4 = Get bIsCrouching()
   > Set isMoving(isMoving:bool=AND Boolean(A=Get bIsMoving2D(), B=NOT Boolean(A=@n4))) -> (Output_Get:bool)
   n5 = "Gravity Override State Function (Parameter Asset) (Nativized)"(
    execute=@n2.else
    CharacterPartAnimInstance=Self-Reference()
    GameTimeInSeconds=GetGameTimeInSeconds()
    DebugJointName=neck_02
    Parameters=&/F_MED_TrickCrate_FaceAcc_Parameters1
   )
   > Set GravityOverride(GravityOverride=@n5)
   > Set RBAN Settings Out(RBAN Settings Out=@n5.OutRigidBodyAnimNodeParameters)
   Set Is Wearing Backbling(execute=@n1.then_0, Is Wearing Backbling=CheckForBackbling())
   > n6 = Cast To FortPlayerPawn(Object=TryGetPawnOwner())
   n7 = GetSkeletalMeshForPartType(self=@n6, PartType=Backpack)
   > Set IsTheCape(IsTheCape:bool=MatchesWildcard(SourceString=GetDisplayName(Object=@n7), Wildcard=*F_MED_TrickCrate_Cape*, SearchCase=CaseSensitive)) -> (Output_Get:bool)
   Event BlueprintInitializeAnimation()
   > Set Skel Mesh(Skel Mesh=GetOwnerSkeletalMeshComponent())
   n8 = Branch(Condition=Get bIsInFrontEnd())
   n9 = Cast To FortPlayerPawn(execute=@n8.else, Object=GetOwningActor())
   > n10 = "(Deprecated) Get Variant Channel Active Index For Item"!(
    execute
    Item:Athena Cosmetic Item Definition Object Reference=/BRCosmetics/Athena/Items/Cosmetics/Backpacks/Backpack_TrickCrateCape
    Loadout:"Fort Athena Loadout Structure (by ref)"=Get CosmeticLoadout!(self:Fort Player Pawn Object Reference=@n9) -> (CosmeticLoadout:Fort Athena Loadout Structure)
    Channel:Gameplay Tag Structure=(TagName="Cosmetics.Variant.Channel.Parts")
    VariantList:Array of Gameplay Tag Structures=Make Array!("[0]"=(TagName="Cosmetics.Variant.Property.Stage1"), "[1]":Gameplay Tag Structure=(TagName="Cosmetics.Variant.Property.Stage2")) -> (Array)
    bErrorIfChannelMissing:bool=false
    bErrorIfTagNotInArray:bool=false
   ) -> (then, Index:int=0, bIndexIsValid:bool, bFoundChannel:bool)
   > n11 = Set isFitted(isFitted:bool="To Boolean (Integer)"!(InInt:int=@n10) -> (ReturnValue:bool)) -> (Output_Get:bool)
   > n12 = Branch(Condition=@n11)
   > Set IsTheCape(IsTheCape=Get IsTheCape())
   Set IsTheCape(execute=@n12.else, IsTheCape=false)
   Branch()
  ]}
]
variables [
 Gravity Override Params {
  type=FortGravityOverrideParameterSet
  default=(Skydive_GoingUp=(GravityOverride=()),Skydive_FloatingDown=(GravityOverride=()),Skydive_DivingDown=(GravityOverride=()),Skydive_Parachuting=(GravityOverride=()),OnGround_Standing=(GravityOverride=()),OnGround_Moving=(GravityOverride=()),OnGround_Moving_HighRotationRate=(GravityOverride=()),OnGround_Crouch_Moving=(GravityOverride=()),EmoteOrMelee=(GravityOverride=()),Swimming=(GravityOverride=()),Front_End=(GravityOverride=()),DBNO=(GravityOverride=()),DBNO_Carried=(GravityOverride=()),OnGround_Falling=(GravityOverride=()),Ziplining=(GravityOverride=()),Skydive_Surfing=(GravityOverride=()),Riding_Standing=(GravityOverride=()),Riding_Moving=(GravityOverride=()),Riding_Moving_Targeting=(GravityOverride=()),Tactical_Sprint=(GravityOverride=()),Rail_Grinding=(GravityOverride=()),)
 }
 GravityOverride {type=Vector}
 HeadUp_Rot {type=number}
 Is Wearing Backbling {type=boolean}
 isFitted {type=boolean}
 isMoving {type=boolean}
 IsTheCape {type=boolean}
 RBAN Settings Out {type=FortRigidBodyAnimNodeParameters}
 Skel Mesh {type=SkeletalMeshComponent}
]
