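# Physics asset for the cape collider face-accessory part of the female medium TrickCrate player cosmetic, built on dyn_hair bones.
# Capsule bodies form center, left and right chains, with the center reaching one bone further.
# Most center bones add a large, differently oriented capsule.
# Constraints link neighboring bones within each chain, with no cross-links; swing and twist are all Free, so the stored angle limits are inactive.
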PhysicsAsset /BRCosmetics/Characters/Player/Female/Medium/Bodies/F_MED_TrickCrate/Meshes/Parts/F_MED_TrickCrate_FaceAcc_CapeCollider_Physics
bodies {
 dyn_hair_04_cr [
  {shapeType=Capsule, length=5.7861, radius=5.1989, shapeName=dyn_hair_04_cr_capsule, center=<10.5157 0 6.5277>, rotation=<rot 90 0 0>}
  {shapeType=Capsule, length=15, radius=15, shapeName=dyn_hair_07_cr_capsule, center=<3.4791 0 17.3294>, rotation=<rot 90 -90 0>}
 ]
 dyn_hair_04_lf [{shapeType=Capsule, length=4.6861, radius=5.1989, shapeName=dyn_hair_04_cr_capsule, center=<9.9657 1.5272 11.0929>, rotation=<rot 90 0 0>}]
 dyn_hair_04_rt [{shapeType=Capsule, length=5.1471, radius=5.1989, shapeName=dyn_hair_04_cr_capsule, center=<10.1962 -1.9687 10.3474>, rotation=<rot 90 0 0>}]
 dyn_hair_05_cr [
  {shapeType=Capsule, length=7.13, radius=3.5, shapeName=dyn_hair_05_cr_capsule, center=<10.3382 0 5.2353>, rotation=<rot 90 0 0>}
  {shapeType=Capsule, length=15, radius=15, shapeName=dyn_hair_07_cr_capsule, center=<3.4791 0 14.1432>, rotation=<rot 90 -90 0>}
 ]
 dyn_hair_05_lf [{shapeType=Capsule, length=5.4957, radius=5.2903, shapeName=dyn_hair_05_cr_capsule, center=<10.4162 -0.7161 8.7418>, rotation=<rot 90 0 0>}]
 dyn_hair_05_rt [{shapeType=Capsule, length=5.4737, radius=5.2903, shapeName=dyn_hair_05_cr_capsule, center=<8.5892 2.0712 8.0209>, rotation=<rot 90 0 0>}]
 dyn_hair_06_cr [
  {shapeType=Capsule, length=4.7997, radius=5.2903, shapeName=dyn_hair_05_cr_capsule, center=<10.0682 0 5.2353>, rotation=<rot 90 0 0>}
  {shapeType=Capsule, length=15, radius=15, shapeName=dyn_hair_07_cr_capsule, center=<3.4791 0 11.3152>, rotation=<rot 90 -90 0>}
 ]
 dyn_hair_06_lf [{shapeType=Capsule, length=2.1797, radius=4.6693, shapeName=dyn_hair_05_cr_capsule, center=<8.4477 -2.6041 8.7099>, rotation=<rot 90 0 0>}]
 dyn_hair_06_rt [{shapeType=Capsule, length=3.3658, radius=4.6693, shapeName=dyn_hair_05_cr_capsule, center=<9.0407 3.1347 7.9257>, rotation=<rot 90 0 0>}]
 dyn_hair_07_cr [
  {shapeType=Capsule, length=5.8308, radius=4.6693, shapeName=dyn_hair_05_cr_capsule, center=<10.2732 0 5.2353>, rotation=<rot 90 0 0>}
  {shapeType=Capsule, length=15, radius=15, shapeName=dyn_hair_07_cr_capsule, center=<3.4791 0 11.9272>, rotation=<rot 90 -90 0>}
 ]
 dyn_hair_07_lf [{shapeType=Capsule, length=4.3887, radius=5.39, shapeName=dyn_hair_05_cr_capsule, center=<11.3288 -2.9979 8.93>, rotation=<rot 90 0 0>}]
 dyn_hair_07_rt [{shapeType=Capsule, length=4.3887, radius=5.39, shapeName=dyn_hair_05_cr_capsule, center=<11.3288 4.2529 7.8663>, rotation=<rot 90 0 0>}]
 dyn_hair_08_cr [{shapeType=Capsule, length=8.3887, radius=5.39, shapeName=dyn_hair_05_cr_capsule, center=<13.3288 0 5.2353>, rotation=<rot 90 0 0>}]
}
constraints [
 {
  bone1Name=dyn_hair_07_rt
  bone2Name=dyn_hair_06_rt
  swing1LimitDegrees=9
  swing1Motion=Free
  swing2LimitDegrees=45
  swing2Motion=Free
  twistLimitDegrees=0
  twistMotion=Free
 }
 {
  bone1Name=dyn_hair_05_cr
  bone2Name=dyn_hair_04_cr
  swing1LimitDegrees=9
  swing1Motion=Free
  swing2LimitDegrees=60
  swing2Motion=Free
  twistLimitDegrees=0
  twistMotion=Free
 }
 {
  bone1Name=dyn_hair_06_cr
  bone2Name=dyn_hair_05_cr
  swing1LimitDegrees=9
  swing1Motion=Free
  swing2LimitDegrees=60
  swing2Motion=Free
  twistLimitDegrees=0
  twistMotion=Free
 }
 {
  bone1Name=dyn_hair_07_cr
  bone2Name=dyn_hair_06_cr
  swing1LimitDegrees=9
  swing1Motion=Free
  swing2LimitDegrees=60
  swing2Motion=Free
  twistLimitDegrees=0
  twistMotion=Free
 }
 {
  bone1Name=dyn_hair_08_cr
  bone2Name=dyn_hair_07_cr
  swing1LimitDegrees=9
  swing1Motion=Free
  swing2LimitDegrees=60
  swing2Motion=Free
  twistLimitDegrees=0
  twistMotion=Free
 }
 {
  bone1Name=dyn_hair_05_lf
  bone2Name=dyn_hair_04_lf
  swing1LimitDegrees=9
  swing1Motion=Free
  swing2LimitDegrees=45
  swing2Motion=Free
  twistLimitDegrees=0
  twistMotion=Free
 }
 {
  bone1Name=dyn_hair_06_lf
  bone2Name=dyn_hair_05_lf
  swing1LimitDegrees=9
  swing1Motion=Free
  swing2LimitDegrees=45
  swing2Motion=Free
  twistLimitDegrees=0
  twistMotion=Free
 }
 {
  bone1Name=dyn_hair_07_lf
  bone2Name=dyn_hair_06_lf
  swing1LimitDegrees=9
  swing1Motion=Free
  swing2LimitDegrees=45
  swing2Motion=Free
  twistLimitDegrees=0
  twistMotion=Free
 }
 {
  bone1Name=dyn_hair_05_rt
  bone2Name=dyn_hair_04_rt
  swing1LimitDegrees=9
  swing1Motion=Free
  swing2LimitDegrees=45
  swing2Motion=Free
  twistLimitDegrees=0
  twistMotion=Free
 }
 {
  bone1Name=dyn_hair_06_rt
  bone2Name=dyn_hair_05_rt
  swing1LimitDegrees=9
  swing1Motion=Free
  swing2LimitDegrees=45
  swing2Motion=Free
  twistLimitDegrees=0
  twistMotion=Free
 }
]
